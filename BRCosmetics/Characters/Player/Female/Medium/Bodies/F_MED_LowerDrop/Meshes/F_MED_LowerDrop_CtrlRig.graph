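# Control Rig for the F_MED_LowerDrop female body cosmetic that drives its secondary bones: buttons, waistband, loops and left/right pad chains.
# Weighted multi-parent constraints make these bones follow spine, pelvis, limbs and helper bones.
# Interpolated targeting, weapon and swing flags feed blend weights for pad correctives.
# ModifyTransforms nodes apply those correctives, also weighted by deform-rig animation attributes and upper-arm spherical pose readers, with a front-end variant.
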ControlRigBlueprint /BRCosmetics/Characters/Player/Female/Medium/Bodies/F_MED_LowerDrop/Meshes/F_MED_LowerDrop_CtrlRig
graphs [
 Multiply_13_ContainedGraph {nodes [
   n1 = RigVMFunctionEntryNode()
   RigVMFunctionReturnNode(Result=&RigVMFunction_MathDoubleMul(B=@n1.C, A=&RigVMFunction_MathDoubleMul(A=@n1.A, B=@n1.B).Result).Result)
  ]}
 Multiply_14_ContainedGraph {nodes [
   n1 = RigVMFunctionEntryNode()
   RigVMFunctionReturnNode(Result=&RigVMFunction_MathDoubleMul(B=@n1.C, A=&RigVMFunction_MathDoubleMul(A=@n1.A, B=@n1.B).Result).Result)
  ]}
 Multiply_8_ContainedGraph {nodes [
   n1 = RigVMFunctionEntryNode()
   RigVMFunctionReturnNode(Result=&RigVMFunction_MathDoubleMul(B=@n1.C, A=&RigVMFunction_MathDoubleMul(A=@n1.A, B=@n1.B).Result).Result)
  ]}
 Multiply_8_ContainedGraph {nodes [
   n1 = RigVMFunctionEntryNode()
   RigVMFunctionReturnNode(Result=&RigVMFunction_MathDoubleMul(B=@n1.C, A=&RigVMFunction_MathDoubleMul(A=@n1.A, B=@n1.B).Result).Result)
  ]}
 Multiply_8_ContainedGraph {nodes [
   n1 = RigVMFunctionEntryNode()
   RigVMFunctionReturnNode(Result=&RigVMFunction_MathDoubleMul(B=@n1.C, A=&RigVMFunction_MathDoubleMul(A=@n1.A, B=@n1.B).Result).Result)
  ]}
 New Function_ContainedGraph {nodes [
   RigVMFunctionEntryNode()
   > RigVMFunction_Sequence = RigVMAggregateNode()
   RigVMFunctionReturnNode(ExecuteContext=@RigVMFunction_Sequence.E)
   &RigUnit_ParentConstraint(
    AdvancedSettings=(RotationOrderForFilter=XZY)
    Child=(Type=Bone,Name="button_l_06")
    Parents="((Item=(Type=Bone,Name=\"spine_03\"),Weight=0.0005),(Item=(Type=Bone,Name=\"upperarm_twist_01_l\"),Weight=0.0015),(Item=(Type=Bone,Name=\"clavicle_l\"),Weight=0.0048),(Item=(Type=Bone,Name=\"spine_05\"),Weight=0.012),(Item=(Type=Bone,Name=\"spine_04\"),Weight=0.0402 ...(truncated)"
    Weight=1
    ExecutePin=@RigVMFunction_Sequence.A
   )
   > &RigUnit_ParentConstraint(
    AdvancedSettings=(RotationOrderForFilter=XZY)
    Child=(Type=Bone,Name="button_l_07")
    Parents="((Item=(Type=Bone,Name=\"spine_03\"),Weight=0.0014),(Item=(Type=Bone,Name=\"upperarm_twist_01_l\"),Weight=0.0014),(Item=(Type=Bone,Name=\"dyn_side_jacket_l\"),Weight=0.0195),(Item=(Type=Bone,Name=\"clavicle_l\"),Weight=0.0199),(Item=(Type=Bone,Name=\"spine_05\"),Wei ...(truncated)"
    Weight=1
   )
   > &RigUnit_ParentConstraint(
    AdvancedSettings=(RotationOrderForFilter=XZY)
    Child=(Type=Bone,Name="button_l_08")
    Parents="((Item=(Type=Bone,Name=\"spine_02\"),Weight=0.0013),(Item=(Type=Bone,Name=\"upperarm_twist_01_l\"),Weight=0.0014),(Item=(Type=Bone,Name=\"spine_03\"),Weight=0.0019),(Item=(Type=Bone,Name=\"dyn_side_jacket_l\"),Weight=0.0032),(Item=(Type=Bone,Name=\"clavicle_l\"),Wei ...(truncated)"
    Weight=1
   )
   > &RigUnit_ParentConstraint(
    AdvancedSettings=(RotationOrderForFilter=XZY)
    Child=(Type=Bone,Name="button_l_09")
    Parents=((Item=(Type=Bone,Name="clavicle_l"),Weight=0.0521),(Item=(Type=Bone,Name="spine_05"),Weight=0.2841),(Item=(Type=Bone,Name="spine_04"),Weight=0.6637))
    Weight=1
   )
   > &RigUnit_ParentConstraint(
    AdvancedSettings=(RotationOrderForFilter=XZY)
    Child=(Type=Bone,Name="button_l_10")
    Parents=((Item=(Type=Bone,Name="neck_02"),Weight=0.0008),(Item=(Type=Bone,Name="neck_01"),Weight=0.0029),(Item=(Type=Bone,Name="spine_04"),Weight=0.0108),(Item=(Type=Bone,Name="clavicle_l"),Weight=0.0522),(Item=(Type=Bone,Name="spine_05"),Weight=0.9333))
    Weight=1
   )
   > &RigUnit_ParentConstraint(
    AdvancedSettings=(RotationOrderForFilter=XZY)
    Child=(Type=Bone,Name="button_r_10")
    Parents=((Item=(Type=Bone,Name="neck_02"),Weight=0.0008),(Item=(Type=Bone,Name="neck_01"),Weight=0.0029),(Item=(Type=Bone,Name="spine_04"),Weight=0.0108),(Item=(Type=Bone,Name="clavicle_r"),Weight=0.0522),(Item=(Type=Bone,Name="spine_05"),Weight=0.9333))
    Weight=1
   )
   > &RigUnit_ParentConstraint(
    AdvancedSettings=(RotationOrderForFilter=XZY)
    Child=(Type=Bone,Name="button_r_09")
    Parents=((Item=(Type=Bone,Name="clavicle_r"),Weight=0.0521),(Item=(Type=Bone,Name="spine_05"),Weight=0.2841),(Item=(Type=Bone,Name="spine_04"),Weight=0.6637))
    Weight=1
   )
   > &RigUnit_ParentConstraint(
    AdvancedSettings=(RotationOrderForFilter=XZY)
    Child=(Type=Bone,Name="button_r_08")
    Parents="((Item=(Type=Bone,Name=\"spine_02\"),Weight=0.0013),(Item=(Type=Bone,Name=\"upperarm_twist_01_r\"),Weight=0.0014),(Item=(Type=Bone,Name=\"spine_03\"),Weight=0.0019),(Item=(Type=Bone,Name=\"dyn_side_jacket_r\"),Weight=0.0032),(Item=(Type=Bone,Name=\"clavicle_r\"),Wei ...(truncated)"
    Weight=1
   )
   > &RigUnit_ParentConstraint(
    AdvancedSettings=(RotationOrderForFilter=XZY)
    Child=(Type=Bone,Name="button_r_07")
    Parents="((Item=(Type=Bone,Name=\"spine_03\"),Weight=0.0014),(Item=(Type=Bone,Name=\"upperarm_twist_01_r\"),Weight=0.0014),(Item=(Type=Bone,Name=\"dyn_side_jacket_r\"),Weight=0.0195),(Item=(Type=Bone,Name=\"clavicle_r\"),Weight=0.0199),(Item=(Type=Bone,Name=\"spine_05\"),Wei ...(truncated)"
    Weight=1
   )
   > &RigUnit_ParentConstraint(
    AdvancedSettings=(RotationOrderForFilter=XZY)
    Child=(Type=Bone,Name="button_r_06")
    Parents="((Item=(Type=Bone,Name=\"spine_03\"),Weight=0.0005),(Item=(Type=Bone,Name=\"upperarm_twist_01_r\"),Weight=0.0015),(Item=(Type=Bone,Name=\"clavicle_r\"),Weight=0.0048),(Item=(Type=Bone,Name=\"spine_05\"),Weight=0.012),(Item=(Type=Bone,Name=\"spine_04\"),Weight=0.0402 ...(truncated)"
    Weight=1
   )
   &RigUnit_ParentConstraint(
    AdvancedSettings=(RotationOrderForFilter=XZY)
    Child=(Type=Bone,Name="Base_WaistBand")
    Parents="((Item=(Type=Bone,Name=\"thigh_r\"),Weight=0.001),(Item=(Type=Bone,Name=\"dyn_front_skirt_r\"),Weight=0.001),(Item=(Type=Bone,Name=\"thigh_twist_01_l\"),Weight=0.0011),(Item=(Type=Bone,Name=\"spine_01\"),Weight=0.0015),(Item=(Type=Bone,Name=\"spine_02\"),Weight=0.02 ...(truncated)"
    Weight=1
    ExecutePin=@RigVMFunction_Sequence.B
   )
   > &RigUnit_ParentConstraint(
    AdvancedSettings=(RotationOrderForFilter=XZY)
    Child=(Type=Bone,Name="base_Loop_A")
    Parents=((Item=(Type=Bone,Name="spine_04"),Weight=0.0126),(Item=(Type=Bone,Name="spine_01"),Weight=0.0468),(Item=(Type=Bone,Name="spine_02"),Weight=0.0788),(Item=(Type=Bone,Name="pelvis"),Weight=0.8618))
    Weight=1
   )
   > &RigUnit_ParentConstraint(
    AdvancedSettings=(RotationOrderForFilter=XZY)
    Child=(Type=Bone,Name="base_Loop_B")
    Parents=((Item=(Type=Bone,Name="spine_04"),Weight=0.0173),(Item=(Type=Bone,Name="spine_01"),Weight=0.0582),(Item=(Type=Bone,Name="spine_02"),Weight=0.1006),(Item=(Type=Bone,Name="pelvis"),Weight=0.8239))
    Weight=1
   )
   > &RigUnit_ParentConstraint(
    AdvancedSettings=(RotationOrderForFilter=XZY)
    Child=(Type=Bone,Name="base_Loop_C")
    Parents=((Item=(Type=Bone,Name="spine_04"),Weight=0.022),(Item=(Type=Bone,Name="spine_01"),Weight=0.0707),(Item=(Type=Bone,Name="spine_02"),Weight=0.1226),(Item=(Type=Bone,Name="pelvis"),Weight=0.7846))
    Weight=1
   )
   > &RigUnit_ParentConstraint(
    AdvancedSettings=(RotationOrderForFilter=XZY)
    Child=(Type=Bone,Name="Waistband_Collider_01")
    Parents="((Item=(Type=Bone,Name=\"spine_02\"),Weight=0.0037),(Item=(Type=Bone,Name=\"thigh_r\"),Weight=0.0081),(Item=(Type=Bone,Name=\"dyn_front_skirt_c\"),Weight=0.0151),(Item=(Type=Bone,Name=\"dyn_side_skirt_r\"),Weight=0.0512),(Item=(Type=Bone,Name=\"dyn_front_skirt_r\"), ...(truncated)"
    Weight=1
   )
   > &RigUnit_ParentConstraint(
    AdvancedSettings=(RotationOrderForFilter=XZY)
    Child=(Type=Bone,Name="Waistband_Collider_02")
    Parents="((Item=(Type=Bone,Name=\"dyn_front_skirt_l\"),Weight=0.0025),(Item=(Type=Bone,Name=\"thigh_r\"),Weight=0.0031),(Item=(Type=Bone,Name=\"deform_groin_r\"),Weight=0.0518),(Item=(Type=Bone,Name=\"thigh_twist_01_r\"),Weight=0.0571),(Item=(Type=Bone,Name=\"dyn_front_skir ...(truncated)"
    Weight=1
   )
   > &RigUnit_ParentConstraint(
    AdvancedSettings=(RotationOrderForFilter=XZY)
    Child=(Type=Bone,Name="Waistband_Collider_03")
    Parents="((Item=(Type=Bone,Name=\"dyn_front_skirt_l\"),Weight=0.0004),(Item=(Type=Bone,Name=\"dyn_bk_skirt_r\"),Weight=0.0006),(Item=(Type=Bone,Name=\"deform_groin_r\"),Weight=0.0028),(Item=(Type=Bone,Name=\"thigh_twist_01_r\"),Weight=0.0031),(Item=(Type=Bone,Name=\"pelvis\" ...(truncated)"
    Weight=1
   )
   > &RigUnit_ParentConstraint(
    AdvancedSettings=(RotationOrderForFilter=XZY)
    Child=(Type=Bone,Name="Waistband_Collider_04")
    Parents="((Item=(Type=Bone,Name=\"thigh_r\"),Weight=0.0022),(Item=(Type=Bone,Name=\"dyn_front_skirt_l\"),Weight=0.005),(Item=(Type=Bone,Name=\"thigh_twist_01_r\"),Weight=0.0476),(Item=(Type=Bone,Name=\"deform_groin_r\"),Weight=0.0536),(Item=(Type=Bone,Name=\"dyn_side_skirt_ ...(truncated)"
    Weight=1
   )
   > &RigUnit_ParentConstraint(
    AdvancedSettings=(RotationOrderForFilter=XZY)
    Child=(Type=Bone,Name="base_padChain_r_Helper")
    Parents="((Item=(Type=Bone,Name=\"base_padChain_r_Helper1\"),Weight=0.0488),(Item=(Type=Bone,Name=\"upperarm_r\"),Weight=0.0865),(Item=(Type=Bone,Name=\"clavicle_r\"),Weight=0.2193),(Item=(Type=Bone,Name=\"deform_upperarm_up_r\"),Weight=0.2355),(Item=(Type=Bone,Name=\"upper ...(truncated)"
    Weight=1
   )
   &RigUnit_ParentConstraint(
    AdvancedSettings=(RotationOrderForFilter=XZY)
    Child=(Type=Bone,Name="base_l_padChain1")
    Parents="((Item=(Type=Bone,Name=\"deform_upperarm_up_l\"),Weight=0.0061),(Item=(Type=Bone,Name=\"spine_04\"),Weight=0.0661),(Item=(Type=Bone,Name=\"spine_05\"),Weight=0.1118),(Item=(Type=Bone,Name=\"upperarm_twist_01_l\"),Weight=0.132),(Item=(Type=Bone,Name=\"clavicle_l\"),W ...(truncated)"
    Weight=1
    ExecutePin=@RigVMFunction_Sequence.C
   )
   > &RigUnit_ParentConstraint(
    AdvancedSettings=(RotationOrderForFilter=XZY)
    Child=(Type=Bone,Name="base_l_padChain1A")
    Parents="((Item=(Type=Bone,Name=\"deform_upperarm_up_l\"),Weight=0.0065),(Item=(Type=Bone,Name=\"spine_04\"),Weight=0.0652),(Item=(Type=Bone,Name=\"spine_05\"),Weight=0.1102),(Item=(Type=Bone,Name=\"upperarm_twist_01_l\"),Weight=0.1327),(Item=(Type=Bone,Name=\"clavicle_l\"), ...(truncated)"
    Weight=1
   )
   > &RigUnit_ParentConstraint(
    AdvancedSettings=(RotationOrderForFilter=XZY)
    Child=(Type=Bone,Name="base_l_padChain2")
    Parents="((Item=(Type=Bone,Name=\"deform_upperarm_up_l\"),Weight=0.0306),(Item=(Type=Bone,Name=\"spine_04\"),Weight=0.0413),(Item=(Type=Bone,Name=\"spine_05\"),Weight=0.0728),(Item=(Type=Bone,Name=\"upperarm_twist_01_l\"),Weight=0.2549),(Item=(Type=Bone,Name=\"clavicle_l\"), ...(truncated)"
    Weight=1
   )
   > &RigUnit_ParentConstraint(
    AdvancedSettings=(RotationOrderForFilter=XZY)
    Child=(Type=Bone,Name="base_l_padChain2A")
    Parents="((Item=(Type=Bone,Name=\"spine_04\"),Weight=0.0239),(Item=(Type=Bone,Name=\"spine_05\"),Weight=0.0476),(Item=(Type=Bone,Name=\"deform_upperarm_up_l\"),Weight=0.0545),(Item=(Type=Bone,Name=\"upperarm_twist_01_l\"),Weight=0.3726),(Item=(Type=Bone,Name=\"clavicle_l\"), ...(truncated)"
    Weight=1
   )
   > &RigUnit_ParentConstraint(
    AdvancedSettings=(RotationOrderForFilter=XZY)
    Child=(Type=Bone,Name="base_l_padChain3")
    Parents="((Item=(Type=Bone,Name=\"spine_04\"),Weight=0.0239),(Item=(Type=Bone,Name=\"spine_05\"),Weight=0.0476),(Item=(Type=Bone,Name=\"deform_upperarm_up_l\"),Weight=0.0545),(Item=(Type=Bone,Name=\"upperarm_twist_01_l\"),Weight=0.3726),(Item=(Type=Bone,Name=\"clavicle_l\"), ...(truncated)"
    Weight=1
   )
   > &RigUnit_ParentConstraint(
    AdvancedSettings=(RotationOrderForFilter=XZY)
    Child=(Type=Bone,Name="base_l_padChain3A")
    Parents="((Item=(Type=Bone,Name=\"spine_04\"),Weight=0.0131),(Item=(Type=Bone,Name=\"spine_05\"),Weight=0.0305),(Item=(Type=Bone,Name=\"deform_upperarm_up_l\"),Weight=0.0635),(Item=(Type=Bone,Name=\"clavicle_l\"),Weight=0.4193),(Item=(Type=Bone,Name=\"upperarm_twist_01_l\"), ...(truncated)"
    Weight=1
   )
   > &RigUnit_ParentConstraint(
    AdvancedSettings=(RotationOrderForFilter=XZY)
    Child=(Type=Bone,Name="base_l_padChain4")
    Parents="((Item=(Type=Bone,Name=\"spine_04\"),Weight=0.0097),(Item=(Type=Bone,Name=\"spine_05\"),Weight=0.024),(Item=(Type=Bone,Name=\"deform_upperarm_up_l\"),Weight=0.0677),(Item=(Type=Bone,Name=\"clavicle_l\"),Weight=0.4102),(Item=(Type=Bone,Name=\"upperarm_twist_01_l\"),W ...(truncated)"
    Weight=1
   )
   > &RigUnit_ParentConstraint(
    AdvancedSettings=(RotationOrderForFilter=XZY)
    Child=(Type=Bone,Name="base_l_padChain5")
    Parents="((Item=(Type=Bone,Name=\"spine_04\"),Weight=0.0053),(Item=(Type=Bone,Name=\"spine_05\"),Weight=0.0151),(Item=(Type=Bone,Name=\"deform_upperarm_up_l\"),Weight=0.0748),(Item=(Type=Bone,Name=\"clavicle_l\"),Weight=0.3921),(Item=(Type=Bone,Name=\"upperarm_twist_01_l\"), ...(truncated)"
    Weight=1
   )
   > &RigUnit_ParentConstraint(
    AdvancedSettings=(RotationOrderForFilter=XZY)
    Child=(Type=Bone,Name="base_l_padChain6")
    Parents="((Item=(Type=Bone,Name=\"spine_04\"),Weight=0.0053),(Item=(Type=Bone,Name=\"spine_05\"),Weight=0.0151),(Item=(Type=Bone,Name=\"deform_upperarm_up_l\"),Weight=0.0748),(Item=(Type=Bone,Name=\"clavicle_l\"),Weight=0.3921),(Item=(Type=Bone,Name=\"upperarm_twist_01_l\"), ...(truncated)"
    Weight=1
   )
   > &RigUnit_ParentConstraint(
    AdvancedSettings=(RotationOrderForFilter=XZY)
    Child=(Type=Bone,Name="base_l_padChain7")
    Parents="((Item=(Type=Bone,Name=\"spine_04\"),Weight=0.0039),(Item=(Type=Bone,Name=\"spine_05\"),Weight=0.0123),(Item=(Type=Bone,Name=\"deform_upperarm_up_l\"),Weight=0.0759),(Item=(Type=Bone,Name=\"clavicle_l\"),Weight=0.3732),(Item=(Type=Bone,Name=\"upperarm_twist_01_l\"), ...(truncated)"
    Weight=1
   )
   > &RigUnit_ParentConstraint(
    AdvancedSettings=(RotationOrderForFilter=XZY)
    Child=(Type=Bone,Name="base_l_padChain8")
    Parents="((Item=(Type=Bone,Name=\"spine_04\"),Weight=0.0019),(Item=(Type=Bone,Name=\"spine_05\"),Weight=0.0079),(Item=(Type=Bone,Name=\"deform_upperarm_up_l\"),Weight=0.0805),(Item=(Type=Bone,Name=\"clavicle_l\"),Weight=0.3278),(Item=(Type=Bone,Name=\"upperarm_twist_01_l\"), ...(truncated)"
    Weight=1
   )
   > &RigUnit_ParentConstraint(
    AdvancedSettings=(RotationOrderForFilter=XZY)
    Child=(Type=Bone,Name="base_l_padChain9")
    Parents="((Item=(Type=Bone,Name=\"spine_04\"),Weight=0.0019),(Item=(Type=Bone,Name=\"spine_05\"),Weight=0.0077),(Item=(Type=Bone,Name=\"deform_upperarm_up_l\"),Weight=0.0785),(Item=(Type=Bone,Name=\"clavicle_l\"),Weight=0.3556),(Item=(Type=Bone,Name=\"upperarm_twist_01_l\"), ...(truncated)"
    Weight=1
   )
   > &RigUnit_ParentConstraint(
    AdvancedSettings=(RotationOrderForFilter=XZY)
    Child=(Type=Bone,Name="base_l_padChain10")
    Parents="((Item=(Type=Bone,Name=\"spine_04\"),Weight=0.0015),(Item=(Type=Bone,Name=\"spine_05\"),Weight=0.0068),(Item=(Type=Bone,Name=\"deform_upperarm_up_l\"),Weight=0.0777),(Item=(Type=Bone,Name=\"clavicle_l\"),Weight=0.3473),(Item=(Type=Bone,Name=\"upperarm_twist_01_l\"), ...(truncated)"
    Weight=1
   )
   > &RigUnit_ParentConstraint(
    AdvancedSettings=(RotationOrderForFilter=XZY)
    Child=(Type=Bone,Name="base_l_padChain11")
    Parents="((Item=(Type=Bone,Name=\"spine_04\"),Weight=0.001),(Item=(Type=Bone,Name=\"spine_05\"),Weight=0.0054),(Item=(Type=Bone,Name=\"deform_upperarm_up_l\"),Weight=0.0765),(Item=(Type=Bone,Name=\"clavicle_l\"),Weight=0.3456),(Item=(Type=Bone,Name=\"upperarm_twist_01_l\"),W ...(truncated)"
    Weight=1
   )
   > &RigUnit_ParentConstraint(
    AdvancedSettings=(RotationOrderForFilter=XZY)
    Child=(Type=Bone,Name="base_l_padChain12")
    Parents="((Item=(Type=Bone,Name=\"spine_04\"),Weight=0.0005),(Item=(Type=Bone,Name=\"spine_05\"),Weight=0.0044),(Item=(Type=Bone,Name=\"deform_upperarm_up_l\"),Weight=0.0745),(Item=(Type=Bone,Name=\"clavicle_l\"),Weight=0.3471),(Item=(Type=Bone,Name=\"upperarm_twist_01_l\"), ...(truncated)"
    Weight=1
   )
   > &RigUnit_ParentConstraint(
    AdvancedSettings=(RotationOrderForFilter=XZY)
    Child=(Type=Bone,Name="base_l_padChain13")
    Parents="((Item=(Type=Bone,Name=\"spine_04\"),Weight=0.0005),(Item=(Type=Bone,Name=\"spine_05\"),Weight=0.0051),(Item=(Type=Bone,Name=\"deform_upperarm_up_l\"),Weight=0.0783),(Item=(Type=Bone,Name=\"clavicle_l\"),Weight=0.3765),(Item=(Type=Bone,Name=\"upperarm_twist_01_l\"), ...(truncated)"
    Weight=1
   )
   > &RigUnit_ParentConstraint(
    AdvancedSettings=(RotationOrderForFilter=XZY)
    Child=(Type=Bone,Name="base_l_padChain14")
    Parents="((Item=(Type=Bone,Name=\"spine_04\"),Weight=0.0005),(Item=(Type=Bone,Name=\"spine_05\"),Weight=0.0051),(Item=(Type=Bone,Name=\"deform_upperarm_up_l\"),Weight=0.0783),(Item=(Type=Bone,Name=\"clavicle_l\"),Weight=0.3765),(Item=(Type=Bone,Name=\"upperarm_twist_01_l\"), ...(truncated)"
    Weight=1
   )
   > &RigUnit_ParentConstraint(
    AdvancedSettings=(RotationOrderForFilter=XZY)
    Child=(Type=Bone,Name="base_l_padChain15")
    Parents="((Item=(Type=Bone,Name=\"spine_04\"),Weight=0.001),(Item=(Type=Bone,Name=\"spine_05\"),Weight=0.0106),(Item=(Type=Bone,Name=\"deform_upperarm_up_l\"),Weight=0.0771),(Item=(Type=Bone,Name=\"upperarm_twist_01_l\"),Weight=0.4281),(Item=(Type=Bone,Name=\"clavicle_l\"),W ...(truncated)"
    Weight=1
   )
   > &RigUnit_ParentConstraint(
    AdvancedSettings=(RotationOrderForFilter=XZY)
    Child=(Type=Bone,Name="base_l_padChain16")
    Parents="((Item=(Type=Bone,Name=\"spine_04\"),Weight=0.0029),(Item=(Type=Bone,Name=\"spine_05\"),Weight=0.0212),(Item=(Type=Bone,Name=\"deform_upperarm_up_l\"),Weight=0.0494),(Item=(Type=Bone,Name=\"upperarm_twist_01_l\"),Weight=0.3054),(Item=(Type=Bone,Name=\"clavicle_l\"), ...(truncated)"
    Weight=1
   )
   > &RigUnit_ParentConstraint(
    AdvancedSettings=(RotationOrderForFilter=XZY)
    Child=(Type=Bone,Name="base_l_padChain17")
    Parents="((Item=(Type=Bone,Name=\"spine_04\"),Weight=0.0055),(Item=(Type=Bone,Name=\"spine_05\"),Weight=0.0324),(Item=(Type=Bone,Name=\"deform_upperarm_up_l\"),Weight=0.0335),(Item=(Type=Bone,Name=\"upperarm_twist_01_l\"),Weight=0.1931),(Item=(Type=Bone,Name=\"clavicle_l\"), ...(truncated)"
    Weight=1
   )
   > &RigUnit_ParentConstraint(
    AdvancedSettings=(RotationOrderForFilter=XZY)
    Child=(Type=Bone,Name="base_l_padChain18")
    Parents="((Item=(Type=Bone,Name=\"deform_upperarm_up_l\"),Weight=0.0036),(Item=(Type=Bone,Name=\"spine_04\"),Weight=0.0044),(Item=(Type=Bone,Name=\"upperarm_twist_01_l\"),Weight=0.0396),(Item=(Type=Bone,Name=\"spine_05\"),Weight=0.0514),(Item=(Type=Bone,Name=\"clavicle_l\"), ...(truncated)"
    Weight=1
   )
   &RigUnit_ParentConstraint(
    AdvancedSettings=(RotationOrderForFilter=XZY)
    Child=(Type=Bone,Name="base_r_padChain1")
    Parents="((Item=(Type=Bone,Name=\"base_padChain_r_Helper1\"),Weight=0.0004),(Item=(Type=Bone,Name=\"spine_04\"),Weight=0.0439),(Item=(Type=Bone,Name=\"spine_05\"),Weight=0.0749),(Item=(Type=Bone,Name=\"upperarm_twist_01_r\"),Weight=0.086),(Item=(Type=Bone,Name=\"clavicle_r\" ...(truncated)"
    Weight=1
    ExecutePin=@RigVMFunction_Sequence.D
   )
   > &RigUnit_ParentConstraint(
    AdvancedSettings=(RotationOrderForFilter=XZY)
    Child=(Type=Bone,Name="base_r_padChain1A")
    Parents="((Item=(Type=Bone,Name=\"base_padChain_r_Helper1\"),Weight=0.0004),(Item=(Type=Bone,Name=\"spine_04\"),Weight=0.0427),(Item=(Type=Bone,Name=\"spine_05\"),Weight=0.0731),(Item=(Type=Bone,Name=\"upperarm_twist_01_r\"),Weight=0.086),(Item=(Type=Bone,Name=\"clavicle_r\" ...(truncated)"
    Weight=1
   )
   > &RigUnit_ParentConstraint(
    AdvancedSettings=(RotationOrderForFilter=XZY)
    Child=(Type=Bone,Name="base_r_padChain2")
    Parents="((Item=(Type=Bone,Name=\"base_padChain_r_Helper1\"),Weight=0.0064),(Item=(Type=Bone,Name=\"spine_04\"),Weight=0.0272),(Item=(Type=Bone,Name=\"spine_05\"),Weight=0.047),(Item=(Type=Bone,Name=\"upperarm_twist_01_r\"),Weight=0.154),(Item=(Type=Bone,Name=\"clavicle_r\") ...(truncated)"
    Weight=1
   )
   > &RigUnit_ParentConstraint(
    AdvancedSettings=(RotationOrderForFilter=XZY)
    Child=(Type=Bone,Name="base_r_padChain2A")
    Parents=((Item=(Type=Bone,Name="spine_05"),Weight=0.0272),(Item=(Type=Bone,Name="upperarm_twist_01_r"),Weight=0.256),(Item=(Type=Bone,Name="clavicle_r"),Weight=0.2902),(Item=(Type=Bone,Name="deform_upperarm_up_r"),Weight=0.4267))
    Weight=1
   )
   > &RigUnit_ParentConstraint(
    AdvancedSettings=(RotationOrderForFilter=XZY)
    Child=(Type=Bone,Name="base_r_padChain3")
    Parents=((Item=(Type=Bone,Name="spine_05"),Weight=0.0272),(Item=(Type=Bone,Name="upperarm_twist_01_r"),Weight=0.256),(Item=(Type=Bone,Name="clavicle_r"),Weight=0.2902),(Item=(Type=Bone,Name="deform_upperarm_up_r"),Weight=0.4267))
    Weight=1
   )
   > &RigUnit_ParentConstraint(
    AdvancedSettings=(RotationOrderForFilter=XZY)
    Child=(Type=Bone,Name="base_r_padChain3A")
    Parents="((Item=(Type=Bone,Name=\"base_padChain_r_Helper1\"),Weight=0.0368),(Item=(Type=Bone,Name=\"upperarm_r\"),Weight=0.0382),(Item=(Type=Bone,Name=\"clavicle_r\"),Weight=0.2619),(Item=(Type=Bone,Name=\"upperarm_twist_01_r\"),Weight=0.3199),(Item=(Type=Bone,Name=\"deform ...(truncated)"
    Weight=1
   )
   > &RigUnit_ParentConstraint(
    AdvancedSettings=(RotationOrderForFilter=XZY)
    Child=(Type=Bone,Name="base_r_padChain4")
    Parents="((Item=(Type=Bone,Name=\"base_padChain_r_Helper1\"),Weight=0.0428),(Item=(Type=Bone,Name=\"upperarm_r\"),Weight=0.0561),(Item=(Type=Bone,Name=\"clavicle_r\"),Weight=0.245),(Item=(Type=Bone,Name=\"deform_upperarm_up_r\"),Weight=0.2995),(Item=(Type=Bone,Name=\"uppera ...(truncated)"
    Weight=1
   )
   > &RigUnit_ParentConstraint(
    AdvancedSettings=(RotationOrderForFilter=XZY)
    Child=(Type=Bone,Name="base_r_padChain5")
    Parents="((Item=(Type=Bone,Name=\"base_padChain_r_Helper1\"),Weight=0.0467),(Item=(Type=Bone,Name=\"upperarm_r\"),Weight=0.0764),(Item=(Type=Bone,Name=\"clavicle_r\"),Weight=0.2288),(Item=(Type=Bone,Name=\"deform_upperarm_up_r\"),Weight=0.2663),(Item=(Type=Bone,Name=\"upper ...(truncated)"
    Weight=1
   )
   > &RigUnit_ParentConstraint(
    AdvancedSettings=(RotationOrderForFilter=XZY)
    Child=(Type=Bone,Name="base_r_padChain6")
    Parents="((Item=(Type=Bone,Name=\"base_padChain_r_Helper1\"),Weight=0.0467),(Item=(Type=Bone,Name=\"upperarm_r\"),Weight=0.0764),(Item=(Type=Bone,Name=\"clavicle_r\"),Weight=0.2288),(Item=(Type=Bone,Name=\"deform_upperarm_up_r\"),Weight=0.2663),(Item=(Type=Bone,Name=\"upper ...(truncated)"
    Weight=1
   )
   > &RigUnit_ParentConstraint(
    AdvancedSettings=(RotationOrderForFilter=XZY)
    Child=(Type=Bone,Name="base_r_padChain7")
    Parents="((Item=(Type=Bone,Name=\"base_padChain_r_Helper1\"),Weight=0.0488),(Item=(Type=Bone,Name=\"upperarm_r\"),Weight=0.0865),(Item=(Type=Bone,Name=\"clavicle_r\"),Weight=0.2193),(Item=(Type=Bone,Name=\"deform_upperarm_up_r\"),Weight=0.2355),(Item=(Type=Bone,Name=\"upper ...(truncated)"
    Weight=1
   )
   > &RigUnit_ParentConstraint(
    AdvancedSettings=(RotationOrderForFilter=XZY)
    Child=(Type=Bone,Name="base_r_padChain8")
    Parents="((Item=(Type=Bone,Name=\"base_padChain_r_Helper1\"),Weight=0.0436),(Item=(Type=Bone,Name=\"upperarm_r\"),Weight=0.0968),(Item=(Type=Bone,Name=\"deform_upperarm_up_r\"),Weight=0.2027),(Item=(Type=Bone,Name=\"clavicle_r\"),Weight=0.2144),(Item=(Type=Bone,Name=\"upper ...(truncated)"
    Weight=1
   )
   > &RigUnit_ParentConstraint(
    AdvancedSettings=(RotationOrderForFilter=XZY)
    Child=(Type=Bone,Name="base_r_padChain9")
    Parents="((Item=(Type=Bone,Name=\"spine_04\"),Weight=0.0007),(Item=(Type=Bone,Name=\"spine_05\"),Weight=0.0034),(Item=(Type=Bone,Name=\"base_padChain_r_Helper1\"),Weight=0.0501),(Item=(Type=Bone,Name=\"upperarm_r\"),Weight=0.1015),(Item=(Type=Bone,Name=\"deform_upperarm_up_ ...(truncated)"
    Weight=1
   )
   > &RigUnit_ParentConstraint(
    AdvancedSettings=(RotationOrderForFilter=XZY)
    Child=(Type=Bone,Name="base_r_padChain10")
    Parents="((Item=(Type=Bone,Name=\"spine_04\"),Weight=0.0005),(Item=(Type=Bone,Name=\"spine_05\"),Weight=0.003),(Item=(Type=Bone,Name=\"base_padChain_r_Helper1\"),Weight=0.0433),(Item=(Type=Bone,Name=\"upperarm_r\"),Weight=0.0817),(Item=(Type=Bone,Name=\"deform_upperarm_up_r ...(truncated)"
    Weight=1
   )
   > &RigUnit_ParentConstraint(
    AdvancedSettings=(RotationOrderForFilter=XZY)
    Child=(Type=Bone,Name="base_r_padChain11")
    Parents="((Item=(Type=Bone,Name=\"spine_04\"),Weight=0.0004),(Item=(Type=Bone,Name=\"spine_05\"),Weight=0.0028),(Item=(Type=Bone,Name=\"base_padChain_r_Helper1\"),Weight=0.0359),(Item=(Type=Bone,Name=\"upperarm_r\"),Weight=0.0584),(Item=(Type=Bone,Name=\"deform_upperarm_up_ ...(truncated)"
    Weight=1
   )
   > &RigUnit_ParentConstraint(
    AdvancedSettings=(RotationOrderForFilter=XZY)
    Child=(Type=Bone,Name="base_r_padChain12")
    Parents="((Item=(Type=Bone,Name=\"spine_04\"),Weight=0.0003),(Item=(Type=Bone,Name=\"spine_05\"),Weight=0.003),(Item=(Type=Bone,Name=\"upperarm_r\"),Weight=0.0255),(Item=(Type=Bone,Name=\"base_padChain_r_Helper1\"),Weight=0.0275),(Item=(Type=Bone,Name=\"deform_upperarm_up_r ...(truncated)"
    Weight=1
   )
   > &RigUnit_ParentConstraint(
    AdvancedSettings=(RotationOrderForFilter=XZY)
    Child=(Type=Bone,Name="base_r_padChain13")
    Parents="((Item=(Type=Bone,Name=\"spine_04\"),Weight=0.0005),(Item=(Type=Bone,Name=\"upperarm_r\"),Weight=0.0038),(Item=(Type=Bone,Name=\"spine_05\"),Weight=0.0068),(Item=(Type=Bone,Name=\"base_padChain_r_Helper1\"),Weight=0.0162),(Item=(Type=Bone,Name=\"deform_upperarm_up_ ...(truncated)"
    Weight=1
   )
   > &RigUnit_ParentConstraint(
    AdvancedSettings=(RotationOrderForFilter=XZY)
    Child=(Type=Bone,Name="base_r_padChain14")
    Parents="((Item=(Type=Bone,Name=\"spine_04\"),Weight=0.0005),(Item=(Type=Bone,Name=\"upperarm_r\"),Weight=0.0038),(Item=(Type=Bone,Name=\"spine_05\"),Weight=0.0068),(Item=(Type=Bone,Name=\"base_padChain_r_Helper1\"),Weight=0.0162),(Item=(Type=Bone,Name=\"deform_upperarm_up_ ...(truncated)"
    Weight=1
   )
   > &RigUnit_ParentConstraint(
    AdvancedSettings=(RotationOrderForFilter=XZY)
    Child=(Type=Bone,Name="base_r_padChain15")
    Parents="((Item=(Type=Bone,Name=\"upperarm_r\"),Weight=0.0002),(Item=(Type=Bone,Name=\"spine_04\"),Weight=0.0013),(Item=(Type=Bone,Name=\"base_padChain_r_Helper1\"),Weight=0.0019),(Item=(Type=Bone,Name=\"spine_05\"),Weight=0.0137),(Item=(Type=Bone,Name=\"deform_upperarm_up_ ...(truncated)"
    Weight=1
   )
   > &RigUnit_ParentConstraint(
    AdvancedSettings=(RotationOrderForFilter=XZY)
    Child=(Type=Bone,Name="base_r_padChain16")
    Parents="((Item=(Type=Bone,Name=\"spine_04\"),Weight=0.0027),(Item=(Type=Bone,Name=\"spine_05\"),Weight=0.0201),(Item=(Type=Bone,Name=\"deform_upperarm_up_r\"),Weight=0.0974),(Item=(Type=Bone,Name=\"upperarm_twist_01_r\"),Weight=0.29),(Item=(Type=Bone,Name=\"clavicle_r\"),We ...(truncated)"
    Weight=1
   )
   > &RigUnit_ParentConstraint(
    AdvancedSettings=(RotationOrderForFilter=XZY)
    Child=(Type=Bone,Name="base_r_padChain17")
    Parents="((Item=(Type=Bone,Name=\"spine_04\"),Weight=0.0052),(Item=(Type=Bone,Name=\"spine_05\"),Weight=0.0308),(Item=(Type=Bone,Name=\"deform_upperarm_up_r\"),Weight=0.0815),(Item=(Type=Bone,Name=\"upperarm_twist_01_r\"),Weight=0.1835),(Item=(Type=Bone,Name=\"clavicle_r\"), ...(truncated)"
    Weight=1
   )
   > &RigUnit_ParentConstraint(
    AdvancedSettings=(RotationOrderForFilter=XZY)
    Child=(Type=Bone,Name="base_r_padChain18")
    Parents="((Item=(Type=Bone,Name=\"spine_04\"),Weight=0.0042),(Item=(Type=Bone,Name=\"upperarm_twist_01_r\"),Weight=0.0377),(Item=(Type=Bone,Name=\"spine_05\"),Weight=0.0489),(Item=(Type=Bone,Name=\"deform_upperarm_up_r\"),Weight=0.0516),(Item=(Type=Bone,Name=\"clavicle_r\"), ...(truncated)"
    Weight=1
   )
  ]}
 New Function_ContainedGraph {nodes [
   RigVMFunctionEntryNode()
   n1 = &RigVMFunction_AlphaInterp(
    InterpSpeedDecreasing=15
    InterpSpeedIncreasing=15
    Scale=1
    bClampResult=False
    Value=&RigVMFunction_MathBoolToFloat(Value=RigVMVariableNode() {variable=pub is Targeting}.Value).Result
   )
   > RigVMVariableNode(Value=@n1.Result) {variable=isTargeting lerp}
   RigVMFunction_MathBoolOr = RigVMAggregateNode(
    C=RigVMVariableNode() {variable=SMG_P90}.Value
    B=RigVMVariableNode() {variable=is OutLaw_SG}.Value
    A=RigVMVariableNode() {variable=isAR Equipped}.Value
   )
   n2 = &RigVMFunction_AlphaInterp(
    InterpSpeedDecreasing=15
    InterpSpeedIncreasing=15
    Scale=1
    bClampResult=False
    Value=&RigVMFunction_MathBoolToFloat(Value=@RigVMFunction_MathBoolOr.Result).Result
   )
   > RigVMVariableNode(Value=@n2.Result) {variable=isAR Equipped lerp}
   n3 = &RigVMFunction_AlphaInterp(
    InterpSpeedDecreasing=15
    InterpSpeedIncreasing=15
    Scale=1
    bClampResult=False
    Value=&RigVMFunction_MathBoolToFloat(Value=RigVMVariableNode() {variable="pub isPistol Equipped "}.Value).Result
   )
   > RigVMVariableNode(Value=@n3.Result) {variable=isPistol Equipped lerp}
   n4 = &RigVMFunction_AlphaInterp(
    InterpSpeedDecreasing=15
    InterpSpeedIncreasing=15
    Scale=1
    bClampResult=False
    Value=&RigVMFunction_MathBoolToFloat(Value=RigVMVariableNode() {variable=pub PA Swing}.Value).Result
   )
   > RigVMVariableNode(Value=@n4.Result) {variable=isPA Swing lerp}
   > RigVMFunctionReturnNode()
  ]}
 New Function_ContainedGraph {nodes [
   RigVMFunctionEntryNode()
   > RigVMFunction_Sequence = RigVMAggregateNode()
   RigVMFunctionReturnNode(ExecuteContext=@RigVMFunction_Sequence.E)
   &RigUnit_ModifyTransforms(
    ItemToModify="((Item=(Type=Bone,Name=\"base_r_padChain7\"),Transform=(Rotation=(X=0,Y=-0.0872,Z=0,W=0.9962),Translation=(X=0,Y=0,Z=0),Scale3D=(X=1,Y=1,Z=1))),(Item=(Type=Bone,Name=\"base_r_padChain8\"),Transform=(Rotation=(X=0,Y=-0.0872,Z=0,W=0.9962),Translation=(X=0,Y=0,Z= ...(truncated)"
    WeightMinimum=0
    Weight=RigVMDispatchNode(BoneName=root, Default=0, Name=deformRig_r_upperarm_up_alpha) {template="DISPATCH_RigDispatch_GetAnimAttribute(in Name,in BoneName,in Default,out Value,out Found)"}.Value
    ExecutePin=@RigVMFunction_Sequence.B
   )
   > &RigUnit_ModifyTransforms(
    ItemToModify="((Item=(Type=Bone,Name=\"base_l_padChain5\"),Transform=(Rotation=(X=0,Y=-0.0872,Z=0,W=0.9962),Translation=(X=0,Y=0,Z=0),Scale3D=(X=1,Y=1,Z=1))),(Item=(Type=Bone,Name=\"base_l_padChain6\"),Transform=(Rotation=(X=0,Y=-0.1736,Z=0,W=0.9848),Translation=(X=0,Y=0,Z= ...(truncated)"
    WeightMinimum=0
    Weight=RigVMDispatchNode(BoneName=root, Default=0, Name=deformRig_l_upperarm_up_alpha) {template="DISPATCH_RigDispatch_GetAnimAttribute(in Name,in BoneName,in Default,out Value,out Found)"}.Value
   )
   > n1 = &RigUnit_SphericalPoseReader(
    ActiveRegionScaleFactors=(PositiveWidth=2,NegativeWidth=1.6,PositiveHeight=1,NegativeHeight=5)
    ActiveRegionSize=0.03
    Debug=(bDrawDebug=false,bDraw2D=False,bDrawLocalAxes=False,DebugScale=25,DebugSegments=20,DebugThickness=0.25)
    DriverAxis=(X=1,Y=0,Z=0)
    DriverItem=(Type=Bone,Name="upperarm_r")
    FalloffRegionScaleFactors=(PositiveWidth=0.5,NegativeWidth=0.7,PositiveHeight=1.6,NegativeHeight=0.5)
    FalloffSize=0.25
    FlipHeightScaling=False
    FlipWidthScaling=False
    OptionalParentItem=(Type=Bone,Name="None")
    RotationOffset=(X=75,Y=25,Z=10)
   )
   Multiply = &RigVMFunction_MathDoubleMul(A=@n1.OutputParam, B=RigVMVariableNode() {variable=isTargeting lerp}.Value)
   n2 = &RigVMFunction_AnimEvalRichCurve(
    Curve=(EditorCurveData=(Keys=((),(Time=0.6098,Value=0.529),(Time=1,Value=1))))
    SourceMinimum=0
    TargetMinimum=0
    Value=&RigVMFunction_MathDoubleMul(A=@Multiply.Result, B=RigVMVariableNode() {variable=isPistol Equipped lerp}.Value).Result
   )
   n3 = &RigVMFunction_AlphaInterp(InterpSpeedDecreasing=5.5, InterpSpeedIncreasing=5, Scale=1, bClampResult=False, Value=@n2.Result)
   > &RigUnit_ModifyTransforms(
    ItemToModify="((Item=(Type=Bone,Name=\"base_r_padChain1\"),Transform=(Rotation=(X=0.4111,Y=-0.4583,Z=-0.0182,W=0.7878),Translation=(X=0,Y=0,Z=0),Scale3D=(X=1,Y=1,Z=1))),(Item=(Type=Bone,Name=\"base_r_padChain1A\"),Transform=(Rotation=(X=0.3701,Y=-0.5031,Z=-0.0261,W=0.7805), ...(truncated)"
    WeightMinimum=0
    Weight=@n3.Result
   )
   n4 = RigVMDispatchNode(BoneName=root, Default=0, Name=deformRig_l_upperarm_fr_alpha) {template="DISPATCH_RigDispatch_GetAnimAttribute(in Name,in BoneName,in Default,out Value,out Found)"}
   n5 = RigVMVariableNode() {variable=isTargeting lerp}
   > &RigUnit_ModifyTransforms(
    ItemToModify="((Item=(Type=Bone,Name=\"base_l_padChain1\"),Transform=(Rotation=(X=0.0227,Y=-0.1722,Z=0.1285,W=0.9764),Translation=(X=0,Y=0,Z=0),Scale3D=(X=1,Y=1,Z=1))),(Item=(Type=Bone,Name=\"base_l_padChain1A\"),Transform=(Rotation=(X=-0.2672,Y=-0.4271,Z=0.1555,W=0.8497),T ...(truncated)"
    WeightMinimum=0
    Weight=&RigVMFunction_MathDoubleMul(A=&RigVMFunction_MathDoubleMul(A=@n4.Value, B=@n5.Value).Result, B=RigVMVariableNode() {variable=isPistol Equipped lerp}.Value).Result
   )
   RigVMFunction_AnimEvalRichCurve = &RigVMFunction_AnimEvalRichCurve(
    Curve=(EditorCurveData=(Keys=((),(Time=0.456,Value=0.1936),(Time=0.9865,Value=0.4867))))
    SourceMinimum=0
    TargetMinimum=0
    Value=&RigVMFunction_MathDoubleMul(
     A=RigVMDispatchNode(BoneName=root, Default=0, Name=deformRig_r_upperarm_fr_alpha) {template="DISPATCH_RigDispatch_GetAnimAttribute(in Name,in BoneName,in Default,out Value,out Found)"}.Value
     B=RigVMVariableNode() {variable=isAR Equipped lerp}.Value
    ).Result
   )
   > &RigUnit_ModifyTransforms(
    ItemToModify="((Item=(Type=Bone,Name=\"base_r_padChain1\"),Transform=(Rotation=(X=0.0614,Y=-0.59,Z=-0.0614,W=0.8027),Translation=(X=0,Y=0,Z=0),Scale3D=(X=1,Y=1,Z=1))),(Item=(Type=Bone,Name=\"base_r_padChain1A\"),Transform=(Rotation=(X=0.0871,Y=-0.5868,Z=-0.0263,W=0.8046),Tr ...(truncated)"
    WeightMinimum=0
    Weight=@RigVMFunction_AnimEvalRichCurve.Result
   )
   Remap = &RigVMFunction_MathDoubleRemap(
    SourceMaximum=0
    SourceMinimum=1
    TargetMaximum=1
    TargetMinimum=0
    bClamp=False
    Value=RigVMVariableNode() {variable=isTargeting lerp}.Value
   )
   Multiply_7 = RigVMAggregateNode(
    A=RigVMDispatchNode(BoneName=root, Default=0, Name=deformRig_l_upperarm_fr_alpha) {template="DISPATCH_RigDispatch_GetAnimAttribute(in Name,in BoneName,in Default,out Value,out Found)"}.Value
    C=@Remap.Result
    B=RigVMVariableNode() {variable=isPistol Equipped lerp}.Value
   )
   > &RigUnit_ModifyTransforms(
    ItemToModify="((Item=(Type=Bone,Name=\"base_l_padChain1\"),Transform=(Rotation=(X=0.0302,Y=-0.171,Z=0.171,W=0.9698),Translation=(X=0,Y=0,Z=0),Scale3D=(X=1,Y=1,Z=1))),(Item=(Type=Bone,Name=\"base_l_padChain1A\"),Transform=(Rotation=(X=0.0734,Y=-0.4162,Z=0.1574,W=0.8925),Tran ...(truncated)"
    WeightMinimum=0
    Weight=@Multiply_7.Result
   )
   Multiply_8 = RigVMAggregateNode(
    A=RigVMDispatchNode(BoneName=root, Default=0, Name=deformRig_l_upperarm_fr_alpha) {template="DISPATCH_RigDispatch_GetAnimAttribute(in Name,in BoneName,in Default,out Value,out Found)"}.Value
    C=&RigVMFunction_MathDoubleRemap(
     SourceMaximum=0
     SourceMinimum=1
     TargetMaximum=1
     TargetMinimum=0
     bClamp=False
     Value=RigVMVariableNode() {variable=isTargeting lerp}.Value
    ).Result
    B=RigVMVariableNode() {variable=isAR Equipped lerp}.Value
   )
   > &RigUnit_ModifyTransforms(
    ItemToModify="((Item=(Type=Bone,Name=\"base_l_padChain1\"),Transform=(Rotation=(X=0.0302,Y=-0.171,Z=0.171,W=0.9698),Translation=(X=0,Y=0,Z=0),Scale3D=(X=1,Y=1,Z=1))),(Item=(Type=Bone,Name=\"base_l_padChain1A\"),Transform=(Rotation=(X=0.0734,Y=-0.4162,Z=0.1574,W=0.8925),Tran ...(truncated)"
    WeightMinimum=0
    Weight=@Multiply_8.Result
   )
   Multiply_4 = RigVMAggregateNode(
    A=RigVMDispatchNode(BoneName=root, Default=0, Name=deformRig_l_upperarm_fr_alpha) {template="DISPATCH_RigDispatch_GetAnimAttribute(in Name,in BoneName,in Default,out Value,out Found)"}.Value
    C=RigVMVariableNode() {variable=isTargeting lerp}.Value
    B=RigVMVariableNode() {variable=isAR Equipped lerp}.Value
   )
   > &RigUnit_ModifyTransforms(
    ItemToModify="((Item=(Type=Bone,Name=\"base_l_padChain1\"),Transform=(Rotation=(X=0.0302,Y=-0.171,Z=0.171,W=0.9698),Translation=(X=0,Y=0,Z=0),Scale3D=(X=1,Y=1,Z=1))),(Item=(Type=Bone,Name=\"base_l_padChain1A\"),Transform=(Rotation=(X=0.0734,Y=-0.4162,Z=0.1574,W=0.8925),Tran ...(truncated)"
    WeightMinimum=0
    Weight=@Multiply_4.Result
   )
   > n6 = &RigUnit_SphericalPoseReader(
    ActiveRegionScaleFactors=(PositiveWidth=3,NegativeWidth=1,PositiveHeight=1.5,NegativeHeight=1.5)
    ActiveRegionSize=0.1
    Debug=(bDrawDebug=false,DebugScale=25,DebugSegments=20,DebugThickness=0.25)
    DriverAxis=(X=1,Y=0,Z=0)
    DriverItem=(Type=Bone,Name="upperarm_l")
    FalloffRegionScaleFactors=(PositiveWidth=1,NegativeWidth=1,PositiveHeight=1,NegativeHeight=1)
    FalloffSize=0.4
    OptionalParentItem=(Type=Bone)
    RotationOffset=(X=100,Y=-20,Z=240)
   )
   > &RigUnit_ModifyTransforms(
    ItemToModify="((Item=(Type=Bone,Name=\"base_l_padChain18\"),Transform=(Rotation=(X=0.0019,Y=-0.0436,Z=0.0436,W=0.9981),Translation=(X=-0.3,Y=0,Z=0),Scale3D=(X=1,Y=1,Z=1))),(Item=(Type=Bone,Name=\"base_l_padChain17\"),Transform=(Rotation=(X=0.0046,Y=-0.0521,Z=0.087,W=0.9948) ...(truncated)"
    WeightMinimum=0
    Weight=@n6.OutputParam
   )
   > n7 = &RigUnit_SphericalPoseReader(
    ActiveRegionScaleFactors=(PositiveWidth=3,NegativeWidth=1,PositiveHeight=1.5,NegativeHeight=1.5)
    ActiveRegionSize=0.1
    Debug=(bDrawDebug=false,DebugScale=25,DebugSegments=20,DebugThickness=0.25)
    DriverAxis=(X=-1,Y=0,Z=0)
    DriverItem=(Type=Bone,Name="upperarm_r")
    FalloffRegionScaleFactors=(PositiveWidth=1,NegativeWidth=1,PositiveHeight=1,NegativeHeight=1)
    FalloffSize=0.4
    FlipHeightScaling=true
    FlipWidthScaling=false
    OptionalParentItem=(Type=Bone)
    RotationOffset=(X=80,Y=30,Z=80)
   )
   n8 = &RigVMFunction_MathBoolToFloat(Value=RigVMVariableNode() {variable=isFrontEnd}.Value)
   > &RigUnit_ModifyTransforms(
    ItemToModify="((Item=(Type=Bone,Name=\"base_r_padChain18\"),Transform=(Rotation=(X=0,Y=-0.0872,Z=0,W=0.9962),Translation=(X=-0.5,Y=0,Z=0),Scale3D=(X=1,Y=1,Z=1))),(Item=(Type=Bone,Name=\"base_r_padChain17\"),Transform=(Rotation=(X=-0.0434,Y=-0.1044,Z=-0.0046,W=0.9936),Transl ...(truncated)"
    WeightMinimum=0
    Weight=&RigVMFunction_MathDoubleMul(
     B=&RigVMFunction_MathDoubleRemap(SourceMaximum=0, SourceMinimum=1, TargetMaximum=1, TargetMinimum=0, bClamp=False, Value=@n8.Result).Result
     A=@n7.OutputParam
    ).Result
   )
   > n9 = &RigUnit_SphericalPoseReader(
    ActiveRegionScaleFactors=(PositiveWidth=3,NegativeWidth=1,PositiveHeight=1.5,NegativeHeight=1.5)
    ActiveRegionSize=0.125
    Debug=(bDrawDebug=false,DebugScale=25,DebugSegments=20,DebugThickness=0.25)
    DriverAxis=(X=-1,Y=0,Z=0)
    DriverItem=(Type=Bone,Name="upperarm_r")
    FalloffRegionScaleFactors=(PositiveWidth=1,NegativeWidth=1,PositiveHeight=1,NegativeHeight=1)
    FalloffSize=0.4
    FlipHeightScaling=true
    FlipWidthScaling=false
    OptionalParentItem=(Type=Bone)
    RotationOffset=(X=90,Y=30,Z=70)
   )
   n10 = RigVMVariableNode() {variable=isFrontEnd}
   > &RigUnit_ModifyTransforms(
    ItemToModify="((Item=(Type=Bone,Name=\"base_r_padChain18\"),Transform=(Rotation=(X=0,Y=-0.0872,Z=0,W=0.9962),Translation=(X=-0.5,Y=0,Z=0),Scale3D=(X=1,Y=1,Z=1))),(Item=(Type=Bone,Name=\"base_r_padChain17\"),Transform=(Rotation=(X=-0.0434,Y=-0.1044,Z=-0.0046,W=0.9936),Transl ...(truncated)"
    WeightMinimum=0
    Weight=&RigVMFunction_MathDoubleMul(B=&RigVMFunction_MathBoolToFloat(Value=@n10.Value).Result, A=@n9.OutputParam).Result
   )
   > &RigUnit_ModifyTransforms(
    ItemToModify="((Item=(Type=Bone,Name=\"base_r_padChain1\"),Transform=(Rotation=(X=0,Y=-0.1736,Z=0,W=0.9848),Translation=(X=0,Y=0,Z=0),Scale3D=(X=1,Y=1,Z=1))),(Item=(Type=Bone,Name=\"base_r_padChain1A\"),Transform=(Rotation=(X=0,Y=-0.2164,Z=0,W=0.9763),Translation=(X=0,Y=0,Z ...(truncated)"
    WeightMinimum=0
    Weight=&RigVMFunction_MathDoubleMul(
     A=RigVMDispatchNode(BoneName=root, Default=0, Name=deformRig_r_upperarm_fr_alpha) {template="DISPATCH_RigDispatch_GetAnimAttribute(in Name,in BoneName,in Default,out Value,out Found)"}.Value
     B=RigVMVariableNode() {variable=isPistol Equipped lerp}.Value
    ).Result
   )
   Multiply_12 = RigVMAggregateNode(
    C=4
    A=RigVMDispatchNode(BoneName=root, Default=0, Name=deformRig_l_upperarm_up_alpha) {template="DISPATCH_RigDispatch_GetAnimAttribute(in Name,in BoneName,in Default,out Value,out Found)"}.Value
    B=RigVMVariableNode() {variable=isAR Equipped lerp}.Value
   )
   n11 = &RigVMFunction_AnimEvalRichCurve(Curve=(EditorCurveData=(Keys=((),(Time=0.9865,Value=0.4867)))), SourceMinimum=0, TargetMinimum=0, Value=@Multiply_12.Result)
   > &RigUnit_ModifyTransforms(
    ItemToModify="((Item=(Type=Bone,Name=\"base_l_padChain3\"),Transform=(Rotation=(X=0,Y=-0.2334,Z=0,W=0.9724),Translation=(X=0,Y=0,Z=0),Scale3D=(X=1,Y=1,Z=1))),(Item=(Type=Bone,Name=\"base_l_padChain3A\"),Transform=(Rotation=(X=0,Y=-0.2756,Z=0,W=0.9613),Translation=(X=0,Y=0,Z ...(truncated)"
    WeightMinimum=0
    Weight=@n11.Result
   )
   Multiply_13 = RigVMAggregateNode(
    C=4
    A=RigVMDispatchNode(BoneName=root, Default=0, Name=deformRig_l_upperarm_up_alpha) {template="DISPATCH_RigDispatch_GetAnimAttribute(in Name,in BoneName,in Default,out Value,out Found)"}.Value
    B=RigVMVariableNode() {variable=isPistol Equipped lerp}.Value
   )
   > &RigUnit_ModifyTransforms(
    ItemToModify="((Item=(Type=Bone,Name=\"base_l_padChain2A\"),Transform=(Rotation=(X=0,Y=0,Z=0,W=1),Translation=(X=0,Y=0,Z=0),Scale3D=(X=1,Y=1,Z=1))),(Item=(Type=Bone,Name=\"base_l_padChain3\"),Transform=(Rotation=(X=0,Y=0,Z=0,W=1),Translation=(X=0,Y=0,Z=0),Scale3D=(X=1,Y=1,Z ...(truncated)"
    WeightMinimum=0
    Weight=@Multiply_13.Result
   )
   n12 = RigVMDispatchNode(BoneName=root, Default=0, Name=deformRig_l_upperarm_fr_alpha) {template="DISPATCH_RigDispatch_GetAnimAttribute(in Name,in BoneName,in Default,out Value,out Found)"}
   n13 = RigVMVariableNode() {variable=isPistol Equipped lerp}
   > &RigUnit_ModifyTransforms(
    ItemToModify=((Item=(Type=Bone,Name="None"),Transform=(Rotation=(X=0,Y=0,Z=0,W=1),Translation=(X=0,Y=0,Z=0),Scale3D=(X=1,Y=1,Z=1))))
    WeightMinimum=0
    Weight=&RigVMFunction_MathDoubleMul(A=&RigVMFunction_MathDoubleMul(A=@n12.Value, B=@n13.Value).Result, B=RigVMVariableNode() {variable=isTargeting lerp}.Value).Result
   )
   RigVMFunction_MathBoolToFloat = &RigVMFunction_MathBoolToFloat(Value=RigVMVariableNode() {variable=isPlaying emote}.Value)
   AlphaInterp = &RigVMFunction_AlphaInterp(InterpSpeedDecreasing=3, InterpSpeedIncreasing=3, Scale=1, bClampResult=False, Value=@RigVMFunction_MathBoolToFloat.Result)
   &RigUnit_ModifyTransforms(
    ItemToModify="((Item=(Type=Bone,Name=\"base_front_skirt_l\"),Transform=(Rotation=(X=0,Y=-0.0872,Z=0,W=0.9962),Translation=(X=0,Y=0,Z=0),Scale3D=(X=1,Y=1,Z=1))),(Item=(Type=Bone,Name=\"base_side_skirt_l\"),Transform=(Rotation=(X=0,Y=0,Z=0,W=1),Translation=(X=0,Y=0,Z=1),Scale ...(truncated)"
    WeightMinimum=0
    Weight=@AlphaInterp.Result
    ExecutePin=@RigVMFunction_Sequence.A
   )
   > &RigUnit_ParentConstraint(
    AdvancedSettings=(InterpolationType=Average,RotationOrderForFilter=XZY)
    Child=(Type=Bone,Name="base_front_skirt_c")
    Parents=((Item=(Type=Bone,Name="base_front_skirt_l"),Weight=0.5),(Item=(Type=Bone,Name="base_front_skirt_r"),Weight=0.5))
    Weight=1
   )
   > &RigUnit_ParentConstraint(
    AdvancedSettings=(InterpolationType=Average,RotationOrderForFilter=XZY)
    Child=(Type=Bone,Name="base_bk_skirt_c")
    Filter=(TranslationFilter=(bX=true,bY=true,bZ=true),RotationFilter=(bX=false,bY=false,bZ=false),ScaleFilter=(bX=True,bY=True,bZ=True))
    Parents=((Item=(Type=Bone,Name="base_bk_skirt_l"),Weight=0.5),(Item=(Type=Bone,Name="base_bk_skirt_r"),Weight=0.5))
    Weight=1
   )
   n14 = &RigVMFunction_AlphaInterp(
    InterpSpeedDecreasing=20
    InterpSpeedIncreasing=5
    Scale=1
    bClampResult=False
    Value=&RigVMFunction_MathDoubleMul(A=RigVMVariableNode() {variable=is_Parachuting}.Value, B=RigVMVariableNode() {variable=isMoving_Left}.Value).Result
   )
   > &RigUnit_ModifyTransforms(
    ItemToModify="((Item=(Type=Bone,Name=\"base_front_skirt_r\"),Transform=(Rotation=(X=-0.0068,Y=-0.1303,Z=-0.0519,W=0.9901),Translation=(X=0,Y=0,Z=0),Scale3D=(X=1,Y=1,Z=1))),(Item=(Type=Bone,Name=\"base_bk_skirt_l\"),Transform=(Rotation=(X=0,Y=0.1305,Z=0,W=0.9914),Translation ...(truncated)"
    WeightMinimum=0
    Weight=@n14.Result
   )
   n15 = &RigVMFunction_AlphaInterp(
    InterpSpeedDecreasing=20
    InterpSpeedIncreasing=5
    Scale=1
    bClampResult=False
    Value=&RigVMFunction_MathDoubleMul(B=RigVMVariableNode() {variable=isMoving_Right}.Value, A=RigVMVariableNode() {variable=is_Parachuting}.Value).Result
   )
   > &RigUnit_ModifyTransforms(
    ItemToModify="((Item=(Type=Bone,Name=\"base_front_skirt_l\"),Transform=(Rotation=(X=0.0011,Y=-0.0436,Z=0.0262,W=0.9987),Translation=(X=0,Y=0,Z=0),Scale3D=(X=1,Y=1,Z=1))),(Item=(Type=Bone,Name=\"base_bk_skirt_r\"),Transform=(Rotation=(X=0,Y=0.1736,Z=0,W=0.9848),Translation=( ...(truncated)"
    WeightMinimum=0
    Weight=@n15.Result
   )
   > &RigUnit_ModifyTransforms(
    ItemToModify="((Item=(Type=Bone,Name=\"base_bandA\"),Transform=(Rotation=(X=-0.0894,Y=0.5072,Z=0.1488,W=0.8441),Translation=(X=0,Y=0,Z=0),Scale3D=(X=1,Y=1,Z=1))),(Item=(Type=Bone,Name=\"base_bandB\"),Transform=(Rotation=(X=-0.2241,Y=0.483,Z=0.1294,W=0.8365),Translation=(X=0 ...(truncated)"
    WeightMinimum=0
    Weight=RigVMDispatchNode(BoneName=root, Default=0, Name=deformRig_r_thigh_up_alpha) {template="DISPATCH_RigDispatch_GetAnimAttribute(in Name,in BoneName,in Default,out Value,out Found)"}.Value
   )
   > &RigUnit_ModifyTransforms(
    ItemToModify="((Item=(Type=Bone,Name=\"base_bk_skirt_l\"),Transform=(Rotation=(X=0,Y=0,Z=0,W=1),Translation=(X=0,Y=0,Z=2),Scale3D=(X=1,Y=1,Z=1))),(Item=(Type=Bone,Name=\"base_bk_skirt_r\"),Transform=(Rotation=(X=-0.2028,Y=0.228,Z=-0.2028,W=0.9305),Translation=(X=0,Y=0,Z=3), ...(truncated)"
    WeightMinimum=0
    Weight=RigVMVariableNode() {variable=isBreakDancing}.Value
   )
   > &RigUnit_ModifyTransforms(
    ItemToModify="((Item=(Type=Bone,Name=\"base_front_skirt_l\"),Transform=(Rotation=(X=-0.2619,Y=0.008,Z=0.1372,W=0.9553),Translation=(X=-1.5,Y=2,Z=-1),Scale3D=(X=1,Y=1,Z=1))),(Item=(Type=Bone,Name=\"base_side_skirt_l\"),Transform=(Rotation=(X=0,Y=0.0872,Z=0,W=0.9962),Translat ...(truncated)"
    WeightMinimum=0
    Weight=RigVMVariableNode() {variable=isLobbyIddle}.Value
   )
   > &RigUnit_ModifyTransforms(
    ItemToModify="((Item=(Type=Bone,Name=\"Waistband_Collider_01\"),Transform=(Rotation=(X=0,Y=0,Z=0,W=1),Translation=(X=0.5,Y=-1,Z=-2),Scale3D=(X=1,Y=1,Z=1))),(Item=(Type=Bone,Name=\"Waistband_Collider_02\"),Transform=(Rotation=(X=0,Y=0,Z=0,W=1),Translation=(X=2,Y=2,Z=-3),Scal ...(truncated)"
    WeightMinimum=0
    Weight=RigVMVariableNode() {variable=isBalletSpin}.Value
   )
   &RigUnit_ModifyTransforms(
    ItemToModify="((Item=(Type=Bone,Name=\"Waistband_Collider_03\"),Transform=(Rotation=(X=0,Y=0,Z=0,W=1),Translation=(X=-2.5,Y=0,Z=0),Scale3D=(X=1,Y=1,Z=1))),(Item=(Type=Bone,Name=\"Waistband_Collider_02\"),Transform=(Rotation=(X=0,Y=0,Z=0,W=1),Translation=(X=-2,Y=0,Z=0),Scale ...(truncated)"
    Weight=1
    WeightMinimum=0
    ExecutePin=@RigVMFunction_Sequence.C
   )
   > &RigUnit_ModifyTransforms(
    ItemToModify="((Item=(Type=Bone,Name=\"base_side_skirt_l\"),Transform=(Rotation=(X=0,Y=-0.0436,Z=0,W=0.999),Translation=(X=0,Y=0,Z=0.5),Scale3D=(X=1,Y=1,Z=1))),(Item=(Type=Bone,Name=\"base_bk_skirt_l\"),Transform=(Rotation=(X=0,Y=-0.1908,Z=0,W=0.9816),Translation=(X=-1,Y=0, ...(truncated)"
    Weight=1
    WeightMinimum=0
   )
   n16 = &RigVMFunction_AlphaInterp(
    InterpSpeedDecreasing=5
    InterpSpeedIncreasing=5
    Scale=1
    bClampResult=False
    Value=RigVMVariableNode() {variable=isCrouching_NotSprinting}.Value
   )
   > &RigUnit_ModifyTransforms(
    ItemToModify="((Item=(Type=Bone,Name=\"base_front_skirt_l\"),Transform=(Rotation=(X=-0.173,Y=-0.0858,Z=-0.0151,W=0.9811),Translation=(X=4,Y=0,Z=0),Scale3D=(X=1,Y=1,Z=1))),(Item=(Type=Bone,Name=\"base_side_skirt_l\"),Transform=(Rotation=(X=-0.0218,Y=-0.0378,Z=-0.4995,W=0.865 ...(truncated)"
    WeightMinimum=0
    Weight=@n16.Result
   )
   n17 = &RigVMFunction_AlphaInterp(InterpSpeedDecreasing=3, InterpSpeedIncreasing=20, Scale=1, bClampResult=False, Value=RigVMVariableNode() {variable=isCrouching}.Value)
   > &RigUnit_ModifyTransforms(
    ItemToModify="((Item=(Type=Bone,Name=\"base_front_skirt_l\"),Transform=(Rotation=(X=0,Y=0.1045,Z=0,W=0.9945),Translation=(X=0,Y=0,Z=0),Scale3D=(X=1,Y=1,Z=1))),(Item=(Type=Bone,Name=\"base_LoopA_dw\"),Transform=(Rotation=(X=0,Y=-0.1305,Z=0,W=0.9914),Translation=(X=0,Y=0,Z=0) ...(truncated)"
    WeightMinimum=0
    Weight=@n17.Result
   )
   n18 = &RigVMFunction_AlphaInterp(
    InterpSpeedDecreasing=5
    InterpSpeedIncreasing=20
    Scale=1
    bClampResult=False
    Value=&RigVMFunction_MathDoubleMul(A=RigVMVariableNode() {variable=isCrouching}.Value, B=RigVMVariableNode() {variable=isMoving_Forward}.Value).Result
   )
   > &RigUnit_ModifyTransforms(
    ItemToModify="((Item=(Type=Bone,Name=\"Waistband_Collider_01\"),Transform=(Rotation=(X=0,Y=0,Z=0,W=1),Translation=(X=0,Y=0,Z=0),Scale3D=(X=1,Y=1,Z=1))),(Item=(Type=Bone,Name=\"Waistband_Collider_02\"),Transform=(Rotation=(X=0,Y=0,Z=0,W=1),Translation=(X=1,Y=0,Z=0),Scale3D=( ...(truncated)"
    WeightMinimum=0
    Weight=@n18.Result
   )
   n19 = &RigVMFunction_AlphaInterp(
    InterpSpeedDecreasing=5
    InterpSpeedIncreasing=5
    Scale=1
    bClampResult=False
    Value=RigVMVariableNode() {variable=isTactical Sprinting}.Value
   )
   > &RigUnit_ModifyTransforms(
    ItemToModify="((Item=(Type=Bone,Name=\"base_bk_skirt_r\"),Transform=(Rotation=(X=0,Y=-0.2164,Z=0,W=0.9763),Translation=(X=0,Y=-10,Z=0),Scale3D=(X=1,Y=1,Z=1))),(Item=(Type=Bone,Name=\"base_side_skirt_r\"),Transform=(Rotation=(X=0,Y=0,Z=0,W=1),Translation=(X=0,Y=0,Z=0),Scale3 ...(truncated)"
    WeightMinimum=0
    Weight=@n19.Result
   )
   > &RigUnit_ModifyTransforms(
    ItemToModify=((Item=(Type=Bone,Name="Waistband_Collider_04"),Transform=(Rotation=(X=0,Y=0,Z=0,W=1),Translation=(X=4,Y=0,Z=-2),Scale3D=(X=1,Y=1,Z=1))))
    WeightMinimum=0
    Weight=RigVMVariableNode() {variable=isJumping}.Value
   )
   > DeltaRotation v4 = RigVMFunctionReferenceNode(
    Rotation Order=ZYX
    child=(Type=Bone,Name="thigh_r")
    debugRotation=false
    outputMinMax=(X=0,Y=1)
    parent=(Type=Null,Name="thigh_r_null")
    premult=1
    rotationMinMax=(X=0,Y=-20)
    triggerAxis=Z
    twistAxis=(X=1,Y=0,Z=0)
   )
   > &RigUnit_ModifyTransforms(
    ItemToModify="((Item=(Type=Bone,Name=\"Waistband_Collider_01\"),Transform=(Rotation=(X=0,Y=0,Z=0,W=1),Translation=(X=1.5,Y=0,Z=0),Scale3D=(X=1,Y=1,Z=1))),(Item=(Type=Bone,Name=\"Waistband_Collider_02\"),Transform=(Rotation=(X=0,Y=0,Z=0,W=1),Translation=(X=2,Y=0,Z=0),Scale3D ...(truncated)"
    WeightMinimum=0
    Weight=@"DeltaRotation v4".outputWeight
   )
   n20 = &RigVMFunction_AlphaInterp(InterpSpeedDecreasing=5, InterpSpeedIncreasing=5, Scale=1, bClampResult=False, Value=RigVMVariableNode() {variable=isMoving_Forward}.Value)
   > &RigUnit_ModifyTransforms(
    ItemToModify="((Item=(Type=Bone,Name=\"base_front_skirt_r\"),Transform=(Rotation=(X=0,Y=-0.0436,Z=0,W=0.999),Translation=(X=0,Y=0,Z=1),Scale3D=(X=1,Y=1,Z=1))),(Item=(Type=Bone,Name=\"Waistband_Collider_03\"),Transform=(Rotation=(X=0,Y=0,Z=0,W=1),Translation=(X=1,Y=0,Z=0),Sc ...(truncated)"
    WeightMinimum=0
    Weight=@n20.Result
   )
   n21 = &RigVMFunction_AlphaInterp(InterpSpeedDecreasing=5, InterpSpeedIncreasing=5, Scale=1, bClampResult=False, Value=RigVMVariableNode() {variable=isMoving_Back}.Value)
   > &RigUnit_ModifyTransforms(
    ItemToModify="((Item=(Type=Bone,Name=\"base_front_skirt_r\"),Transform=(Rotation=(X=0,Y=-0.0698,Z=0,W=0.9976),Translation=(X=0,Y=0,Z=1),Scale3D=(X=1,Y=1,Z=1))),(Item=(Type=Bone,Name=\"Waistband_Collider_02\"),Transform=(Rotation=(X=0,Y=0,Z=0,W=1),Translation=(X=1,Y=0,Z=0),S ...(truncated)"
    WeightMinimum=0
    Weight=@n21.Result
   )
   > &RigUnit_ModifyTransforms(
    ItemToModify="((Item=(Type=Bone,Name=\"Waistband_Collider_01\"),Transform=(Rotation=(X=0,Y=0,Z=0,W=1),Translation=(X=1,Y=0,Z=0),Scale3D=(X=1,Y=1,Z=1))),(Item=(Type=Bone,Name=\"Waistband_Collider_02\"),Transform=(Rotation=(X=0,Y=0,Z=-0.0872,W=0.9962),Translation=(X=4,Y=2,Z=0 ...(truncated)"
    WeightMinimum=0
    Weight=RigVMVariableNode() {variable=is_SkyFloating}.Value
   )
   > &RigUnit_ModifyTransforms(
    ItemToModify="((Item=(Type=Bone,Name=\"Waistband_Collider_02\"),Transform=(Rotation=(X=0,Y=0,Z=0,W=1),Translation=(X=-0.5,Y=0,Z=0),Scale3D=(X=1,Y=1,Z=1))),(Item=(Type=Bone,Name=\"Waistband_Collider_03\"),Transform=(Rotation=(X=0,Y=0,Z=0,W=1),Translation=(X=-1.5,Y=1,Z=0),Sca ...(truncated)"
    WeightMinimum=0
    Weight=RigVMVariableNode() {variable=isRegular_Sliding}.Value
   )
   > &RigUnit_ModifyTransforms(
    ItemToModify="((Item=(Type=Bone,Name=\"base_front_skirt_l\"),Transform=(Rotation=(X=0,Y=0.1305,Z=0,W=0.9914),Translation=(X=0,Y=0,Z=0),Scale3D=(X=1,Y=1,Z=1))),(Item=(Type=Bone,Name=\"base_front_skirt_c\"),Transform=(Rotation=(X=0,Y=0.1305,Z=0,W=0.9914),Translation=(X=0,Y=0, ...(truncated)"
    WeightMinimum=0
    Weight=&RigVMFunction_MathBoolToFloat(Value=RigVMVariableNode() {variable=isDBNO}.Value).Result
   )
   &RigUnit_ModifyTransforms(
    ItemToModify="((Item=(Type=Bone,Name=\"deform_upperarm_up_l\"),Transform=(Rotation=(X=0,Y=0,Z=0,W=1),Translation=(X=6,Y=0,Z=0),Scale3D=(X=1,Y=1,Z=1))),(Item=(Type=Bone,Name=\"deform_upperarm_up_r\"),Transform=(Rotation=(X=0,Y=0,Z=0,W=1),Translation=(X=-1,Y=0,Z=1),Scale3D=(X ...(truncated)"
    WeightMinimum=0
    Weight=&RigVMFunction_MathDoubleMul(
     A=RigVMDispatchNode(BoneName=root, Default=0, Name=deformRig_r_upperarm_up_alpha) {template="DISPATCH_RigDispatch_GetAnimAttribute(in Name,in BoneName,in Default,out Value,out Found)"}.Value
     B=RigVMVariableNode() {variable=is_Parachuting}.Value
    ).Result
    ExecutePin=@RigVMFunction_Sequence.D
   )
   > &RigUnit_ModifyTransforms(
    ItemToModify="((Item=(Type=Bone,Name=\"base_bandA\"),Transform=(Rotation=(X=0.171,Y=0.2962,Z=-0.4698,W=0.8138),Translation=(X=0,Y=0,Z=0),Scale3D=(X=1,Y=1,Z=1))),(Item=(Type=Bone,Name=\"base_bandB\"),Transform=(Rotation=(X=0.2717,Y=0.3237,Z=-0.5826,W=0.6943),Translation=(X=0 ...(truncated)"
    WeightMinimum=0
    Weight=RigVMVariableNode() {variable=isBaseball_Sliding}.Value
   )
   > &RigUnit_ModifyTransforms(
    ItemToModify="((Item=(Type=Bone,Name=\"Waistband_Collider_03\"),Transform=(Rotation=(X=0,Y=0,Z=0,W=1),Translation=(X=1.5,Y=0,Z=0),Scale3D=(X=1,Y=1,Z=1))),(Item=(Type=Bone,Name=\"Waistband_Collider_02\"),Transform=(Rotation=(X=0,Y=0,Z=0,W=1),Translation=(X=1.25,Y=0,Z=0),Scal ...(truncated)"
    WeightMinimum=0
    Weight=RigVMVariableNode() {variable=is_Celebrating}.Value
   )
   > n22 = &RigUnit_SphericalPoseReader(
    ActiveRegionScaleFactors=(PositiveWidth=0.5,NegativeWidth=3,PositiveHeight=1.5,NegativeHeight=1.5)
    ActiveRegionSize=0.05
    Debug=(bDrawDebug=false,DebugScale=25,DebugSegments=20,DebugThickness=0.25)
    DriverAxis=(X=1,Y=0,Z=0)
    DriverItem=(Type=Bone,Name="upperarm_l")
    FalloffRegionScaleFactors=(PositiveWidth=1,NegativeWidth=1,PositiveHeight=1,NegativeHeight=1)
    FalloffSize=0.2
    OptionalParentItem=(Type=Bone)
    RotationOffset=(X=105,Y=0,Z=-90)
   )
   > &RigUnit_ModifyTransforms(
    ItemToModify="((Item=(Type=Bone,Name=\"base_l_padChain3A\"),Transform=(Rotation=(X=0,Y=-0.0872,Z=0,W=0.9962),Translation=(X=0,Y=0,Z=0),Scale3D=(X=1,Y=1,Z=1))),(Item=(Type=Bone,Name=\"base_l_padChain4\"),Transform=(Rotation=(X=0,Y=-0.0872,Z=0,W=0.9962),Translation=(X=0,Y=0,Z ...(truncated)"
    WeightMinimum=0
    Weight=&RigVMFunction_MathDoubleMul(A=@n22.OutputParam, B=RigVMVariableNode() {variable=isAR Equipped lerp}.Value).Result
   )
   > n23 = &RigUnit_SphericalPoseReader(
    ActiveRegionScaleFactors=(PositiveWidth=0.5,NegativeWidth=3,PositiveHeight=1.5,NegativeHeight=1.5)
    ActiveRegionSize=0.05
    Debug=(bDrawDebug=false,DebugScale=25,DebugSegments=20,DebugThickness=0.25)
    DriverAxis=(X=-1,Y=0,Z=0)
    DriverItem=(Type=Bone,Name="upperarm_r")
    FalloffRegionScaleFactors=(PositiveWidth=1,NegativeWidth=1,PositiveHeight=1,NegativeHeight=1)
    FalloffSize=0.15
    FlipWidthScaling=false
    OptionalParentItem=(Type=Bone)
    RotationOffset=(X=-85,Y=0,Z=-50)
   )
   > &RigUnit_ModifyTransforms(
    ItemToModify="((Item=(Type=Bone,Name=\"base_r_padChain3\"),Transform=(Rotation=(X=0,Y=-0.0523,Z=0,W=0.9986),Translation=(X=0,Y=0,Z=0),Scale3D=(X=1,Y=1,Z=1))),(Item=(Type=Bone,Name=\"base_r_padChain3A\"),Transform=(Rotation=(X=0,Y=-0.0698,Z=0,W=0.9976),Translation=(X=0,Y=0,Z ...(truncated)"
    WeightMinimum=0
    Weight=&RigVMFunction_MathDoubleMul(A=@n23.OutputParam, B=RigVMVariableNode() {variable=isAR Equipped lerp}.Value).Result
   )
   > DeltaRotation v4_1 = RigVMFunctionReferenceNode(
    Rotation Order=ZYX
    child=(Type=Bone,Name="thigh_l")
    debugRotation=false
    onlyTwist=true
    outputMinMax=(X=0,Y=1)
    parent=(Type=Null,Name="thigh_l_null")
    premult=1
    rotationMinMax=(X=0,Y=-65)
    triggerAxis=X
    twistAxis=(X=1,Y=0,Z=0)
   )
   > &RigUnit_ModifyTransforms(
    ItemToModify="((Item=(Type=Bone,Name=\"base_bk_skirt_l\"),Transform=(Rotation=(X=-0.3368,Y=-0.0594,Z=0.1632,W=0.9254),Translation=(X=0,Y=5,Z=5),Scale3D=(X=1,Y=1,Z=1))),(Item=(Type=Bone,Name=\"base_side_skirt_l\"),Transform=(Rotation=(X=0,Y=0,Z=0,W=1),Translation=(X=0,Y=1,Z= ...(truncated)"
    WeightMinimum=0
    Weight=@"DeltaRotation v4_1".outputWeight
   )
   > DeltaRotation v4_2 = RigVMFunctionReferenceNode(
    Rotation Order=ZYX
    child=(Type=Bone,Name="thigh_r")
    debugRotation=false
    onlyTwist=true
    outputMinMax=(X=0,Y=1)
    parent=(Type=Null,Name="thigh_r_null")
    premult=1
    rotationMinMax=(X=0,Y=-65)
    triggerAxis=X
    twistAxis=(X=1,Y=0,Z=0)
   )
   > &RigUnit_ModifyTransforms(
    ItemToModify="((Item=(Type=Bone,Name=\"base_bk_skirt_r\"),Transform=(Rotation=(X=0.3827,Y=0,Z=0,W=0.9239),Translation=(X=0,Y=-6,Z=0),Scale3D=(X=1,Y=1,Z=1))),(Item=(Type=Bone,Name=\"base_side_skirt_r\"),Transform=(Rotation=(X=0,Y=0,Z=0,W=1),Translation=(X=0,Y=0,Z=1.5),Scale3 ...(truncated)"
    WeightMinimum=0
    Weight=@"DeltaRotation v4_2".outputWeight
   )
   > n24 = &RigUnit_SphericalPoseReader(
    ActiveRegionScaleFactors=(PositiveWidth=1,NegativeWidth=1.4,PositiveHeight=1,NegativeHeight=1)
    ActiveRegionSize=0.13
    Debug=(bDrawDebug=true,DebugScale=25,DebugSegments=20,DebugThickness=0.25)
    DriverAxis=(X=1,Y=0,Z=0)
    DriverItem=(Type=Bone,Name="upperarm_l")
    FalloffRegionScaleFactors=(PositiveWidth=1,NegativeWidth=3,PositiveHeight=1,NegativeHeight=1)
    FalloffSize=0.24
    OptionalParentItem=(Type=Bone)
    RotationOffset=(X=115,Y=-120,Z=-40)
   )
   n25 = RigVMVariableNode() {variable=isTargeting lerp}
   n26 = &RigVMFunction_AlphaInterp(
    InterpSpeedDecreasing=10
    InterpSpeedIncreasing=5
    Scale=1
    bClampResult=False
    Value=&RigVMFunction_MathDoubleMul(
     B=&RigVMFunction_MathDoubleRemap(SourceMaximum=0, SourceMinimum=1, TargetMaximum=1, TargetMinimum=0, bClamp=False, Value=@n25.Value).Result
     A=@n24.OutputParam
    ).Result
   )
   > &RigUnit_ModifyTransforms(
    ItemToModify="((Item=(Type=Bone,Name=\"base_l_padChain1\"),Transform=(Rotation=(X=0,Y=-0.5,Z=0,W=0.866),Translation=(X=-1,Y=0,Z=0),Scale3D=(X=1,Y=1,Z=1))),(Item=(Type=Bone,Name=\"base_l_padChain1A\"),Transform=(Rotation=(X=-0.0755,Y=-0.4981,Z=-0.0436,W=0.8627),Translation=( ...(truncated)"
    WeightMinimum=0
    Weight=@n26.Result
   )
   > n27 = &RigUnit_SphericalPoseReader(
    ActiveRegionScaleFactors=(PositiveWidth=1,NegativeWidth=1.4,PositiveHeight=1,NegativeHeight=1)
    ActiveRegionSize=0.12
    Debug=(bDrawDebug=true,DebugScale=25,DebugSegments=20,DebugThickness=0.25)
    DriverAxis=(X=-1,Y=0,Z=0)
    DriverItem=(Type=Bone,Name="upperarm_r")
    FalloffRegionScaleFactors=(PositiveWidth=1,NegativeWidth=3,PositiveHeight=1,NegativeHeight=1)
    FalloffSize=0.24
    OptionalParentItem=(Type=Bone)
    RotationOffset=(X=-120,Y=80,Z=-40)
   )
   n28 = RigVMVariableNode() {variable=isTargeting lerp}
   n29 = &RigVMFunction_AlphaInterp(
    InterpSpeedDecreasing=5
    InterpSpeedIncreasing=5
    Scale=1
    bClampResult=False
    Value=&RigVMFunction_MathDoubleMul(
     B=&RigVMFunction_MathDoubleRemap(SourceMaximum=0, SourceMinimum=1, TargetMaximum=1, TargetMinimum=0, bClamp=False, Value=@n28.Value).Result
     A=@n27.OutputParam
    ).Result
   )
   > &RigUnit_ModifyTransforms(
    ItemToModify="((Item=(Type=Bone,Name=\"base_r_padChain1\"),Transform=(Rotation=(X=-0.0151,Y=-0.173,Z=-0.0858,W=0.9811),Translation=(X=-0.25,Y=0,Z=0),Scale3D=(X=1,Y=1,Z=1))),(Item=(Type=Bone,Name=\"base_r_padChain1A\"),Transform=(Rotation=(X=0,Y=-0.4617,Z=0,W=0.887),Translat ...(truncated)"
    WeightMinimum=0
    Weight=@n29.Result
   )
  ]}
 New Function_ContainedGraph {nodes [
   n1 = RigVMFunctionEntryNode()
   RigVMFunction_MathBoolToInteger = &RigVMFunction_MathBoolToInteger(Value=@n1.onlyTwist)
   RigVMFunction_MathQuaternionSwingTwist = &RigVMFunction_MathQuaternionSwingTwist(
    TwistAxis=@n1.twistAxis
    Input=&RigUnit_GetRelativeTransformForItem(bChildInitial=False, bParentInitial=False, Child=@n1.child, Parent=@n1.parent).RelativeTransform.Rotation
   )
   n2 = RigVMDispatchNode(
    Values=((),())
    Index=@RigVMFunction_MathBoolToInteger.Result
    Values.0=@RigVMFunction_MathQuaternionSwingTwist.Swing
    Values.1=@RigVMFunction_MathQuaternionSwingTwist.Twist
   ) {template="DISPATCH_RigVMDispatch_SelectInt32(in Index,in Values,out Result)"}
   RigVMFunction_MathQuaternionToEuler = &RigVMFunction_MathQuaternionToEuler(Value=@n2.Result, RotationOrder=@n1.Rotation Order)
   > DISPATCH_RigVMDispatch_Print = RigVMDispatchNode(ScreenColor=(R=1,G=1,B=1,A=1), ScreenDuration=1.5, Enabled=@n1.debugRotation, Value=@RigVMFunction_MathQuaternionToEuler.Result) {template="DISPATCH_RigVMDispatch_Print(in Prefix,in Value,in Enabled,in ScreenDuration,in ScreenColor)"}
   Cast = RigVMDispatchNode(Value=@n1.triggerAxis) {template="DISPATCH_RigVMDispatch_CastEnumToInt(in Value,out Result)"}
   DISPATCH_RigVMDispatch_SelectInt32 = RigVMDispatchNode(
    Values=(0,0,0)
    Index=@Cast.Result
    Values.0=@RigVMFunction_MathQuaternionToEuler.Result.X
    Values.1=@RigVMFunction_MathQuaternionToEuler.Result.Y
    Values.2=@RigVMFunction_MathQuaternionToEuler.Result.Z
   ) {template="DISPATCH_RigVMDispatch_SelectInt32(in Index,in Values,out Result)"}
   Remap = &RigVMFunction_MathDoubleRemap(
    bClamp=False
    Value=@DISPATCH_RigVMDispatch_SelectInt32.Result
    TargetMinimum=@n1.outputMinMax.X
    TargetMaximum=@n1.outputMinMax.Y
    SourceMinimum=@n1.rotationMinMax.X
    SourceMaximum=@n1.rotationMinMax.Y
   )
   Absolute = &RigVMFunction_MathDoubleAbs(Value=@Remap.Result)
   Multiply = &RigVMFunction_MathDoubleMul(
    A=&RigVMFunction_MathDoubleClamp(Value=RigVMDispatchNode(True=@Absolute.Result, Condition=@n1.absoluteValue, False=@Remap.Result) {template="DISPATCH_RigVMDispatch_If(in Condition,in True,in False,out Result)"}.Result).Result
    B=@n1.premult
   )
   > RigVMFunctionReturnNode(outputQuat=@n2.Result, outputWeight=@Multiply.Result)
  ]}
 RigVMFunctionLibrary {nodes [
   DeltaRotation v4 = RigVMCollapseNode(
    ExecuteContext=()
    Rotation Order=ZYX
    child=(Type=Bone,Name="None")
    outputMinMax=(X=0,Y=1)
    parent=(Type=Bone,Name="None")
    premult=1
    rotationMinMax=(X=0,Y=0)
    twistAxis=(X=1,Y=0,Z=0)
   )
   Surface_Contraints = RigVMCollapseNode(ExecuteContext=())
   lerp_Bools = RigVMCollapseNode(ExecuteContext=())
   padChains_Fixes = RigVMCollapseNode(ExecuteContext=())
  ]}
 RigVMFunction_MathBoolOr_1_ContainedGraph {nodes [
   n1 = RigVMFunctionEntryNode()
   RigVMFunction_MathBoolOr = &RigVMFunction_MathBoolOr(A=@n1.A, B=@n1.B)
   RigVMFunctionReturnNode(Result=&RigVMFunction_MathBoolOr(B=@n1.C, A=@RigVMFunction_MathBoolOr.Result).Result)
  ]}
 RigVMFunction_Sequence_1_ContainedGraph {nodes [
   n1 = RigVMFunctionEntryNode()
   RigVMFunction_Sequence = &RigVMFunction_Sequence(ExecuteContext=@n1.ExecuteContext)
   n2 = &RigVMFunction_Sequence(ExecuteContext=@RigVMFunction_Sequence.B)
   n3 = &RigVMFunction_Sequence(ExecuteContext=@n2.B)
   n4 = &RigVMFunction_Sequence(ExecuteContext=@n3.B)
   RigVMFunctionReturnNode(A=@RigVMFunction_Sequence.A, B=@n2.A, C=@n3.A, D=@n4.A, E=@n4.B)
  ]}
 RigVMFunction_Sequence_1_ContainedGraph {nodes [
   n1 = RigVMFunctionEntryNode()
   RigVMFunction_Sequence = &RigVMFunction_Sequence(ExecuteContext=@n1.ExecuteContext)
   n2 = &RigVMFunction_Sequence(ExecuteContext=@RigVMFunction_Sequence.B)
   n3 = &RigVMFunction_Sequence(ExecuteContext=@n2.B)
   n4 = &RigVMFunction_Sequence(ExecuteContext=@n3.B)
   RigVMFunctionReturnNode(A=@RigVMFunction_Sequence.A, B=@n2.A, C=@n3.A, D=@n4.A, E=@n4.B)
  ]}
 RigVMModel {nodes [
   n1 = &RigUnit_PrepareForExecution()
   addFKCtrl = RigVMFunctionReferenceNode(
    Custom_FKCtrl=()
    Options_FKCtrl="(pelvis_1_#1=false,spine_3_#2=false,clavicle_5_#3=true,arm_7_#4=true,leg_9_#5=true,head_11_46DC15EF465CD29 ...(truncated)"
    Shape_Settings="(pelvis_2_#6=(bVisible=True,Name=\"Hexagon_Thin\",Color=(R=0.0117,G=0.8281,B=1,A=1),Transform=(Rotation=(X=0,Y=-0.7071,Z=0,W=0.7071),Translation=(X=2,Y=0,Z=0),Scale3D=(X=5,Y=5,Z=7))),spine_4_#7=(bVi ...(truncated)"
    ExecuteContext=@n1.ExecutePin
   )
   RigUnit_BeginExecution = &RigUnit_BeginExecution()
   setFKCtrl = RigVMFunctionReferenceNode(setFK_Ctrl=false, ExecuteContext=@RigUnit_BeginExecution.ExecutePin)
   > FN_DeformRig_FNC = RigVMFunctionReferenceNode(
    DeformRig_ Curve="(deformRig_l_upperarm_up_alpha_72_#8=true,deformRig_r_upperarm_up_alpha_73_#9=true,deformRig_l_upperarm_fr_alpha_74_#10=true,deformRig_r_upperarm_fr_alpha_75_CB51B80C4 ...(truncated)"
    DeformRig_Bool="(upperarm_1_#11=True,elbow_3_#12=True,wrist_7_#13=True,groin_8_#14=True,glute_17_#15=True,knee_9_40C408324CA261 ...(truncated)"
    DeformRig_Debug=(upperarm_debug_18_#11=False,wrist_debug_20_#13=False,lat_debug_26_#16=False,groinGlute_debug_27_#14=False)
    DeformRig_Value="(upperarm_up_18_#17=(Rotation=(X=0.0,Y=-0.0,Z=0.0,W=1.0),Translation=(X=0.0,Y=0.0,Z=8.0),Scale3D=(X=0.70,Y=1.0,Z=1.0)),upperarm_fwd_fr_19_#18=(Rotation=(X=0.0,Y=-0.0,Z=0.0,W=1.0),Translation=(X=0.0, ...(truncated)"
   )
   > lerp_Bools = RigVMFunctionReferenceNode()
   > Surface_Contraints = RigVMFunctionReferenceNode()
   > padChains_Fixes = RigVMFunctionReferenceNode()
  ]}
]
hierarchy [
 root {loc=<0 0 0>, rot=<0 0 0 1>}
 pelvis {parent=root, loc=<0 -0.5301 98.0571>, rot=<0.0127 0.707 -0.0127 -0.707>}
 spine_01 {parent=pelvis, loc=<6.1208 -0.9328 0>, rot=<0 0 0.034 -0.9994>}
 spine_02 {parent=spine_01, loc=<6.5997 -0.2095 0>, rot=<0 0 0.0238 -0.9997>}
 spine_03 {parent=spine_02, loc=<6.4914 -0.1237 0>, rot=<0 0 0.0149 -0.9999>}
 spine_04 {parent=spine_03, loc=<6.9598 0.0794 0>, rot=<0 0 -0.0132 -0.9999>}
 spine_05 {parent=spine_04, loc=<6.9987 0 0>, rot=<0 0 -0.0015 -1>}
 neck_01 {parent=spine_05, loc=<11.3138 0.1529 0>, rot=<0 0 -0.2003 -0.9797>}
 neck_02 {parent=neck_01, loc=<4.5627 0 0>, rot=<0 0 0.0504 -0.9987>}
 head {parent=neck_02, loc=<5.3913 0 0>, rot=<0 0 0.115 -0.9934>}
 clavicle_l {parent=spine_05, loc=<5.4831 -0.1053 -3.2946>, rot=<0.0465 -0.7289 -0.009 -0.6829>}
 upperarm_l {parent=clavicle_l, loc=<9.8658 -1.2136 -0.4105>, rot=<0.0164 -0.3245 -0.0269 -0.9454>}
 lowerarm_l {parent=upperarm_l, loc=<26.465 -0.8345 0.259>, rot=<0.0036 0.0161 -0.2298 -0.9731>}
 hand_l {parent=lowerarm_l, loc=<22.7355 -0.3203 -0.6105>, rot=<-0.6784 -0.0223 0.0735 -0.7307>}
 thumb_01_l {parent=hand_l, loc=<1.3118 -0.3191 -2.459>, rot=<0.6546 -0.4042 0.0664 -0.6354>}
 thumb_02_l {parent=thumb_01_l, loc=<5.9336 -0.0448 0.0401>, rot=<-0.0037 0.0213 0.298 -0.9543>}
 thumb_03_l {parent=thumb_02_l, loc=<2.2223 0.3682 -0.0907>, rot=<0.0118 -0.013 0.0066 -0.9998>}
 index_metacarpal_l {parent=hand_l, loc=<1.6535 0.0183 -2.96>, rot=<0.0158 -0.0124 -0.0173 -0.9996>}
 index_01_l {parent=index_metacarpal_l, loc=<8.3008 0.0357 -0.0524>, rot=<-0.0044 0.0485 0.0924 -0.9945>}
 index_02_l {parent=index_01_l, loc=<3.3587 -0.3004 0.0452>, rot=<-0.002 0.0027 0.3225 -0.9466>}
 index_03_l {parent=index_02_l, loc=<2.5578 -0.123 -0.0252>, rot=<-0.002 0.0067 0.2558 -0.9667>}
 middle_metacarpal_l {parent=hand_l, loc=<2.1856 -0.1725 -0.5984>, rot=<0.006 0.0168 -0.0329 -0.9993>}
 middle_01_l {parent=middle_metacarpal_l, loc=<7.3275 -0.2447 0.0129>, rot=<-0.0208 0.0042 0.0858 -0.9961>}
 middle_02_l {parent=middle_01_l, loc=<4.0123 0.1401 0.1175>, rot=<-0.0218 0.0258 0.3512 -0.9357>}
 middle_03_l {parent=middle_02_l, loc=<2.6077 0.0534 0.1373>, rot=<0.0082 0.0188 0.1708 -0.9851>}
 ring_metacarpal_l {parent=hand_l, loc=<2.2074 -0.3625 1.3638>, rot=<-0.0284 0.0722 -0.0458 -0.9959>}
 ring_01_l {parent=ring_metacarpal_l, loc=<7.0295 -0.6197 -0.1446>, rot=<-0.0489 -0.0252 0.1892 -0.9804>}
 ring_02_l {parent=ring_01_l, loc=<3.6174 0.4647 -0.1027>, rot=<0 0 0.2806 -0.9598>}
 ring_03_l {parent=ring_02_l, loc=<2.4152 0.114 0.0199>, rot=<0 0 0.1189 -0.9929>}
 pinky_metacarpal_l {parent=hand_l, loc=<1.7616 -0.9753 3.3814>, rot=<-0.0397 0.0814 -0.069 -0.9935>}
 pinky_01_l {parent=pinky_metacarpal_l, loc=<7.1709 -0.4831 -0.1094>, rot=<-0.0654 -0.0035 0.2242 -0.9723>}
 pinky_02_l {parent=pinky_01_l, loc=<2.8274 0.0893 -0.1043>, rot=<0.0037 0.0011 0.2854 -0.9584>}
 pinky_03_l {parent=pinky_02_l, loc=<2.2046 0.0307 -0.0078>, rot=<-0.0002 0.0098 0.1445 -0.9895>}
 deform_wrist_dn_base_l {parent=hand_l, loc=<0 -1.5326 0>, rot=<0 0 0 -1>}
 deform_wrist_dn_l {parent=deform_wrist_dn_base_l, loc=<0 0 0>, rot=<0 0 0 1>}
 deform_wrist_up_base_l {parent=hand_l, loc=<0 2.3009 0>, rot=<0 0 0 -1>}
 deform_wrist_up_l {parent=deform_wrist_up_base_l, loc=<0 0 0>, rot=<0 0 0 1>}
 lowerarm_twist_02_l {parent=lowerarm_l, loc=<10.6707 0 0>, rot=<0 0 0 1>}
 lowerarm_twist_01_l {parent=lowerarm_l, loc=<18.316 0 0>, rot=<0 0 0 1>}
 deform_elbow_in_base_l {parent=lowerarm_l, loc=<0.1834 3.6062 -0.2572>, rot=<0 0 0 -1>}
 deform_elbow_in_l {parent=deform_elbow_in_base_l, loc=<0 0 0>, rot=<0 0 0 1>}
 deform_elbow_out_base_l {parent=lowerarm_l, loc=<0 -4 0>, rot=<0 0 0 1>}
 deform_elbow_out_l {parent=deform_elbow_out_base_l, loc=<0 0 0>, rot=<0 0 0 1>}
 upperarm_twist_01_l {parent=upperarm_l, loc=<7.2135 -0.1134 0>, rot=<0 0 0 1>}
 deform_upperarm_up_base_l {parent=upperarm_twist_01_l, loc=<-8.891 0.4825 3.9517>, rot=<0 0 0 -1>}
 deform_upperarm_up_l {parent=deform_upperarm_up_base_l, loc=<0 0 0>, rot=<0 0 0 1>}
 base_l_padChain1 {parent=upperarm_twist_01_l, loc=<-9.2157 5.9118 0.383>, rot=<0.4359 -0.1224 -0.3282 -0.829>}
 dyn_l_padChain1_01 {parent=base_l_padChain1, loc=<0 0 0>, rot=<0 0 0 1>}
 dyn_l_padChain1_02 {parent=dyn_l_padChain1_01, loc=<3.3913 0 0>, rot=<-0.0683 -0.1282 0.0089 -0.9894>}
 base_l_padChain2 {parent=upperarm_twist_01_l, loc=<-7.9186 5.0709 1.5055>, rot=<0.4389 -0.162 -0.3086 -0.8282>}
 dyn_l_padChain2_01 {parent=base_l_padChain2, loc=<0 0 0>, rot=<0 0 0 1>}
 dyn_l_padChain2_02 {parent=dyn_l_padChain2_01, loc=<3.3928 0 0>, rot=<-0.0677 -0.1064 0.0073 -0.992>}
 base_l_padChain3 {parent=upperarm_twist_01_l, loc=<-6.9214 4.1285 2.7651>, rot=<0.3977 -0.1798 -0.2795 -0.8552>}
 dyn_l_padChain3_01 {parent=base_l_padChain3, loc=<0 0 0>, rot=<0 0 0 1>}
 dyn_l_padChain3_02 {parent=dyn_l_padChain3_01, loc=<3.395 0 0>, rot=<-0.0295 -0.0902 0.0027 -0.9955>}
 base_l_padChain4 {parent=upperarm_twist_01_l, loc=<-6.4101 3.0023 4.0183>, rot=<0.3323 -0.1845 -0.2698 -0.8847>}
 dyn_l_padChain4_01 {parent=base_l_padChain4, loc=<0 0 0>, rot=<0 0 0 1>}
 dyn_l_padChain4_02 {parent=dyn_l_padChain4_01, loc=<3.3968 0 0>, rot=<0 -0.072 0 -0.9974>}
 base_l_padChain5 {parent=upperarm_twist_01_l, loc=<-6.397 2.3831 4.5813>, rot=<0.1756 -0.2057 -0.2195 -0.9373>}
 dyn_l_padChain5_01 {parent=base_l_padChain5, loc=<0 0 0>, rot=<0 0 0 1>}
 dyn_l_padChain5_02 {parent=dyn_l_padChain5_01, loc=<3.3965 0 0>, rot=<0.1457 -0.0641 -0.0095 -0.9872>}
 base_l_padChain6 {parent=upperarm_twist_01_l, loc=<-6.362 1.772 4.9586>, rot=<0.0529 -0.1891 -0.1773 -0.9644>}
 dyn_l_padChain6_01 {parent=base_l_padChain6, loc=<0 0 0>, rot=<0 0 0 1>}
 dyn_l_padChain6_02 {parent=dyn_l_padChain6_01, loc=<3.3957 0 0>, rot=<0.2107 -0.0661 -0.0143 -0.9752>}
 base_l_padChain7 {parent=upperarm_twist_01_l, loc=<-6.3211 1.1521 5.2432>, rot=<0.0891 -0.1269 -0.1142 -0.9813>}
 dyn_l_padChain7_01 {parent=base_l_padChain7, loc=<0 0 0>, rot=<0 0 0 1>}
 dyn_l_padChain7_02 {parent=dyn_l_padChain7_01, loc=<3.3977 0 0>, rot=<0.025 -0.0494 -0.0012 -0.9985>}
 base_l_padChain8 {parent=upperarm_twist_01_l, loc=<-6.3052 0.5629 5.4062>, rot=<-0.1096 -0.1199 -0.0085 -0.9867>}
 dyn_l_padChain8_01 {parent=base_l_padChain8, loc=<0 0 0>, rot=<0 0 0 1>}
 dyn_l_padChain8_02 {parent=dyn_l_padChain8_01, loc=<3.3974 0 0>, rot=<0.1339 -0.0522 -0.0071 -0.9896>}
 base_l_padChain9 {parent=upperarm_twist_01_l, loc=<-6.2665 -0.1853 5.4387>, rot=<-0.0769 -0.1277 0.0251 -0.9885>}
 dyn_l_padChain9_01 {parent=base_l_padChain9, loc=<0 0 0>, rot=<0 0 0 1>}
 dyn_l_padChain9_02 {parent=dyn_l_padChain9_01, loc=<3.3979 0 0>, rot=<0 -0.0405 0 -0.9992>}
 base_l_padChain10 {parent=upperarm_twist_01_l, loc=<-6.3115 -0.9584 5.2859>, rot=<-0.1918 -0.1281 0.1194 -0.9657>}
 dyn_l_padChain10_01 {parent=base_l_padChain10, loc=<0 0 0>, rot=<0 0 0 1>}
 dyn_l_padChain10_02 {parent=dyn_l_padChain10_01, loc=<3.3974 0 0>, rot=<0 -0.045 0 -0.999>}
 base_l_padChain11 {parent=upperarm_twist_01_l, loc=<-6.3753 -1.795 4.8983>, rot=<-0.2537 -0.147 0.1741 -0.9401>}
 dyn_l_padChain11_01 {parent=base_l_padChain11, loc=<0 0 0>, rot=<0 0 0 1>}
 dyn_l_padChain11_02 {parent=dyn_l_padChain11_01, loc=<3.3972 0 0>, rot=<-0.1831 -0.0446 0.0083 -0.9821>}
 base_l_padChain12 {parent=upperarm_twist_01_l, loc=<-6.496 -2.5818 4.2767>, rot=<-0.4076 -0.1642 0.216 -0.8719>}
 dyn_l_padChain12_01 {parent=base_l_padChain12, loc=<0 0 0>, rot=<0 0 0 1>}
 dyn_l_padChain12_02 {parent=dyn_l_padChain12_01, loc=<3.3963 0 0>, rot=<-0.0616 -0.0273 0.0017 -0.9977>}
 base_l_padChain13 {parent=upperarm_twist_01_l, loc=<-6.7056 -3.1616 3.5786>, rot=<-0.4833 -0.1897 0.2422 -0.8196>}
 dyn_l_padChain13_01 {parent=base_l_padChain13, loc=<0 0 0>, rot=<0 0 0 1>}
 dyn_l_padChain13_02 {parent=dyn_l_padChain13_01, loc=<3.3908 0 0>, rot=<0 -0.0454 0 -0.999>}
 base_l_padChain14 {parent=upperarm_twist_01_l, loc=<-7.0182 -3.5852 2.7945>, rot=<-0.5629 -0.1653 0.2566 -0.7681>}
 dyn_l_padChain14_01 {parent=base_l_padChain14, loc=<0 0 0>, rot=<0 0 0 1>}
 dyn_l_padChain14_02 {parent=dyn_l_padChain14_01, loc=<3.3902 0 0>, rot=<0 -0.0613 0 -0.9981>}
 base_l_padChain15 {parent=upperarm_twist_01_l, loc=<-7.2582 -3.8508 2.1339>, rot=<-0.5728 -0.209 0.3019 -0.7329>}
 dyn_l_padChain15_01 {parent=base_l_padChain15, loc=<0 0 0>, rot=<0 0 0 1>}
 dyn_l_padChain15_02 {parent=dyn_l_padChain15_01, loc=<3.3956 0 0>, rot=<0 -0.0594 0 -0.9982>}
 base_l_padChain16 {parent=upperarm_twist_01_l, loc=<-7.9642 -4.1554 0.918>, rot=<-0.5907 -0.2209 0.315 -0.7093>}
 dyn_l_padChain16_01 {parent=base_l_padChain16, loc=<0 0 0>, rot=<0 0 0 1>}
 dyn_l_padChain16_02 {parent=dyn_l_padChain16_01, loc=<3.3936 0 0>, rot=<0 -0.072 0 -0.9974>}
 base_l_padChain17 {parent=upperarm_twist_01_l, loc=<-9.1411 -4.4209 -0.479>, rot=<-0.629 -0.2416 0.3314 -0.6604>}
 dyn_l_padChain17_01 {parent=base_l_padChain17, loc=<0 0 0>, rot=<0 0 0 1>}
 dyn_l_padChain17_02 {parent=dyn_l_padChain17_01, loc=<3.3914 0 0>, rot=<0 -0.0578 0 -0.9983>}
 base_l_padChain18 {parent=upperarm_twist_01_l, loc=<-10.4072 -4.514 -1.5882>, rot=<-0.6253 -0.2693 0.3277 -0.655>}
 dyn_l_padChain18_01 {parent=base_l_padChain18, loc=<0 0 0>, rot=<0 0 0 1>}
 dyn_l_padChain18_02 {parent=dyn_l_padChain18_01, loc=<3.3943 0 0>, rot=<0 -0.045 0 -0.999>}
 base_l_padChain1A {parent=upperarm_twist_01_l, loc=<-8.6424 5.5555 0.9489>, rot=<0.4359 -0.1224 -0.3282 -0.829>}
 dyn_l_padChain1A_01 {parent=base_l_padChain1A, loc=<0 0 0>, rot=<0 0 0 1>}
 dyn_l_padChain1A_02 {parent=dyn_l_padChain1A_01, loc=<3.3913 0 0>, rot=<-0.0683 -0.1282 0.0089 -0.9894>}
 base_l_padChain2A {parent=upperarm_twist_01_l, loc=<-7.3469 4.6071 2.1265>, rot=<0.3998 -0.1762 -0.3008 -0.8477>}
 dyn_l_padChain2A_01 {parent=base_l_padChain2A, loc=<0 0 0>, rot=<0 0 0 1>}
 dyn_l_padChain2A_02 {parent=dyn_l_padChain2A_01, loc=<3.3928 0 0>, rot=<-0.0677 -0.1064 0.0073 -0.992>}
 base_l_padChain3A {parent=upperarm_twist_01_l, loc=<-6.6115 3.6037 3.3516>, rot=<0.3927 -0.1814 -0.2784 -0.8575>}
 dyn_l_padChain3A_01 {parent=base_l_padChain3A, loc=<0 0 0>, rot=<0 0 0 1>}
 dyn_l_padChain3A_02 {parent=dyn_l_padChain3A_01, loc=<3.395 0 0>, rot=<-0.0295 -0.0902 0.0027 -0.9955>}
 upperarm_twist_02_l {parent=upperarm_l, loc=<19.7998 -0.1134 0>, rot=<0 0 0 1>}
 clavicle_r {parent=spine_05, loc=<5.4831 -0.1053 3.2946>, rot=<0.7289 0.0465 -0.6829 0.009>}
 upperarm_r {parent=clavicle_r, loc=<-9.8658 1.2136 0.4105>, rot=<0.0164 -0.3245 -0.0269 -0.9454>}
 lowerarm_r {parent=upperarm_r, loc=<-26.465 0.8345 -0.259>, rot=<0.0036 0.0161 -0.2298 -0.9731>}
 hand_r {parent=lowerarm_r, loc=<-22.7355 0.3203 0.6106>, rot=<-0.6784 -0.0223 0.0735 -0.7307>}
 thumb_01_r {parent=hand_r, loc=<-1.3118 0.3191 2.459>, rot=<0.6546 -0.4042 0.0664 -0.6354>}
 thumb_02_r {parent=thumb_01_r, loc=<-5.9336 0.0448 -0.0401>, rot=<-0.0037 0.0213 0.298 -0.9543>}
 thumb_03_r {parent=thumb_02_r, loc=<-2.2223 -0.3682 0.0907>, rot=<0.0118 -0.013 0.0066 -0.9998>}
 index_metacarpal_r {parent=hand_r, loc=<-1.6535 -0.0183 2.96>, rot=<0.0158 -0.0124 -0.0173 -0.9996>}
 index_01_r {parent=index_metacarpal_r, loc=<-8.3008 -0.0357 0.0524>, rot=<-0.0044 0.0485 0.0924 -0.9945>}
 index_02_r {parent=index_01_r, loc=<-3.3587 0.3004 -0.0452>, rot=<-0.002 0.0027 0.3225 -0.9466>}
 index_03_r {parent=index_02_r, loc=<-2.5578 0.123 0.0252>, rot=<-0.002 0.0067 0.2558 -0.9667>}
 middle_metacarpal_r {parent=hand_r, loc=<-2.1856 0.1725 0.5984>, rot=<0.006 0.0168 -0.0329 -0.9993>}
 middle_01_r {parent=middle_metacarpal_r, loc=<-7.3275 0.2447 -0.0129>, rot=<-0.0208 0.0042 0.0858 -0.9961>}
 middle_02_r {parent=middle_01_r, loc=<-4.0123 -0.1401 -0.1175>, rot=<-0.0218 0.0258 0.3512 -0.9357>}
 middle_03_r {parent=middle_02_r, loc=<-2.6077 -0.0534 -0.1373>, rot=<0.0082 0.0188 0.1708 -0.9851>}
 ring_metacarpal_r {parent=hand_r, loc=<-2.2074 0.3625 -1.3638>, rot=<-0.0284 0.0722 -0.0458 -0.9959>}
 ring_01_r {parent=ring_metacarpal_r, loc=<-7.0295 0.6197 0.1446>, rot=<-0.0489 -0.0252 0.1892 -0.9804>}
 ring_02_r {parent=ring_01_r, loc=<-3.6174 -0.4647 0.1027>, rot=<0 0 0.2806 -0.9598>}
 ring_03_r {parent=ring_02_r, loc=<-2.4152 -0.114 -0.0199>, rot=<0 0 0.1189 -0.9929>}
 pinky_metacarpal_r {parent=hand_r, loc=<-1.7616 0.9753 -3.3814>, rot=<-0.0397 0.0814 -0.069 -0.9935>}
 pinky_01_r {parent=pinky_metacarpal_r, loc=<-7.1709 0.4831 0.1094>, rot=<-0.0654 -0.0035 0.2242 -0.9723>}
 pinky_02_r {parent=pinky_01_r, loc=<-2.8274 -0.0893 0.1043>, rot=<0.0037 0.0011 0.2854 -0.9584>}
 pinky_03_r {parent=pinky_02_r, loc=<-2.2046 -0.0307 0.0078>, rot=<-0.0002 0.0098 0.1445 -0.9895>}
 deform_wrist_dn_base_r {parent=hand_r, loc=<-0.0001 1.5328 0>, rot=<0 0 0 -1>}
 deform_wrist_dn_r {parent=deform_wrist_dn_base_r, loc=<0 0 0>, rot=<0 0 0 1>}
 deform_wrist_up_base_r {parent=hand_r, loc=<-0.0002 -2.3007 -0.0001>, rot=<0 0 0 -1>}
 deform_wrist_up_r {parent=deform_wrist_up_base_r, loc=<0 0 0>, rot=<0 0 0 1>}
 lowerarm_twist_02_r {parent=lowerarm_r, loc=<-10.6707 0 0>, rot=<0 0 0 -1>}
 lowerarm_twist_01_r {parent=lowerarm_r, loc=<-18.316 0 0>, rot=<0 0 0 -1>}
 deform_elbow_in_base_r {parent=lowerarm_r, loc=<-0.1834 -3.6062 0.2572>, rot=<0 0 0 -1>}
 deform_elbow_in_r {parent=deform_elbow_in_base_r, loc=<0 0 0>, rot=<0 0 0 1>}
 deform_elbow_out_base_r {parent=lowerarm_r, loc=<0 4 -0.0001>, rot=<0 0 0 -1>}
 deform_elbow_out_r {parent=deform_elbow_out_base_r, loc=<0 0 0>, rot=<0 0 0 1>}
 upperarm_twist_01_r {parent=upperarm_r, loc=<-7.2135 0.1134 0>, rot=<0 0 0 -1>}
 deform_upperarm_up_base_r {parent=upperarm_twist_01_r, loc=<8.891 -0.4825 -3.9517>, rot=<0 0 0 -1>}
 deform_upperarm_up_r {parent=deform_upperarm_up_base_r, loc=<0 0 0>, rot=<0 0 0 1>}
 base_r_padChain1 {parent=upperarm_twist_01_r, loc=<9.2157 -5.9118 -0.383>, rot=<0.3282 -0.829 0.4359 0.1224>}
 dyn_r_padChain1_01 {parent=base_r_padChain1, loc=<0 0 0>, rot=<0 0 0 1>}
 dyn_r_padChain1_02 {parent=dyn_r_padChain1_01, loc=<3.391 0.0001 0.0002>, rot=<0.0683 -0.1282 -0.0089 -0.9894>}
 base_r_padChain2 {parent=upperarm_twist_01_r, loc=<7.9186 -5.0709 -1.5055>, rot=<0.3086 -0.8282 0.4389 0.162>}
 dyn_r_padChain2_01 {parent=base_r_padChain2, loc=<0 0 0>, rot=<0 0 0 1>}
 dyn_r_padChain2_02 {parent=dyn_r_padChain2_01, loc=<3.393 0 -0.0001>, rot=<0.0677 -0.1064 -0.0073 -0.992>}
 base_r_padChain3 {parent=upperarm_twist_01_r, loc=<6.9214 -4.1285 -2.7651>, rot=<0.2795 -0.8552 0.3977 0.1798>}
 dyn_r_padChain3_01 {parent=base_r_padChain3, loc=<0 0 0>, rot=<0 0 0 1>}
 dyn_r_padChain3_02 {parent=dyn_r_padChain3_01, loc=<3.395 0 0>, rot=<0.0295 -0.0902 -0.0027 -0.9955>}
 base_r_padChain4 {parent=upperarm_twist_01_r, loc=<6.4103 -3.0023 -4.0184>, rot=<0.2698 -0.8847 0.3323 0.1845>}
 dyn_r_padChain4_01 {parent=base_r_padChain4, loc=<0 0 0>, rot=<0 0 0 1>}
 dyn_r_padChain4_02 {parent=dyn_r_padChain4_01, loc=<3.3971 0 -0.0001>, rot=<0 -0.072 0 -0.9974>}
 base_r_padChain5 {parent=upperarm_twist_01_r, loc=<6.397 -2.3831 -4.5814>, rot=<0.2195 -0.9373 0.1756 0.2057>}
 dyn_r_padChain5_01 {parent=base_r_padChain5, loc=<0 0 0>, rot=<0 0 0 1>}
 dyn_r_padChain5_02 {parent=dyn_r_padChain5_01, loc=<3.3966 0 -0.0001>, rot=<-0.1457 -0.0641 0.0095 -0.9872>}
 base_r_padChain6 {parent=upperarm_twist_01_r, loc=<6.3622 -1.772 -4.9587>, rot=<0.1773 -0.9644 0.0529 0.1891>}
 dyn_r_padChain6_01 {parent=base_r_padChain6, loc=<0 0 0>, rot=<0 0 0 -1>}
 dyn_r_padChain6_02 {parent=dyn_r_padChain6_01, loc=<3.3956 0 0>, rot=<-0.2107 -0.0661 0.0143 -0.9752>}
 base_r_padChain7 {parent=upperarm_twist_01_r, loc=<6.3213 -1.1521 -5.2434>, rot=<0.1142 -0.9813 0.0891 0.1269>}
 dyn_r_padChain7_01 {parent=base_r_padChain7, loc=<0 0 0>, rot=<0 0 0 1>}
 dyn_r_padChain7_02 {parent=dyn_r_padChain7_01, loc=<3.3977 0 -0.0001>, rot=<-0.025 -0.0494 0.0012 -0.9985>}
 base_r_padChain8 {parent=upperarm_twist_01_r, loc=<6.305 -0.5629 -5.4059>, rot=<0.0085 -0.9867 -0.1096 0.1199>}
 dyn_r_padChain8_01 {parent=base_r_padChain8, loc=<0 0 0>, rot=<0 0 0 1>}
 dyn_r_padChain8_02 {parent=dyn_r_padChain8_01, loc=<3.3971 -0.0001 0.0002>, rot=<-0.1339 -0.0522 0.0071 -0.9896>}
 base_r_padChain9 {parent=upperarm_twist_01_r, loc=<6.2666 0.1853 -5.4388>, rot=<-0.0251 -0.9885 -0.0769 0.1277>}
 dyn_r_padChain9_01 {parent=base_r_padChain9, loc=<0 0 0>, rot=<0 0 0 1>}
 dyn_r_padChain9_02 {parent=dyn_r_padChain9_01, loc=<3.3981 0 -0.0002>, rot=<0 -0.0405 0 -0.9992>}
 base_r_padChain10 {parent=upperarm_twist_01_r, loc=<6.3114 0.9584 -5.2858>, rot=<-0.1194 -0.9657 -0.1918 0.1281>}
 dyn_r_padChain10_01 {parent=base_r_padChain10, loc=<0 0 0>, rot=<0 0 0 1>}
 dyn_r_padChain10_02 {parent=dyn_r_padChain10_01, loc=<3.3976 0 -0.0002>, rot=<0 -0.045 0 -0.999>}
 base_r_padChain11 {parent=upperarm_twist_01_r, loc=<6.375 1.795 -4.898>, rot=<-0.1741 -0.9401 -0.2537 0.147>}
 dyn_r_padChain11_01 {parent=base_r_padChain11, loc=<0 0 0>, rot=<0 0 0 1>}
 dyn_r_padChain11_02 {parent=dyn_r_padChain11_01, loc=<3.3968 0 0.0002>, rot=<0.1831 -0.0446 -0.0083 -0.9821>}
 base_r_padChain12 {parent=upperarm_twist_01_r, loc=<6.4963 2.5819 -4.277>, rot=<-0.216 -0.8719 -0.4076 0.1642>}
 dyn_r_padChain12_01 {parent=base_r_padChain12, loc=<0 0 0>, rot=<0 0 0 1>}
 dyn_r_padChain12_02 {parent=dyn_r_padChain12_01, loc=<3.3968 0.0001 -0.0002>, rot=<0.0616 -0.0273 -0.0017 -0.9977>}
 base_r_padChain13 {parent=upperarm_twist_01_r, loc=<6.7057 3.1616 -3.5787>, rot=<-0.2422 -0.8196 -0.4833 0.1897>}
 dyn_r_padChain13_01 {parent=base_r_padChain13, loc=<0 0 0>, rot=<0 0 0 1>}
 dyn_r_padChain13_02 {parent=dyn_r_padChain13_01, loc=<3.3912 0.0001 -0.0001>, rot=<0 -0.0454 0 -0.999>}
 base_r_padChain14 {parent=upperarm_twist_01_r, loc=<7.0182 3.5852 -2.7944>, rot=<-0.2566 -0.7681 -0.5629 0.1653>}
 dyn_r_padChain14_01 {parent=base_r_padChain14, loc=<0 0 0>, rot=<0 0 0 1>}
 dyn_r_padChain14_02 {parent=dyn_r_padChain14_01, loc=<3.3901 0 0>, rot=<0 -0.0613 0 -0.9981>}
 base_r_padChain15 {parent=upperarm_twist_01_r, loc=<7.2584 3.8508 -2.1342>, rot=<-0.3019 -0.7329 -0.5728 0.209>}
 dyn_r_padChain15_01 {parent=base_r_padChain15, loc=<0 0 0>, rot=<0 0 0 1>}
 dyn_r_padChain15_02 {parent=dyn_r_padChain15_01, loc=<3.3957 0 -0.0001>, rot=<0 -0.0594 0 -0.9982>}
 base_r_padChain16 {parent=upperarm_twist_01_r, loc=<7.9645 4.1554 -0.9182>, rot=<-0.315 -0.7093 -0.5907 0.2209>}
 dyn_r_padChain16_01 {parent=base_r_padChain16, loc=<0 0 0>, rot=<0 0 0 1>}
 dyn_r_padChain16_02 {parent=dyn_r_padChain16_01, loc=<3.3934 -0.0001 0>, rot=<0 -0.072 0 -0.9974>}
 base_r_padChain17 {parent=upperarm_twist_01_r, loc=<9.1413 4.4209 0.4788>, rot=<-0.3314 -0.6604 -0.629 0.2416>}
 dyn_r_padChain17_01 {parent=base_r_padChain17, loc=<0 0 0>, rot=<0 0 0 1>}
 dyn_r_padChain17_02 {parent=dyn_r_padChain17_01, loc=<3.3919 0 -0.0002>, rot=<0 -0.0578 0 -0.9983>}
 base_r_padChain18 {parent=upperarm_twist_01_r, loc=<10.4069 4.514 1.5885>, rot=<-0.3277 -0.655 -0.6253 0.2693>}
 dyn_r_padChain18_01 {parent=base_r_padChain18, loc=<0 0 0>, rot=<0 0 0 1>}
 dyn_r_padChain18_02 {parent=dyn_r_padChain18_01, loc=<3.3942 0 0>, rot=<0 -0.045 0 -0.999>}
 base_r_padChain1A {parent=upperarm_twist_01_r, loc=<8.6118 -5.5214 -0.8965>, rot=<0.3282 -0.829 0.4359 0.1224>}
 dyn_r_padChain1A_01 {parent=base_r_padChain1A, loc=<0 0 0>, rot=<0 0 0 1>}
 dyn_r_padChain1A_02 {parent=dyn_r_padChain1A_01, loc=<3.391 0.0001 0.0002>, rot=<0.0683 -0.1282 -0.0089 -0.9894>}
 base_r_padChain2A {parent=upperarm_twist_01_r, loc=<7.3469 -4.6071 -2.1265>, rot=<0.2972 -0.8555 0.3829 0.1821>}
 dyn_r_padChain2A_01 {parent=base_r_padChain2A, loc=<0 0 0>, rot=<0 0 0 1>}
 dyn_r_padChain2A_02 {parent=dyn_r_padChain2A_01, loc=<3.393 0 -0.0001>, rot=<0.0677 -0.1064 -0.0073 -0.992>}
 base_r_padChain3A {parent=upperarm_twist_01_r, loc=<6.6115 -3.6037 -3.3516>, rot=<0.2723 -0.87 0.3643 0.1905>}
 dyn_r_padChain3A_01 {parent=base_r_padChain3A, loc=<0 0 0>, rot=<0 0 0 1>}
 dyn_r_padChain3A_02 {parent=dyn_r_padChain3A_01, loc=<3.395 0 0>, rot=<0.0295 -0.0902 -0.0027 -0.9955>}
 upperarm_twist_02_r {parent=upperarm_r, loc=<-19.7998 0.1134 0>, rot=<0 0 0 -1>}
 base_padChain_r_Helper {parent=upperarm_r, loc=<-0.5715 -1.196 -5.3038>, rot=<0.7774 -0.1261 -0.5957 0.1575>}
 base_padChain_r_Helper1 {parent=base_padChain_r_Helper, loc=<0 0 0>, rot=<0 0 0 1>}
 attach_backpack {parent=spine_05, loc=<-0.1199 -13.474 0>, rot=<-0.0092 -0.707 0.0093 -0.707>}
 Backpack_BR {parent=attach_backpack, type=Null, loc=<0 -8.0843 0.09>, rot=<-0.1023 0 0 0.9948>}
 attach_cape {parent=spine_05, loc=<9.0188 -11.7148 0>, rot=<0.0616 -0.7044 -0.0616 -0.7044>}
 cape {parent=attach_cape, type=Null, loc=<0 -9.2915 0>, rot=<0.0838 0 0 0.9965>}
 spine_05_HLP {parent=spine_05, loc=<0 0 0>, rot=<0 0 0 1>, scale=<1 0.4 0.2>}
 base_front_jacket_l {parent=spine_04, loc=<-1.4147 8.1313 -8.5398>, rot=<0.0462 -0.9095 -0.4116 0.0359>}
 dyn_front_jacket_l {parent=base_front_jacket_l, loc=<0 0 0>, rot=<0 0 0 1>}
 base_side_jacket_l {parent=spine_04, loc=<-0.8422 0.1388 -10.6209>, rot=<0.0294 -0.9902 0.1086 0.0829>}
 dyn_side_jacket_l {parent=base_side_jacket_l, loc=<0 0 0>, rot=<0 0 0 1>}
 base_bk_jacket_l {parent=spine_04, loc=<0.5581 -5.9766 -5.3289>, rot=<-0.0057 -0.7615 0.647 0.0375>}
 dyn_bk_jacket_l {parent=base_bk_jacket_l, loc=<0 0 0>, rot=<0 0 0 1>}
 base_front_jacket_r {parent=spine_04, loc=<-1.4152 8.1313 8.5399>, rot=<-0.0359 0.4116 0.9095 -0.0462>}
 dyn_front_jacket_r {parent=base_front_jacket_r, loc=<0 0 0>, rot=<0 0 0 1>}
 base_side_jacket_r {parent=spine_04, loc=<-0.8395 0.1377 10.6209>, rot=<-0.0829 -0.1086 0.9902 -0.0294>}
 dyn_side_jacket_r {parent=base_side_jacket_r, loc=<0 0 0>, rot=<0 0 0 1>}
 base_bk_jacket_r {parent=spine_04, loc=<0.5593 -5.9772 5.3289>, rot=<-0.0375 -0.647 0.7615 0.0057>}
 dyn_bk_jacket_r {parent=base_bk_jacket_r, loc=<0 0 0>, rot=<0 0 0 1>}
 base_Loop_A {parent=spine_01, loc=<0.9723 8.2387 -3.8674>, rot=<-0.4199 -0.4481 -0.4578 -0.6429>}
 base_LoopA_up {parent=base_Loop_A, loc=<-0.0342 0.0052 1.2251>, rot=<0.7104 -0.0106 0.7036 -0.0134>}
 dyn_LoopA_up {parent=base_LoopA_up, loc=<0 0 0>, rot=<0 0 0 1>}
 base_LoopA_dw {parent=base_Loop_A, loc=<-0.0421 0.0927 -1.2321>, rot=<0.0684 -0.7077 -0.0964 -0.6965>}
 dyn_LoopA_dw {parent=base_LoopA_dw, loc=<0 0 0>, rot=<0 0 0 1>}
 base_Loop_B {parent=spine_01, loc=<1.4673 7.6865 -5.7095>, rot=<-0.4227 -0.4581 -0.436 -0.6492>}
 base_LoopB_up {parent=base_Loop_B, loc=<0.0353 0.0404 1.2231>, rot=<0.7096 0.0017 0.7045 -0.0104>}
 dyn_LoopB_up {parent=base_LoopB_up, loc=<0 0 0>, rot=<0 0 0 1>}
 base_LoopB_dw {parent=base_Loop_B, loc=<0.0331 0.074 -1.2355>, rot=<0.0895 -0.706 -0.1026 -0.695>}
 dyn_LoopB_dw {parent=base_LoopB_dw, loc=<0 0 0>, rot=<0 0 0 1>}
 base_Loop_C {parent=spine_01, loc=<2.0196 6.7784 -7.3751>, rot=<-0.3773 -0.4701 -0.4101 -0.6844>}
 base_LoopC_up {parent=base_Loop_C, loc=<0.0518 -0.0241 1.1765>, rot=<0.693 -0.0541 0.7187 0.0153>}
 dyn_LoopC_up {parent=base_LoopC_up, loc=<0 0 0>, rot=<0 0 0 1>}
 base_LoopC_dw {parent=base_Loop_C, loc=<0.1637 0.1082 -1.2761>, rot=<-0.0611 -0.6966 0.0189 -0.7146>}
 dyn_LoopC_dw {parent=base_LoopC_dw, loc=<0 0 0>, rot=<0 0 0 1>}
 thigh_l {parent=pelvis, loc=<-1.0855 -0.2109 -8.7053>, rot=<0.0009 0.0355 0.0155 -0.9993>}
 calf_l {parent=thigh_l, loc=<-42.5284 -0.0019 -0.7297>, rot=<0.0006 -0.0161 -0.0391 -0.9991>}
 foot_l {parent=calf_l, loc=<-46.7401 -2.5014 -0.592>, rot=<0.0427 -0.0238 0.0353 -0.9982>}
 ball_l {parent=foot_l, loc=<-3.5658 10.8066 -0.1544>, rot=<0.0001 -0.0001 -0.7063 -0.7079>}
 calf_twist_02_l {parent=calf_l, loc=<-19.1546 -1.7943 -0.2638>, rot=<0.0237 0.0095 -0.0035 -0.9997>}
 calf_twist_01_l {parent=calf_l, loc=<-35.8483 -1.9197 -0.5794>, rot=<0.0237 0.0095 -0.0035 -0.9997>}
 deform_knee_in_base_l {parent=calf_l, loc=<0.8031 -5.0231 0.2456>, rot=<0 0 0 -1>}
 deform_knee_in_l {parent=deform_knee_in_base_l, loc=<0 0 0>, rot=<0 0 0 1>}
 boot_l_chain_A_01 {parent=calf_l, loc=<-36.2317 -0.1433 -0.2708>, rot=<-0.446 -0.8875 0.1134 -0.0223>}
 boot_l_chain_B_01 {parent=calf_l, loc=<-36.7596 -1.595 -0.1349>, rot=<-0.4311 -0.8992 0.0748 -0.0031>}
 boot_l_chain_B_02 {parent=boot_l_chain_B_01, loc=<3.6537 0.0384 0>, rot=<0 0 -0.0737 -0.9973>}
 boot_l_chain_C_01 {parent=calf_l, loc=<-38.5477 -1.7271 0.0766>, rot=<-0.4264 -0.9006 0.0844 -0.0029>}
 boot_l_chain_C_02 {parent=boot_l_chain_C_01, loc=<4.3568 0 0>, rot=<0 0 0 1>}
 deform_knee_out_base_l {parent=calf_l, loc=<0 4 0>, rot=<0 0 0 -1>}
 deform_knee_out_l {parent=deform_knee_out_base_l, loc=<0 0 0>, rot=<0 0 0 1>}
 thigh_twist_01_l {parent=thigh_l, loc=<-27.952 0 0>, rot=<0.0013 0 0 -1>}
 base_front_skirt_l {parent=thigh_l, loc=<-5.7314 7.3677 0.1275>, rot=<-0.0174 -0.7506 -0.6598 0.0315>}
 dyn_front_skirt_l {parent=base_front_skirt_l, loc=<0 0 0>, rot=<0 0 0 1>}
 base_side_skirt_l {parent=thigh_l, loc=<-5.7368 -0.3931 -5.4951>, rot=<-0.0048 -0.9991 0.0098 -0.042>}
 dyn_side_skirt_l {parent=base_side_skirt_l, loc=<0 0 0>, rot=<0 0 0 1>}
 base_bk_skirt_l {parent=thigh_l, loc=<-5.7551 -9.5644 1.5428>, rot=<0.0178 -0.7672 0.6405 0.0304>}
 dyn_bk_skirt_l {parent=base_bk_skirt_l, loc=<0 0 0>, rot=<0 0 0 1>}
 deform_groin_base_l {parent=thigh_l, loc=<0.4801 5.6776 -0.2486>, rot=<0 0 0 -1>}
 deform_groin_l {parent=deform_groin_base_l, loc=<0 0 0>, rot=<0 0 0 1>}
 thigh_01_l_helper {parent=thigh_l, loc=<-12.2488 -0.2603 0.9861>, rot=<0.0013 0 0 -1>}
 thigh_r {parent=pelvis, loc=<-1.0855 -0.2109 8.7053>, rot=<0.0355 -0.0009 0.9993 0.0155>}
 calf_r {parent=thigh_r, loc=<42.5284 0.0019 0.7297>, rot=<0.0006 -0.0161 -0.0391 -0.9991>}
 foot_r {parent=calf_r, loc=<46.7401 2.5014 0.592>, rot=<0.0427 -0.0238 0.0353 -0.9982>}
 ball_r {parent=foot_r, loc=<3.5658 -10.8066 0.1544>, rot=<0.0001 -0.0001 -0.7063 -0.7079>}
 calf_twist_02_r {parent=calf_r, loc=<19.1546 1.7943 0.2638>, rot=<0.0237 0.0095 -0.0035 -0.9997>}
 calf_twist_01_r {parent=calf_r, loc=<35.8483 1.9197 0.5794>, rot=<0.0237 0.0095 -0.0035 -0.9997>}
 deform_knee_in_base_r {parent=calf_r, loc=<-0.8031 5.0231 -0.2457>, rot=<0 0 0 -1>}
 deform_knee_in_r {parent=deform_knee_in_base_r, loc=<0 0 0>, rot=<0 0 0 1>}
 boot_r_chain_A_02 {parent=calf_r, loc=<36.2317 0.1433 0.2707>, rot=<0.1134 0.0223 0.446 -0.8875>}
 boot_r_chain_B_03 {parent=calf_r, loc=<36.7596 1.5951 0.1349>, rot=<0.0748 0.0031 0.4311 -0.8992>}
 boot_r_chain_B_02 {parent=boot_r_chain_B_03, loc=<3.6536 -0.0385 0>, rot=<0 0 0.0737 -0.9973>}
 boot_r_chain_C_01 {parent=calf_r, loc=<38.5477 1.7271 -0.0766>, rot=<0.0844 0.0029 0.4264 -0.9006>}
 boot_r_chain_C_02 {parent=boot_r_chain_C_01, loc=<4.3568 0 0>, rot=<0 0 0 1>}
 deform_knee_out_base_r {parent=calf_r, loc=<0 -4 0>, rot=<0 0 0 -1>}
 deform_knee_out_r {parent=deform_knee_out_base_r, loc=<0 0 0>, rot=<0 0 0 1>}
 thigh_twist_01_r {parent=thigh_r, loc=<27.952 0 0>, rot=<0.0013 0 0 -1>}
 base_front_skirt_r {parent=thigh_r, loc=<5.7318 -7.3669 -0.1272>, rot=<-0.6598 -0.0315 0.0174 -0.7506>}
 dyn_front_skirt_r {parent=base_front_skirt_r, loc=<0 0 0>, rot=<0 0 0 1>}
 base_bk_skirt_r {parent=thigh_r, loc=<5.7617 9.8859 -1.543>, rot=<0.6405 -0.0304 -0.0178 -0.7672>}
 dyn_bk_skirt_r {parent=base_bk_skirt_r, loc=<0 0 0>, rot=<0 0 0 1>}
 base_side_skirt_r {parent=thigh_r, loc=<5.7367 0.3931 5.4952>, rot=<0.0098 0.042 0.0048 -0.9991>}
 dyn_side_skirt_r {parent=base_side_skirt_r, loc=<0 0 0>, rot=<0 0 0 1>}
 deform_groin_base_r {parent=thigh_r, loc=<-0.4801 -5.6776 0.2486>, rot=<0 0 0 -1>}
 deform_groin_r {parent=deform_groin_base_r, loc=<0 0 0>, rot=<0 0 0 1>}
 thigh_01_r_helper {parent=thigh_r, loc=<12.2488 0.2603 -0.9861>, rot=<0.0013 0 0 -1>}
 Base_WaistBand {parent=pelvis, loc=<3.929 6.9756 8.6595>, rot=<0.9435 -0.143 -0.0623 -0.2924>}
 dyn_band_Knot_01 {parent=Base_WaistBand, loc=<0 0 0>, rot=<0 0 0 1>}
 dyn_band_Knot_02 {parent=dyn_band_Knot_01, loc=<3.5605 0 0.0004>, rot=<0 0.0015 0 -1>}
 base_bandA {parent=Base_WaistBand, loc=<-0.041 0.0045 0.0518>, rot=<-0.0604 -0.7238 -0.6853 0.0536>}
 dyn_bandA_01 {parent=base_bandA, loc=<0 0 0>, rot=<0 0 0 1>}
 dyn_bandA_02 {parent=dyn_bandA_01, loc=<5.5945 0 0>, rot=<-0.0519 0.0811 0.0377 -0.9946>}
 dyn_bandA_03 {parent=dyn_bandA_02, loc=<5.9508 0 0>, rot=<0.0114 0.0338 -0.0017 -0.9994>}
 dyn_bandA_04 {parent=dyn_bandA_03, loc=<5.9667 0 0>, rot=<0.0085 0.0204 0.0075 -0.9997>}
 dyn_bandA_05 {parent=dyn_bandA_04, loc=<7.43 0 0>, rot=<-0.0081 0.0143 -0.0051 -0.9999>}
 base_bandB {parent=Base_WaistBand, loc=<-0.0686 0.031 -0.0307>, rot=<-0.0044 -0.7497 -0.6588 -0.0625>}
 dyn_bandB_01 {parent=base_bandB, loc=<0 0 0>, rot=<0 0 0 1>}
 dyn_bandB_02 {parent=dyn_bandB_01, loc=<5.4644 0 0>, rot=<-0.0367 0.0554 -0.0596 -0.996>}
 dyn_bandB_03 {parent=dyn_bandB_02, loc=<5.9399 0 0>, rot=<-0.0272 0.0199 -0.007 -0.9994>}
 dyn_bandB_04 {parent=dyn_bandB_03, loc=<5.9504 0 0>, rot=<0.0102 0.0108 -0.0009 -0.9999>}
 dyn_bandB_05 {parent=dyn_bandB_04, loc=<7.4871 0 0>, rot=<0.0091 0.0152 -0.0036 -0.9998>}
 deform_glute_base_l {parent=pelvis, loc=<-7.3466 -9.0213 -9.1245>, rot=<0.0172 0.031 -0.4477 -0.8935>}
 deform_glute_l {parent=deform_glute_base_l, loc=<0 0 0>, rot=<0 0 0 1>}
 deform_glute_base_r {parent=pelvis, loc=<-7.3464 -9.0213 9.1245>, rot=<0.031 -0.0172 0.8935 -0.4477>}
 deform_glute_r {parent=deform_glute_base_r, loc=<0 0 0>, rot=<0 0 0 1>}
 base_front_skirt_c {parent=pelvis, loc=<-5.6936 7.8562 0>, rot=<0.0127 0.707 0.707 0.0127>}
 dyn_front_skirt_c {parent=base_front_skirt_c, loc=<0 0 0>, rot=<0 0 0 1>}
 base_bk_skirt_c {parent=pelvis, loc=<-5.9894 -10.3207 0>, rot=<-0.0127 -0.707 0.707 0.0127>}
 dyn_bk_skirt_c {parent=base_bk_skirt_c, loc=<0 0 0>, rot=<0 0 0 1>}
 pelvis_helper {parent=pelvis, loc=<0 0 0>, rot=<0 0 0 1>, scale=<0.5 0.5 0.5>}
 Waistband_Collider_01 {parent=pelvis, loc=<-1.5783 6.3787 9.3132>, rot=<0.598 0.4073 0.5641 0.3979>}
 Waistband_Collider_02 {parent=pelvis, loc=<-7.3886 7.0045 9.6539>, rot=<0.5836 0.3993 0.5689 0.4199>}
 Waistband_Collider_03 {parent=pelvis, loc=<-13.4118 7.0581 9.7661>, rot=<0.5832 0.3998 0.5685 0.4205>}
 Waistband_Collider_04 {parent=pelvis, loc=<-7.3905 7.7571 8.7049>, rot=<0.5493 0.4453 0.533 0.4647>}
 thigh_l_null {parent=pelvis, type=Null, loc=<-1.0855 -0.2109 -8.7053>, rot=<0.0009 0.0355 0.0155 -0.9993>}
 thigh_r_null {parent=pelvis, type=Null, loc=<-1.0855 -0.2109 8.7053>, rot=<0.0355 -0.0009 0.9993 0.0155>}
]
variables [
 SMG_P90 {type=bool, default=False}
 is OutLaw_SG {type=bool, default=False}
 isAR Equipped {type=bool, default=False}
 isAR Equipped lerp {type=double, default=0}
 isBalletSpin {type=double, default=0}
 isBaseball_Sliding {type=double, default=0}
 isBreakDancing {type=double, default=0}
 isCrouching {type=double, default=0}
 isCrouching_NotSprinting {type=double, default=0}
 isDBNO {type=bool, default=False}
 isFrontEnd {type=bool, default=False}
 isJumping {type=double, default=0}
 isLobbyIddle {type=double, default=0}
 isMoving_Back {type=double, default=0}
 isMoving_Forward {type=double, default=0}
 isMoving_Left {type=double, default=0}
 isMoving_Right {type=double, default=0}
 isPA Swing lerp {type=double, default=0}
 isPistol Equipped lerp {type=double, default=0}
 isPlaying emote {type=bool, default=False}
 isRegular_Sliding {type=double, default=0}
 isTactical Sprinting {type=double, default=0}
 isTargeting lerp {type=double, default=0}
 is_Celebrating {type=double, default=0}
 is_Parachuting {type=double, default=0}
 is_SkyFloating {type=double, default=0}
 pub PA Swing {type=bool, default=False}
 pub is Targeting {type=bool, default=False}
 "pub isPistol Equipped " {type=bool, default=False}
]
