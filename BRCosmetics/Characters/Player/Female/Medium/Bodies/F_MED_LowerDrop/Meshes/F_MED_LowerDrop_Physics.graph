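# Physics asset for the F_MED_LowerDrop female player body cosmetic, simulating two hanging band chains, A and B.
# Each five-capsule chain hangs from a small base body, with tight twist and wider, asymmetric swing limits that generally loosen down the chain.
# Waistband, pelvis, spine and right-side arm and leg capsules provide collision shapes, with arm constraints left free in rotation.
# An oversized root capsule, topping out near the origin, approximates a ground plane.
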PhysicsAsset /BRCosmetics/Characters/Player/Female/Medium/Bodies/F_MED_LowerDrop/Meshes/F_MED_LowerDrop_Physics
bodies {
 Base_WaistBand [{
   shapeType=Capsule
   length=1.6375
   radius=1.1962
   shapeName=Base_WaistBand_capsule
   center=<0.2342 -0.0628 -0.0914>
   rotation=<rot 1.3271 -90 -3.1808>
  }]
 Waistband_Collider_01 [{
   shapeType=Capsule
   length=10
   radius=3.4
   shapeName=Waistband_Collider_01_capsule
   center=<-3.6156 -2.0045 0.4116>
   rotation=<rot 87.664 90.0031 -176.8559>
  }]
 Waistband_Collider_02 [{
   shapeType=Capsule
   length=12
   radius=3.4
   shapeName=Waistband_Collider_02_capsule
   center=<-2.8716 -3.1027 -0.6701>
   rotation=<rot 90 -93.9999 0>
  }]
 Waistband_Collider_03 [{
   shapeType=Capsule
   length=13
   radius=3.65
   shapeName=Waistband_Collider_03_capsule
   center=<-2.864 -3.3591 -1.9947>
   rotation=<rot 90 -93.1036 0>
  }]
 Waistband_Collider_04 [{
   shapeType=Capsule
   length=10
   radius=3.4
   shapeName=Waistband_Collider_02_capsule
   center=<-2.7952 -2.0111 -0.6701>
   rotation=<rot 90 -93.9999 0>
  }]
 base_bandA [{shapeType=Capsule, length=0.505, radius=0.505, shapeName=base_bandA_capsule, rotation=<rot 0 -90 0>}]
 base_bandB [{shapeType=Capsule, length=0.505, radius=0.505, shapeName=base_bandB_capsule, rotation=<rot 0 -90 0>}]
 calf_r [{
   shapeType=Capsule
   length=28.3001
   radius=4.2739
   shapeName=calf_r_capsule
   center=<13.8447 1.5267 0.1995>
   rotation=<rot 0.0245 91.3463 -89.9101>
  }]
 dyn_bandA_01 [{shapeType=Capsule, length=5, radius=1, shapeName=dyn_bandA_01_capsule, center=<4.0894 0 -0.1394>, rotation=<rot 87.1858 180 179.9999>}]
 dyn_bandA_02 [{shapeType=Capsule, length=6, radius=1.4, shapeName=dyn_bandA_02_capsule, center=<2.8 0 0.5>, rotation=<rot 87.617 179.9998 179.9997>}]
 dyn_bandA_03 [{shapeType=Capsule, length=6, radius=1.8, shapeName=dyn_bandA_03_capsule, center=<2.8 0 0.65>, rotation=<rot 90 -180 180>}]
 dyn_bandA_04 [{shapeType=Capsule, length=5.5, radius=2, shapeName=dyn_bandA_04_capsule, center=<2.8 0.3 1>, rotation=<rot 90 0 0>}]
 dyn_bandA_05 [{shapeType=Capsule, length=5, radius=2, shapeName=dyn_bandA_05_capsule, center=<2.8 0 0.5>, rotation=<rot 90 0 0>}]
 dyn_bandB_01 [{shapeType=Capsule, length=5, radius=1, shapeName=dyn_bandB_01_capsule, center=<2.8 0 -0.5>, rotation=<rot 90 0 0>}]
 dyn_bandB_02 [
  {shapeType=Capsule, length=5.5, radius=1.75, shapeName=dyn_bandB_02_capsule, center=<2.8 0.4 0>, rotation=<rot 90 0 0>}
  {shapeType=Capsule, length=4.6784, radius=1.2922, shapeName=dyn_bandB_02_capsule, center=<2.4818 0.3947 -2.2896>, rotation=<rot 90 0 0>}
 ]
 dyn_bandB_03 [{shapeType=Capsule, length=6, radius=1.6, shapeName=dyn_bandB_03_capsule, center=<2.8 0 0.25>, rotation=<rot 90 0 0>}]
 dyn_bandB_04 [{shapeType=Capsule, length=5.5, radius=1.8, shapeName=dyn_bandB_04_capsule, center=<2.8 0 0>, rotation=<rot 90 0 0>}]
 dyn_bandB_05 [{shapeType=Capsule, length=5, radius=2, shapeName=dyn_bandB_05_capsule, center=<2.8 0 0>, rotation=<rot 90 0 0>}]
 lowerarm_r [{shapeType=Capsule, length=18.5, radius=4.3, shapeName=lowerarm_r_capsule, center=<-10.3763 -0.5692 0>, rotation=<rot 90 0 0>}]
 pelvis [{shapeType=Capsule, length=11.5, radius=8.5, center=<0.1052 -1.2774 0.0991>, rotation=<rot 1.8372 -90 -1.1266>}]
 root [{shapeType=Capsule, length=5, radius=350, shapeName=root_capsule, center=<0 0 -348.4153>, rotation=<rot 90 -90 0>}]
 spine_01 [{shapeType=Capsule, length=0.505, radius=0.505, shapeName=spine_01_capsule, rotation=<rot 0 -90 0>}]
 spine_02 [{
   shapeType=Capsule
   length=8.2759
   radius=6.0962
   shapeName=spine_02_capsule
   center=<2.0568 0.9818 0.0422>
   rotation=<rot 1.1251 -90 -0.3192>
  }]
 thigh_r [{shapeType=Capsule, length=22, radius=6, shapeName=thigh_r_capsule, center=<10.8634 -0.8854 -1.0834>, rotation=<rot -0.1563 -88.1398 92.0082>}]
 thigh_twist_01_r [{shapeType=Capsule, length=15, radius=5.25, shapeName=thigh_r_capsule, center=<0.9139 -0.3911 -0.7032>, rotation=<rot 90 3.9375 0>}]
 upperarm_r [{
   shapeType=Capsule
   length=22
   radius=4.25
   shapeName=upperarm_r_capsule
   center=<-12.1793 0 0.6539>
   rotation=<rot 85.5125 -179.9993 -179.9994>
  }]
}
constraints [
 {bone1Name=dyn_bandA_01, bone2Name=base_bandA, swing1LimitDegrees=10, swing2LimitDegrees=20, twistLimitDegrees=2}
 {bone1Name=dyn_bandA_02, bone2Name=dyn_bandA_01, swing1LimitDegrees=8, swing2LimitDegrees=30, twistLimitDegrees=2}
 {bone1Name=dyn_bandA_03, bone2Name=dyn_bandA_02, swing1LimitDegrees=15, swing2LimitDegrees=40, twistLimitDegrees=2}
 {bone1Name=dyn_bandA_04, bone2Name=dyn_bandA_03, swing1LimitDegrees=18, swing2LimitDegrees=47, twistLimitDegrees=2}
 {bone1Name=dyn_bandA_05, bone2Name=dyn_bandA_04, swing1LimitDegrees=15, swing2LimitDegrees=42, twistLimitDegrees=2}
 {bone1Name=dyn_bandB_01, bone2Name=base_bandB, swing1LimitDegrees=10, swing2LimitDegrees=25, twistLimitDegrees=2}
 {bone1Name=dyn_bandB_02, bone2Name=dyn_bandB_01, swing1LimitDegrees=10, swing2LimitDegrees=30, twistLimitDegrees=2}
 {bone1Name=dyn_bandB_03, bone2Name=dyn_bandB_02, swing1LimitDegrees=15, swing2LimitDegrees=47, twistLimitDegrees=2}
 {bone1Name=dyn_bandB_04, bone2Name=dyn_bandB_03, swing1LimitDegrees=15, swing2LimitDegrees=47, twistLimitDegrees=2}
 {bone1Name=dyn_bandB_05, bone2Name=dyn_bandB_04, swing1LimitDegrees=15, swing2LimitDegrees=42, twistLimitDegrees=2}
 {
  bone1Name=upperarm_r
  bone2Name=spine_02
  swing1LimitDegrees=45
  swing1Motion=Free
  swing2LimitDegrees=45
  swing2Motion=Free
  twistLimitDegrees=45
  twistMotion=Free
 }
 {
  bone1Name=lowerarm_r
  bone2Name=upperarm_r
  swing1LimitDegrees=45
  swing1Motion=Free
  swing2LimitDegrees=45
  swing2Motion=Free
  twistLimitDegrees=45
  twistMotion=Free
 }
]
