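# Post-style Anim Blueprint for the F_MED_LowerDrop female body cosmetic, adding secondary motion to skirt, jacket, straps and chains.
# Copies the attached parent's pose into a Control Rig driven by movement, weapon and emote states.
# Constraints pin boot chains to calf bones, then cached stages of spring-limited Anim Dynamics simulate dangling parts.
# A thigh-relative RigidBody pass and additive bone offsets for skirt, chains and waistband colliders finish the pose.
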AnimBlueprint /BRCosmetics/Characters/Player/Female/Medium/Bodies/F_MED_LowerDrop/Meshes/F_MED_LowerDrop_AnimBP
target_skeleton=&/F_MED_LowerDrop_Skeleton
graphs [
 AnimGraph {nodes [
   n1 = Anim Dynamics - Bone: dyn_band_Knot_01(ComponentPose=Local To Component(LocalPose=Use cached pose 'Second_Anim_Dynamics'()), Alpha=1, ExternalForce="0,0,0") {
    angularDampingOverride=1
    angularSpringConstant=12000
    bAngularSpring=true
    bOverrideAngularDamping=true
    bUseGravityOverride=true
    gravityScale=0.35
    simulationSpace=RootRelative
    boundBone {boneName=dyn_band_Knot_01}
    componentLinearAccScale=<1 1 1>
    componentLinearVelScale=<1 1 1>
    gravityOverride {z=-980}
    physicsBodyDefinitions [{
      boundBone {boneName=dyn_band_Knot_01}
      boxExtents=<15 25 25>
      constraintSetup {angularLimitsMax=<0 3 -5>, angularLimitsMin=<0 -3 -20>, angularTarget=<1 0 0>}
      localJointOffset=<4 0 0>
     }]
   }
   n2 = Anim Dynamics - Bone: dyn_band_Knot_02(ComponentPose=@n1, Alpha=1, ExternalForce="0,0,0") {
    angularDampingOverride=1
    angularSpringConstant=12000
    bAngularSpring=true
    bOverrideAngularDamping=true
    bUseGravityOverride=true
    gravityScale=0.35
    simulationSpace=RootRelative
    boundBone {boneName=dyn_band_Knot_02}
    componentLinearAccScale=<1 1 1>
    componentLinearVelScale=<1 1 1>
    gravityOverride {z=-980}
    physicsBodyDefinitions [{
      boundBone {boneName=dyn_band_Knot_02}
      boxExtents=<15 25 25>
      constraintSetup {angularLimitsMax=<0 3 -3>, angularLimitsMin=<0 -3 -15>, angularTarget=<1 0 0>}
      localJointOffset=<4 0 0>
     }]
   }
   n3 = RigidBody!(
    ComponentPose=@n2
    Alpha
    OverrideWorldGravity:Vector="0,0,0"
    ExternalForce
    SimSpaceSettings:Sim Space Settings Structure=(WorldAlpha=0.8,VelocityScaleZ=0.475,DampingAlpha=0,MaxLinearVelocity=500,MaxAngularVelocity=8,MaxLinearAcceleration=2500,MaxAngularAcceleration=30,ExternalLinearDragV=(X=0.3,Y=0.3,Z=0.3),ExternalLinearVelocity=(X=0,Y=0,Z=0),ExternalAngularVelocity=(X=0,Y=0,Z=0))
   ) -> (Pose) {
    bOverrideWorldGravity=true
    overridePhysicsAsset=&/F_MED_LowerDrop_Physics
    simulationSpace=BaseBoneSpace
    baseBoneRef {boneName=thigh_r}
    simSpaceSettings {
     dampingAlpha=0
     maxAngularAcceleration=30
     maxAngularVelocity=8
     maxLinearAcceleration=2500
     maxLinearVelocity=500
     velocityScaleZ=0.475
     worldAlpha=0.8
     externalLinearDragV=<0.3 0.3 0.3>
    }
   }
   n4 = "Transform (Modify) Bone - Bone: base_bk_skirt_l"!(ComponentPose=@n3, bAlphaBoolEnabled:bool=True, Translation="0,0,0", Rotation="20,0,0", Scale="1,1,1") -> (Pose) {alphaInputType=Bool, rotationMode=BMM_Additive, rotationSpace=BCS_BoneSpace, boneToModify {boneName=base_bk_skirt_l}, rotation {pitch=20}}
   n5 = "Transform (Modify) Bone - Bone: base_bk_skirt_r"!(ComponentPose=@n4, bAlphaBoolEnabled:bool=True, Translation="0,0,0", Rotation="20,0,0", Scale="1,1,1") -> (Pose) {alphaInputType=Bool, rotationMode=BMM_Additive, rotationSpace=BCS_BoneSpace, boneToModify {boneName=base_bk_skirt_r}, rotation {pitch=20}}
   n6 = "Transform (Modify) Bone - Bone: boot_l_chain_C_01"!(ComponentPose=@n5, bAlphaBoolEnabled:bool=True, Translation="0,0,0", Rotation="0,8,0", Scale="1,1,1") -> (Pose) {alphaInputType=Bool, rotationMode=BMM_Additive, rotationSpace=BCS_BoneSpace, boneToModify {boneName=boot_l_chain_C_01}, rotation {yaw=8}}
   n7 = "Transform (Modify) Bone - Bone: boot_r_chain_C_01"!(ComponentPose=@n6, bAlphaBoolEnabled:bool=True, Translation="0,0,0", Rotation="0,-9,0", Scale="1,1,1") -> (Pose) {
    alphaInputType=Bool
    rotationMode=BMM_Additive
    rotationSpace=BCS_BoneSpace
    translationMode=BMM_Additive
    translationSpace=BCS_BoneSpace
    boneToModify {boneName=boot_r_chain_C_01}
    rotation {yaw=-9}
   }
   n8 = "Transform (Modify) Bone - Bone: base_bk_skirt_l"!(ComponentPose=@n7, Alpha:float=1, Translation="0,0,0", Rotation="15,0,-0.001", Scale="1,1,1") -> (Pose) {
    rotationMode=BMM_Additive
    rotationSpace=BCS_BoneSpace
    translationMode=BMM_Additive
    translationSpace=BCS_BoneSpace
    alphaScaleBiasClamp {bInterpResult=true, interpSpeedDecreasing=5, interpSpeedIncreasing=5}
    boneToModify {boneName=base_bk_skirt_l}
    rotation {pitch=15, roll=-0.001}
   }
   n9 = "Transform (Modify) Bone - Bone: base_bk_skirt_r"!(ComponentPose=@n8, Alpha:float=1, Translation="0,0,0", Rotation="15,0,-0.003", Scale="1,1,1") -> (Pose) {
    rotationMode=BMM_Additive
    rotationSpace=BCS_BoneSpace
    translationMode=BMM_Additive
    translationSpace=BCS_BoneSpace
    alphaScaleBiasClamp {bInterpResult=true, interpSpeedDecreasing=5, interpSpeedIncreasing=5}
    boneToModify {boneName=base_bk_skirt_r}
    rotation {pitch=15, roll=-0.003}
   }
   n10 = "Transform (Modify) Bone - Bone: base_bk_skirt_c"(ComponentPose=@n9, Alpha=1, Translation="0,0,0", Rotation="15,0.028,-0.006", Scale="1,1,1") {
    rotationMode=BMM_Additive
    rotationSpace=BCS_BoneSpace
    translationMode=BMM_Additive
    translationSpace=BCS_BoneSpace
    alphaScaleBiasClamp {bInterpResult=true, interpSpeedDecreasing=5, interpSpeedIncreasing=5}
    boneToModify {boneName=base_bk_skirt_c}
    rotation=<rot 15 0.028 -0.006>
   }
   n11 = "Transform (Modify) Bone - Bone: base_front_skirt_c"(
    ComponentPose="Transform (Modify) Bone - Bone: pelvis_helper"(ComponentPose=@n10, Alpha=1, Translation="0,0,0", Rotation="0,0,0", Scale="1,0.3,0.6") {scaleMode=BMM_Additive, scaleSpace=BCS_BoneSpace, boneToModify {boneName=pelvis_helper}, scale {y=0.3, z=0.6}}
    Alpha=1
    Translation="0,0,0"
    Rotation="-15,0,0"
    Scale="1,1,1"
   ) {
    rotationMode=BMM_Additive
    rotationSpace=BCS_BoneSpace
    translationMode=BMM_Additive
    translationSpace=BCS_BoneSpace
    alphaScaleBiasClamp {bInterpResult=true, interpSpeedDecreasing=5, interpSpeedIncreasing=5}
    boneToModify {boneName=base_front_skirt_c}
    rotation {pitch=-15}
   }
   n12 = "Transform (Modify) Bone - Bone: Waistband_Collider_02"(ComponentPose=@n11, Alpha=1, Translation="1.2,0,0", Rotation="0,0,0", Scale="1,1,1") {
    rotationMode=BMM_Additive
    rotationSpace=BCS_BoneSpace
    translationMode=BMM_Additive
    translationSpace=BCS_BoneSpace
    alphaScaleBiasClamp {bInterpResult=true, interpSpeedDecreasing=5}
    boneToModify {boneName=Waistband_Collider_02}
    translation {x=1.2}
   }
   n13 = "Transform (Modify) Bone - Bone: Waistband_Collider_03"(ComponentPose=@n12, Alpha=1, Translation="1.5,0,0", Rotation="0,0,0", Scale="1,1,1") {
    rotationMode=BMM_Additive
    rotationSpace=BCS_BoneSpace
    translationMode=BMM_Additive
    translationSpace=BCS_BoneSpace
    alphaScaleBiasClamp {bInterpResult=true, interpSpeedDecreasing=5}
    boneToModify {boneName=Waistband_Collider_03}
    translation {x=1.5}
   }
   n14 = "Transform (Modify) Bone - Bone: base_front_skirt_r"(ComponentPose=@n13, Alpha=1, Translation="-0.001,0.157,-1.25", Rotation="0,0,0", Scale="1,1,1") {
    rotationMode=BMM_Additive
    rotationSpace=BCS_BoneSpace
    translationMode=BMM_Additive
    translationSpace=BCS_BoneSpace
    alphaScaleBiasClamp {bInterpResult=true, interpSpeedDecreasing=5}
    boneToModify {boneName=base_front_skirt_r}
    translation=<-0.001 0.157 -1.25>
   }
   n15 = "Transform (Modify) Bone - Bone: base_side_skirt_l"(ComponentPose=@n14, Alpha=1, Translation="0,0,-2", Rotation="0,0,0", Scale="1,1,1") {
    rotationMode=BMM_Additive
    rotationSpace=BCS_BoneSpace
    translationMode=BMM_Additive
    translationSpace=BCS_BoneSpace
    alphaScaleBiasClamp {interpSpeedDecreasing=5, interpSpeedIncreasing=5}
    boneToModify {boneName=base_side_skirt_l}
    translation {z=-2}
   }
   Output Pose(Result=Component To Local(ComponentPose=@n15)) {layerGroup=None}
   n16 = Control Rig!(
    Source=Copy Pose From Mesh() {bUseAttachedParent=true}
    Alpha
    pub is Targeting:bool=Get bIsTargeting()
    "pub isPistol Equipped ":bool=Get isPistol Equipped() -> (isPistol Equipped:bool)
    isPlaying emote:bool=Get bIsPlayingEmote()
    isAR Equipped:bool=Get isAR Equipped() -> (isAR Equipped:bool)
    is OutLaw_SG:bool=Get Is_outLaw_SG() -> (Is_outLaw_SG:bool)
    SMG_P90:bool=Get SMG_SG() -> (SMG_SG:bool)
    isCrouching_NotSprinting:double=Get Crouching_not_Sprin() -> (Crouching_not_Sprin:double)
    is_Parachuting:double=Get Is_Parachute_Open() -> (Is_Parachute_Open:double)
    is_Celebrating:double=Get isCelebrating() -> (isCelebrating:double)
    isMoving_Left:double=Get strafeLeft float() -> (strafeLeft float:double)
    isMoving_Right:double=Get strafeRight float() -> (strafeRight float:double)
    isMoving_Forward:double=Get moveForward() -> (moveForward:double)
    isCrouching:double=Get Crouching() -> (Crouching:double)
    isTactical Sprinting:double=Get isTacticalSprinting() -> (isTacticalSprinting:double)
    isJumping:double=Get Is Jumping() -> (Is Jumping:double)
    isBaseball_Sliding:double=Get IsBaseball_Sliding() -> (IsBaseball_Sliding:double)
    is_SkyFloating:double=Get Is_skyFloating() -> (Is_skyFloating:double)
    isMoving_Back:double=Get moveBackward() -> (moveBackward:double)
    isRegular_Sliding:double=Get isRegular_Sliding() -> (isRegular_Sliding:double)
    isBreakDancing:double=Get is BreakDancing() -> (is BreakDancing:double)
    isDBNO:bool=Get bIsDBNO()
    isLobbyIddle:double=Get Lobby Iddle() -> (Lobby Iddle:double)
    isFrontEnd:bool=Get bIsInFrontEnd()
    isBalletSpin:double=Get IsBallet_Spin() -> (IsBallet_Spin:double)
   ) -> (Pose) {controlRigAssetReference {blueprintRigClass=&/F_MED_LowerDrop_CtrlRig.F_MED_LowerDrop_CtrlRig_C}}
   Save cached pose 'Initial Ctrl Rig Cache'(Pose=@n16) {cacheName=Initial Ctrl Rig Cache}
   n17 = "Constraint - boot_l_chain_A_01 (PPP)"!(
    ComponentPose=Local To Component(LocalPose=Use cached pose 'Initial Ctrl Rig Cache'())
    Alpha=1
    ConstraintWeights_0=0.7001
    ConstraintWeights_1:float=0.2119
    ConstraintWeights_2:float=0.088
   ) -> (Pose) {
    boneToModify {boneName=boot_l_chain_A_01}
    constraintSetup [{targetBone {boneName=calf_twist_01_l}}, {targetBone {boneName=calf_twist_02_l}}, {targetBone {boneName=calf_l}}]
    constraintWeights [1,1,1]
   }
   n18 = "Constraint - boot_l_chain_B_01 (PPP)"!(ComponentPose=@n17, Alpha=1, ConstraintWeights_0=0.7135, ConstraintWeights_1:float=0.2035, ConstraintWeights_2:float=0.0829) -> (Pose) {
    boneToModify {boneName=boot_l_chain_B_01}
    constraintSetup [{targetBone {boneName=calf_twist_01_l}}, {targetBone {boneName=calf_twist_02_l}}, {targetBone {boneName=calf_l}}]
    constraintWeights [1,1,1]
   }
   n19 = "Constraint - boot_l_chain_C_01 (P)"!(ComponentPose=@n18, Alpha=1, ConstraintWeights_0=1, ConstraintWeights_1:float=0) -> (Pose) {boneToModify {boneName=boot_l_chain_C_01}, constraintSetup [{targetBone {boneName=calf_twist_01_l}}], constraintWeights [1,0]}
   n20 = "Constraint - boot_r_chain_A_02 (PPP)"!(ComponentPose=@n19, Alpha=1, ConstraintWeights_0=0.7001, ConstraintWeights_1:float=0.2119, ConstraintWeights_2:float=0.088) -> (Pose) {
    boneToModify {boneName=boot_r_chain_A_02}
    constraintSetup [{targetBone {boneName=calf_twist_01_r}}, {targetBone {boneName=calf_twist_02_r}}, {targetBone {boneName=calf_r}}]
    constraintWeights [1,1,1]
   }
   n21 = "Constraint - boot_r_chain_B_03 (PPP)"!(ComponentPose=@n20, Alpha=1, ConstraintWeights_0=0.7135, ConstraintWeights_1:float=0.2035, ConstraintWeights_2:float=0.0829) -> (Pose) {
    boneToModify {boneName=boot_r_chain_B_03}
    constraintSetup [{targetBone {boneName=calf_twist_01_r}}, {targetBone {boneName=calf_twist_02_r}}, {targetBone {boneName=calf_r}}]
    constraintWeights [1,1,1]
   }
   n22 = "Constraint - boot_r_chain_C_01 (P)"!(ComponentPose=@n21, Alpha=1, ConstraintWeights_0=1, ConstraintWeights_1:float=0) -> (Pose) {boneToModify {boneName=boot_r_chain_C_01}, constraintSetup [{targetBone {boneName=calf_twist_01_r}}], constraintWeights [1,0]}
   n23 = Anim Dynamics - Bone: dyn_front_skirt_l(ComponentPose=@n22, Alpha=1, ExternalForce="0,0,0") {
    angularDampingOverride=1
    angularSpringConstant=10000
    bAngularSpring=true
    bOverrideAngularDamping=true
    bUseGravityOverride=true
    gravityScale=0.35
    simulationSpace=RootRelative
    boundBone {boneName=dyn_front_skirt_l}
    componentAppliedLinearAccClamp=<4000 4000 4000>
    componentLinearAccScale=<0.4 0.4 0.45>
    componentLinearVelScale=<0.4 0.4 0.45>
    gravityOverride {z=-980}
    physicsBodyDefinitions [{
      boundBone {boneName=dyn_front_skirt_l}
      boxExtents=<20 30 30>
      constraintSetup {angularLimitsMax=<0 5 5>, angularLimitsMin=<0 -20 -5>, angularTarget=<1 0 0>}
      localJointOffset=<4 0 0>
     }]
    relativeSpaceBone {boneName=pelvis}
    simSpaceSettings {maxAngularAcceleration=100, maxAngularVelocity=10, simSpaceAngularAlpha=1}
   }
   n24 = Anim Dynamics - Bone: dyn_side_skirt_l(ComponentPose=@n23, Alpha=1, ExternalForce="0,0,0") {
    angularDampingOverride=1
    angularSpringConstant=10000
    bAngularSpring=true
    bOverrideAngularDamping=true
    bUseGravityOverride=true
    gravityScale=0.35
    simulationSpace=RootRelative
    boundBone {boneName=dyn_side_skirt_l}
    componentAppliedLinearAccClamp=<4000 4000 4000>
    componentLinearAccScale=<0.4 0.4 0.45>
    componentLinearVelScale=<0.4 0.4 0.45>
    gravityOverride {z=-980}
    physicsBodyDefinitions [{
      boundBone {boneName=dyn_side_skirt_l}
      boxExtents=<20 30 30>
      constraintSetup {angularLimitsMax=<0 5 5>, angularLimitsMin=<0 -20 -5>, angularTarget=<1 0 0>}
      localJointOffset=<4 0 0>
     }]
    relativeSpaceBone {boneName=pelvis}
    simSpaceSettings {maxAngularAcceleration=100, maxAngularVelocity=10, simSpaceAngularAlpha=1}
   }
   n25 = Anim Dynamics - Bone: dyn_bk_skirt_l(ComponentPose=@n24, Alpha=1, ExternalForce="0,0,0") {
    angularDampingOverride=1
    angularSpringConstant=10000
    bAngularSpring=true
    bOverrideAngularDamping=true
    bUseGravityOverride=true
    gravityScale=0.35
    simulationSpace=RootRelative
    boundBone {boneName=dyn_bk_skirt_l}
    componentAppliedLinearAccClamp=<4000 4000 4000>
    componentLinearAccScale=<0.4 0.4 0.45>
    componentLinearVelScale=<0.4 0.4 0.45>
    gravityOverride {z=-980}
    physicsBodyDefinitions [{
      boundBone {boneName=dyn_bk_skirt_l}
      boxExtents=<20 30 30>
      constraintSetup {angularLimitsMax=<0 15 5>, angularLimitsMin=<0 -20 -5>, angularTarget=<1 0 0>}
      localJointOffset=<4 0 0>
     }]
    relativeSpaceBone {boneName=pelvis}
    simSpaceSettings {maxAngularAcceleration=100, maxAngularVelocity=10, simSpaceAngularAlpha=1}
   }
   n26 = Anim Dynamics - Bone: dyn_front_skirt_r(ComponentPose=@n25, Alpha=1, ExternalForce="0,0,0") {
    angularDampingOverride=1
    angularSpringConstant=10000
    bAngularSpring=true
    bOverrideAngularDamping=true
    bUseGravityOverride=true
    gravityScale=0.35
    simulationSpace=RootRelative
    boundBone {boneName=dyn_front_skirt_r}
    componentAppliedLinearAccClamp=<4000 4000 4000>
    componentLinearAccScale=<0.4 0.4 0.45>
    componentLinearVelScale=<0.4 0.4 0.45>
    gravityOverride {z=-980}
    physicsBodyDefinitions [{
      boundBone {boneName=dyn_front_skirt_r}
      boxExtents=<20 30 30>
      constraintSetup {angularLimitsMax=<0 4 5>, angularLimitsMin=<0 -10 -5>, angularTarget=<1 0 0>}
      localJointOffset=<4 0 0>
     }]
    relativeSpaceBone {boneName=pelvis}
    simSpaceSettings {maxAngularAcceleration=100, maxAngularVelocity=10, simSpaceAngularAlpha=1}
   }
   n27 = Anim Dynamics - Bone: dyn_side_skirt_r(ComponentPose=@n26, Alpha=1, ExternalForce="0,0,0") {
    angularDampingOverride=1
    angularSpringConstant=10000
    bAngularSpring=true
    bOverrideAngularDamping=true
    bUseGravityOverride=true
    gravityScale=0.35
    simulationSpace=RootRelative
    boundBone {boneName=dyn_side_skirt_r}
    componentAppliedLinearAccClamp=<4000 4000 4000>
    componentLinearAccScale=<0.4 0.4 0.45>
    componentLinearVelScale=<0.4 0.4 0.45>
    gravityOverride {z=-980}
    physicsBodyDefinitions [{
      boundBone {boneName=dyn_side_skirt_r}
      boxExtents=<20 30 30>
      constraintSetup {angularLimitsMax=<0 5 0>, angularLimitsMin=<0 -20 -5>, angularTarget=<1 0 0>}
      localJointOffset=<4 0 0>
     }]
    relativeSpaceBone {boneName=pelvis}
    simSpaceSettings {maxAngularAcceleration=100, maxAngularVelocity=10, simSpaceAngularAlpha=1}
   }
   n28 = Anim Dynamics - Bone: dyn_bk_skirt_r(ComponentPose=@n27, Alpha=1, ExternalForce="0,0,0") {
    angularDampingOverride=1
    angularSpringConstant=10000
    bAngularSpring=true
    bOverrideAngularDamping=true
    bUseGravityOverride=true
    gravityScale=0.35
    simulationSpace=RootRelative
    boundBone {boneName=dyn_bk_skirt_r}
    componentAppliedLinearAccClamp=<4000 4000 4000>
    componentLinearAccScale=<0.4 0.4 0.45>
    componentLinearVelScale=<0.4 0.4 0.45>
    gravityOverride {z=-980}
    physicsBodyDefinitions [{
      boundBone {boneName=dyn_bk_skirt_r}
      boxExtents=<20 30 30>
      constraintSetup {angularLimitsMax=<0 15 5>, angularLimitsMin=<0 -20 -5>, angularTarget=<1 0 0>}
      localJointOffset=<4 0 0>
     }]
    relativeSpaceBone {boneName=pelvis}
    simSpaceSettings {maxAngularAcceleration=100, maxAngularVelocity=10, simSpaceAngularAlpha=1}
   }
   n29 = Anim Dynamics - Bone: dyn_front_skirt_c(ComponentPose=@n28, Alpha=1, ExternalForce="0,0,0") {
    angularDampingOverride=1
    angularSpringConstant=10000
    bAngularSpring=true
    bOverrideAngularDamping=true
    bUseGravityOverride=true
    gravityScale=0.35
    simulationSpace=RootRelative
    boundBone {boneName=dyn_front_skirt_c}
    componentAppliedLinearAccClamp=<4000 4000 4000>
    componentLinearAccScale=<0.4 0.4 0.45>
    componentLinearVelScale=<0.4 0.4 0.45>
    gravityOverride {z=-980}
    physicsBodyDefinitions [{
      boundBone {boneName=dyn_front_skirt_c}
      boxExtents=<20 30 30>
      constraintSetup {angularLimitsMax=<0 15 5>, angularLimitsMin=<0 -20 -5>, angularTarget=<1 0 0>}
      localJointOffset=<4 0 0>
     }]
    relativeSpaceBone {boneName=pelvis}
    simSpaceSettings {maxAngularAcceleration=100, maxAngularVelocity=10, simSpaceAngularAlpha=1}
   }
   n30 = Anim Dynamics - Bone: dyn_bk_skirt_c(ComponentPose=@n29, Alpha=1, ExternalForce="0,0,0") {
    angularDampingOverride=1
    angularSpringConstant=10000
    bAngularSpring=true
    bOverrideAngularDamping=true
    bUseGravityOverride=true
    gravityScale=0.35
    simulationSpace=RootRelative
    boundBone {boneName=dyn_bk_skirt_c}
    componentAppliedLinearAccClamp=<4000 4000 4000>
    componentLinearAccScale=<0.4 0.4 0.45>
    componentLinearVelScale=<0.4 0.4 0.45>
    gravityOverride {z=-980}
    physicsBodyDefinitions [{
      boundBone {boneName=dyn_bk_skirt_c}
      boxExtents=<20 30 30>
      constraintSetup {angularLimitsMax=<0 20 5>, angularLimitsMin=<0 -20 -5>, angularTarget=<1 0 0>}
      localJointOffset=<4 0 0>
     }]
    relativeSpaceBone {boneName=pelvis}
    simSpaceSettings {maxAngularAcceleration=100, maxAngularVelocity=10, simSpaceAngularAlpha=1}
   }
   n31 = Anim Dynamics - Bone: dyn_front_jacket_l(ComponentPose=@n30, Alpha=1, ExternalForce="0,0,0") {
    angularDampingOverride=1
    angularSpringConstant=10000
    bAngularSpring=true
    bOverrideAngularDamping=true
    bUseGravityOverride=true
    gravityScale=0.35
    simulationSpace=RootRelative
    boundBone {boneName=dyn_front_jacket_l}
    componentAppliedLinearAccClamp=<4000 4000 4000>
    componentLinearAccScale=<0.4 0.4 0.5>
    componentLinearVelScale=<0.4 0.4 0.5>
    gravityOverride {z=-980}
    physicsBodyDefinitions [{
      boundBone {boneName=dyn_front_jacket_l}
      boxExtents=<20 30 30>
      constraintSetup {angularLimitsMax=<0 15 3>, angularLimitsMin=<0 -15 -3>, angularTarget=<1 0 0>}
      localJointOffset=<4 0 0>
     }]
    relativeSpaceBone {boneName=spine_05}
    simSpaceSettings {maxAngularAcceleration=100, maxAngularVelocity=10, simSpaceAngularAlpha=1}
   }
   n32 = Anim Dynamics - Bone: dyn_side_jacket_l(ComponentPose=@n31, Alpha=1, ExternalForce="0,0,0") {
    angularDampingOverride=1
    angularSpringConstant=10000
    bAngularSpring=true
    bOverrideAngularDamping=true
    bUseGravityOverride=true
    gravityScale=0.35
    simulationSpace=RootRelative
    boundBone {boneName=dyn_side_jacket_l}
    componentAppliedLinearAccClamp=<4000 4000 4000>
    componentLinearAccScale=<0.4 0.4 0.5>
    componentLinearVelScale=<0.4 0.4 0.5>
    gravityOverride {z=-980}
    physicsBodyDefinitions [{
      boundBone {boneName=dyn_side_jacket_l}
      boxExtents=<20 30 30>
      constraintSetup {angularLimitsMax=<0 5 3>, angularLimitsMin=<0 -15 -3>, angularTarget=<1 0 0>}
      localJointOffset=<4 0 0>
     }]
    relativeSpaceBone {boneName=spine_05}
    simSpaceSettings {maxAngularAcceleration=100, maxAngularVelocity=10, simSpaceAngularAlpha=1}
   }
   n33 = Anim Dynamics - Bone: dyn_bk_jacket_l(ComponentPose=@n32, Alpha=1, ExternalForce="0,0,0") {
    angularDampingOverride=1
    angularSpringConstant=10000
    bAngularSpring=true
    bOverrideAngularDamping=true
    bUseGravityOverride=true
    gravityScale=0.35
    simulationSpace=RootRelative
    boundBone {boneName=dyn_bk_jacket_l}
    componentAppliedLinearAccClamp=<4000 4000 4000>
    componentLinearAccScale=<0.4 0.4 0.5>
    componentLinearVelScale=<0.4 0.4 0.5>
    gravityOverride {z=-980}
    physicsBodyDefinitions [{
      boundBone {boneName=dyn_bk_jacket_l}
      boxExtents=<20 30 30>
      constraintSetup {angularLimitsMax=<0 5 3>, angularLimitsMin=<0 -20 0>, angularTarget=<1 0 0>}
      localJointOffset=<4 0 0>
     }]
    relativeSpaceBone {boneName=spine_05}
    simSpaceSettings {maxAngularAcceleration=100, maxAngularVelocity=10, simSpaceAngularAlpha=1}
   }
   n34 = Anim Dynamics - Bone: dyn_front_jacket_r(ComponentPose=@n33, Alpha=1, ExternalForce="0,0,0") {
    angularDampingOverride=1
    angularSpringConstant=10000
    bAngularSpring=true
    bOverrideAngularDamping=true
    bUseGravityOverride=true
    gravityScale=0.35
    simulationSpace=RootRelative
    boundBone {boneName=dyn_front_jacket_r}
    componentAppliedLinearAccClamp=<4000 4000 4000>
    componentLinearAccScale=<0.4 0.4 0.5>
    componentLinearVelScale=<0.4 0.4 0.5>
    gravityOverride {z=-980}
    physicsBodyDefinitions [{
      boundBone {boneName=dyn_front_jacket_r}
      boxExtents=<20 30 30>
      constraintSetup {angularLimitsMax=<0 15 3>, angularLimitsMin=<0 -15 -3>, angularTarget=<1 0 0>}
      localJointOffset=<4 0 0>
     }]
    relativeSpaceBone {boneName=spine_05}
    simSpaceSettings {maxAngularAcceleration=100, maxAngularVelocity=10, simSpaceAngularAlpha=1}
   }
   n35 = Anim Dynamics - Bone: dyn_side_jacket_r(ComponentPose=@n34, Alpha=1, ExternalForce="0,0,0") {
    angularDampingOverride=1
    angularSpringConstant=10000
    bAngularSpring=true
    bOverrideAngularDamping=true
    bUseGravityOverride=true
    gravityScale=0.35
    simulationSpace=RootRelative
    boundBone {boneName=dyn_side_jacket_r}
    componentAppliedLinearAccClamp=<4000 4000 4000>
    componentLinearAccScale=<0.4 0.4 0.5>
    componentLinearVelScale=<0.4 0.4 0.5>
    gravityOverride {z=-980}
    physicsBodyDefinitions [{
      boundBone {boneName=dyn_side_jacket_r}
      boxExtents=<20 30 30>
      constraintSetup {angularLimitsMax=<0 5 3>, angularLimitsMin=<0 -15 -3>, angularTarget=<1 0 0>}
      localJointOffset=<4 0 0>
     }]
    relativeSpaceBone {boneName=spine_05}
    simSpaceSettings {maxAngularAcceleration=100, maxAngularVelocity=10, simSpaceAngularAlpha=1}
   }
   n36 = Anim Dynamics - Bone: dyn_bk_jacket_r(ComponentPose=@n35, Alpha=1, ExternalForce="0,0,0") {
    angularDampingOverride=1
    angularSpringConstant=10000
    bAngularSpring=true
    bOverrideAngularDamping=true
    bUseGravityOverride=true
    gravityScale=0.35
    simulationSpace=RootRelative
    boundBone {boneName=dyn_bk_jacket_r}
    componentAppliedLinearAccClamp=<4000 4000 4000>
    componentLinearAccScale=<0.4 0.4 0.5>
    componentLinearVelScale=<0.4 0.4 0.5>
    gravityOverride {z=-980}
    physicsBodyDefinitions [{
      boundBone {boneName=dyn_bk_jacket_r}
      boxExtents=<20 30 30>
      constraintSetup {angularLimitsMax=<0 5 0>, angularLimitsMin=<0 -20 -3>, angularTarget=<1 0 0>}
      localJointOffset=<4 0 0>
     }]
    relativeSpaceBone {boneName=spine_05}
    simSpaceSettings {maxAngularAcceleration=100, maxAngularVelocity=10, simSpaceAngularAlpha=1}
   }
   n37 = Anim Dynamics - Bone: boot_l_chain_A_01(ComponentPose=@n36, Alpha=1, ExternalForce="0,0,0") {
    angularDampingOverride=1
    angularSpringConstant=5000
    bAngularSpring=true
    bOverrideAngularDamping=true
    bUseGravityOverride=true
    gravityScale=0.35
    simulationSpace=RootRelative
    boundBone {boneName=boot_l_chain_A_01}
    componentAppliedLinearAccClamp=<4000 4000 4000>
    componentLinearAccScale=<0.5 0.5 0.5>
    componentLinearVelScale=<0.5 0.5 0.5>
    gravityOverride {z=-980}
    physicsBodyDefinitions [{
      boundBone {boneName=boot_l_chain_A_01}
      boxExtents=<10 10 10>
      constraintSetup {angularLimitsMax=<0 0 20>, angularLimitsMin=<0 0 0>, angularTarget=<1 0 0>}
      localJointOffset=<4 0 0>
     }]
    simSpaceSettings {maxAngularAcceleration=100, maxAngularVelocity=10, simSpaceAngularAlpha=1}
   }
   n38 = Anim Dynamics - Bone: boot_l_chain_B_01(ComponentPose=@n37, Alpha=1, ExternalForce="0,0,0") {
    angularDampingOverride=1
    angularSpringConstant=5000
    bAngularSpring=true
    bOverrideAngularDamping=true
    bUseGravityOverride=true
    gravityScale=0.35
    simulationSpace=RootRelative
    boundBone {boneName=boot_l_chain_B_01}
    componentAppliedLinearAccClamp=<4000 4000 4000>
    componentLinearAccScale=<0.5 0.5 0.5>
    componentLinearVelScale=<0.5 0.5 0.5>
    gravityOverride {z=-980}
    physicsBodyDefinitions [{
      boundBone {boneName=boot_l_chain_B_01}
      boxExtents=<10 10 10>
      constraintSetup {angularLimitsMax=<0 0 25>, angularLimitsMin=<0 0 0>, angularTarget=<1 0 0>}
      localJointOffset=<4 0 0>
     }]
    simSpaceSettings {maxAngularAcceleration=100, maxAngularVelocity=10, simSpaceAngularAlpha=1}
   }
   n39 = Anim Dynamics - Bone: boot_l_chain_B_02(ComponentPose=@n38, Alpha=1, ExternalForce="0,0,0") {
    angularDampingOverride=1
    angularSpringConstant=5000
    bAngularSpring=true
    bOverrideAngularDamping=true
    bUseGravityOverride=true
    gravityScale=0.35
    simulationSpace=RootRelative
    boundBone {boneName=boot_l_chain_B_02}
    componentAppliedLinearAccClamp=<4000 4000 4000>
    componentLinearAccScale=<0.5 0.5 0.5>
    componentLinearVelScale=<0.5 0.5 0.5>
    gravityOverride {z=-980}
    physicsBodyDefinitions [{
      boundBone {boneName=boot_l_chain_B_02}
      boxExtents=<10 10 10>
      constraintSetup {angularLimitsMax=<0 0 20>, angularLimitsMin=<0 0 0>, angularTarget=<1 0 0>}
      localJointOffset=<4 0 0>
     }]
    simSpaceSettings {maxAngularAcceleration=100, maxAngularVelocity=10, simSpaceAngularAlpha=1}
   }
   n40 = Anim Dynamics - Bone: boot_l_chain_C_01(ComponentPose=@n39, Alpha=1, ExternalForce="0,0,0") {
    angularDampingOverride=1
    angularSpringConstant=5000
    bAngularSpring=true
    bOverrideAngularDamping=true
    bUseGravityOverride=true
    gravityScale=0.35
    simulationSpace=RootRelative
    boundBone {boneName=boot_l_chain_C_01}
    componentAppliedLinearAccClamp=<4000 4000 4000>
    componentLinearAccScale=<0.5 0.5 0.5>
    componentLinearVelScale=<0.5 0.5 0.5>
    gravityOverride {z=-980}
    physicsBodyDefinitions [{
      boundBone {boneName=boot_l_chain_C_01}
      boxExtents=<10 10 10>
      constraintSetup {angularLimitsMax=<0 0 20>, angularLimitsMin=<0 0 0>, angularTarget=<1 0 0>}
      localJointOffset=<4 0 0>
     }]
    simSpaceSettings {maxAngularAcceleration=100, maxAngularVelocity=10, simSpaceAngularAlpha=1}
   }
   n41 = Anim Dynamics - Bone: boot_l_chain_C_02(ComponentPose=@n40, Alpha=1, ExternalForce="0,0,0") {
    angularDampingOverride=1
    angularSpringConstant=5000
    bAngularSpring=true
    bOverrideAngularDamping=true
    bUseGravityOverride=true
    gravityScale=0.35
    simulationSpace=RootRelative
    boundBone {boneName=boot_l_chain_C_02}
    componentAppliedLinearAccClamp=<4000 4000 4000>
    componentLinearAccScale=<0.5 0.5 0.5>
    componentLinearVelScale=<0.5 0.5 0.5>
    gravityOverride {z=-980}
    physicsBodyDefinitions [{
      boundBone {boneName=boot_l_chain_C_02}
      boxExtents=<10 10 10>
      constraintSetup {angularLimitsMax=<0 0 25>, angularLimitsMin=<0 0 0>, angularTarget=<1 0 0>}
      localJointOffset=<4 0 0>
     }]
    simSpaceSettings {maxAngularAcceleration=100, maxAngularVelocity=10, simSpaceAngularAlpha=1}
   }
   n42 = Anim Dynamics - Bone: boot_r_chain_A_02(ComponentPose=@n41, Alpha=1, ExternalForce="0,0,0") {
    angularDampingOverride=1
    angularSpringConstant=5000
    bAngularSpring=true
    bOverrideAngularDamping=true
    bUseGravityOverride=true
    gravityScale=0.35
    simulationSpace=RootRelative
    boundBone {boneName=boot_r_chain_A_02}
    componentAppliedLinearAccClamp=<4000 4000 4000>
    componentLinearAccScale=<0.5 0.5 0.5>
    componentLinearVelScale=<0.5 0.5 0.5>
    gravityOverride {z=-980}
    physicsBodyDefinitions [{
      boundBone {boneName=boot_r_chain_A_02}
      boxExtents=<10 10 10>
      constraintSetup {angularLimitsMax=<0 0 0>, angularLimitsMin=<0 0 -25>, angularTarget=<1 0 0>}
      localJointOffset=<4 0 0>
     }]
    simSpaceSettings {maxAngularAcceleration=100, maxAngularVelocity=10, simSpaceAngularAlpha=1}
   }
   n43 = Anim Dynamics - Bone: boot_r_chain_B_03(ComponentPose=@n42, Alpha=1, ExternalForce="0,0,0") {
    angularDampingOverride=1
    angularSpringConstant=5000
    bAngularSpring=true
    bOverrideAngularDamping=true
    bUseGravityOverride=true
    gravityScale=0.35
    simulationSpace=RootRelative
    boundBone {boneName=boot_r_chain_B_03}
    componentAppliedLinearAccClamp=<4000 4000 4000>
    componentLinearAccScale=<0.5 0.5 0.5>
    componentLinearVelScale=<0.5 0.5 0.5>
    gravityOverride {z=-980}
    physicsBodyDefinitions [{
      boundBone {boneName=boot_r_chain_B_03}
      boxExtents=<10 10 10>
      constraintSetup {angularLimitsMax=<0 0 0>, angularLimitsMin=<0 0 -20>, angularTarget=<1 0 0>}
      localJointOffset=<4 0 0>
     }]
    simSpaceSettings {maxAngularAcceleration=100, maxAngularVelocity=10, simSpaceAngularAlpha=1}
   }
   n44 = Anim Dynamics - Bone: boot_r_chain_B_02(ComponentPose=@n43, Alpha=1, ExternalForce="0,0,0") {
    angularDampingOverride=1
    angularSpringConstant=5000
    bAngularSpring=true
    bOverrideAngularDamping=true
    bUseGravityOverride=true
    gravityScale=0.35
    simulationSpace=RootRelative
    boundBone {boneName=boot_r_chain_B_02}
    componentAppliedLinearAccClamp=<4000 4000 4000>
    componentLinearAccScale=<0.5 0.5 0.5>
    componentLinearVelScale=<0.5 0.5 0.5>
    gravityOverride {z=-980}
    physicsBodyDefinitions [{
      boundBone {boneName=boot_r_chain_B_02}
      boxExtents=<10 10 10>
      constraintSetup {angularLimitsMax=<0 0 0>, angularLimitsMin=<0 0 -25>, angularTarget=<1 0 0>}
      localJointOffset=<4 0 0>
     }]
    simSpaceSettings {maxAngularAcceleration=100, maxAngularVelocity=10, simSpaceAngularAlpha=1}
   }
   n45 = Anim Dynamics - Bone: boot_r_chain_C_01(ComponentPose=@n44, Alpha=1, ExternalForce="0,0,0") {
    angularDampingOverride=1
    angularSpringConstant=5000
    bAngularSpring=true
    bOverrideAngularDamping=true
    bUseGravityOverride=true
    gravityScale=0.35
    simulationSpace=RootRelative
    boundBone {boneName=boot_r_chain_C_01}
    componentAppliedLinearAccClamp=<4000 4000 4000>
    componentLinearAccScale=<0.5 0.5 0.5>
    componentLinearVelScale=<0.5 0.5 0.5>
    gravityOverride {z=-980}
    physicsBodyDefinitions [{
      boundBone {boneName=boot_r_chain_C_01}
      boxExtents=<10 10 10>
      constraintSetup {angularLimitsMax=<0 0 -5>, angularLimitsMin=<0 0 -20>, angularTarget=<1 0 0>}
      localJointOffset=<4 0 0>
     }]
    simSpaceSettings {maxAngularAcceleration=100, maxAngularVelocity=10, simSpaceAngularAlpha=1}
   }
   n46 = Anim Dynamics - Bone: boot_r_chain_C_02(ComponentPose=@n45, Alpha=1, ExternalForce="0,0,0") {
    angularDampingOverride=1
    angularSpringConstant=5000
    bAngularSpring=true
    bOverrideAngularDamping=true
    bUseGravityOverride=true
    gravityScale=0.35
    simulationSpace=RootRelative
    boundBone {boneName=boot_r_chain_C_02}
    componentAppliedLinearAccClamp=<4000 4000 4000>
    componentLinearAccScale=<0.5 0.5 0.5>
    componentLinearVelScale=<0.5 0.5 0.5>
    gravityOverride {z=-980}
    physicsBodyDefinitions [{
      boundBone {boneName=boot_r_chain_C_02}
      boxExtents=<10 10 10>
      constraintSetup {angularLimitsMax=<0 0 0>, angularLimitsMin=<0 0 -20>, angularTarget=<1 0 0>}
      localJointOffset=<4 0 0>
     }]
    simSpaceSettings {maxAngularAcceleration=100, maxAngularVelocity=10, simSpaceAngularAlpha=1}
   }
   Save cached pose 'Initial Anim Dynamics'(Pose=Component To Local(ComponentPose=@n46)) {cacheName=Initial Anim Dynamics}
   n47 = Anim Dynamics - Bone: dyn_LoopA_up(ComponentPose=Local To Component(LocalPose=Use cached pose 'Initial Anim Dynamics'()), Alpha=1, ExternalForce="0,0,0") {
    angularDampingOverride=1
    angularSpringConstant=20000
    bAngularSpring=true
    bOverrideAngularDamping=true
    bUseGravityOverride=true
    gravityScale=0.35
    simulationSpace=RootRelative
    boundBone {boneName=dyn_LoopA_up}
    componentLinearAccScale=<1 1 1>
    componentLinearVelScale=<1 1 1>
    gravityOverride {z=-980}
    physicsBodyDefinitions [{
      boundBone {boneName=dyn_LoopA_up}
      boxExtents=<20 30 30>
      constraintSetup {angularLimitsMax=<15 15 15>, angularLimitsMin=<-15 -15 -15>, angularTarget=<1 0 0>}
      localJointOffset=<4 0 0>
     }]
   }
   n48 = Anim Dynamics - Bone: dyn_LoopA_dw(ComponentPose=@n47, Alpha=1, ExternalForce="0,0,0") {
    angularDampingOverride=1
    angularSpringConstant=20000
    bAngularSpring=true
    bOverrideAngularDamping=true
    bUseGravityOverride=true
    gravityScale=0.35
    simulationSpace=RootRelative
    boundBone {boneName=dyn_LoopA_dw}
    componentLinearAccScale=<1 1 1>
    componentLinearVelScale=<1 1 1>
    gravityOverride {z=-980}
    physicsBodyDefinitions [{
      boundBone {boneName=dyn_LoopA_dw}
      boxExtents=<20 30 30>
      constraintSetup {angularLimitsMax=<15 15 15>, angularLimitsMin=<-15 -15 -15>, angularTarget=<1 0 0>}
      localJointOffset=<4 0 0>
     }]
   }
   n49 = Anim Dynamics - Bone: dyn_LoopB_up(ComponentPose=@n48, Alpha=1, ExternalForce="0,0,0") {
    angularDampingOverride=1
    angularSpringConstant=20000
    bAngularSpring=true
    bOverrideAngularDamping=true
    bUseGravityOverride=true
    gravityScale=0.35
    simulationSpace=RootRelative
    boundBone {boneName=dyn_LoopB_up}
    componentLinearAccScale=<1 1 1>
    componentLinearVelScale=<1 1 1>
    gravityOverride {z=-980}
    physicsBodyDefinitions [{
      boundBone {boneName=dyn_LoopB_up}
      boxExtents=<20 30 30>
      constraintSetup {angularLimitsMax=<15 15 15>, angularLimitsMin=<-15 -15 -15>, angularTarget=<1 0 0>}
      localJointOffset=<4 0 0>
     }]
   }
   n50 = Anim Dynamics - Bone: dyn_LoopB_dw(ComponentPose=@n49, Alpha=1, ExternalForce="0,0,0") {
    angularDampingOverride=1
    angularSpringConstant=20000
    bAngularSpring=true
    bOverrideAngularDamping=true
    bUseGravityOverride=true
    gravityScale=0.35
    simulationSpace=RootRelative
    boundBone {boneName=dyn_LoopB_dw}
    componentLinearAccScale=<1 1 1>
    componentLinearVelScale=<1 1 1>
    gravityOverride {z=-980}
    physicsBodyDefinitions [{
      boundBone {boneName=dyn_LoopB_dw}
      boxExtents=<20 30 30>
      constraintSetup {angularLimitsMax=<15 15 15>, angularLimitsMin=<-15 -15 -15>, angularTarget=<1 0 0>}
      localJointOffset=<4 0 0>
     }]
   }
   n51 = Anim Dynamics - Bone: dyn_LoopC_up(ComponentPose=@n50, Alpha=1, ExternalForce="0,0,0") {
    angularDampingOverride=1
    angularSpringConstant=20000
    bAngularSpring=true
    bOverrideAngularDamping=true
    bUseGravityOverride=true
    gravityScale=0.35
    simulationSpace=RootRelative
    boundBone {boneName=dyn_LoopC_up}
    componentLinearAccScale=<1 1 1>
    componentLinearVelScale=<1 1 1>
    gravityOverride {z=-980}
    physicsBodyDefinitions [{
      boundBone {boneName=dyn_LoopC_up}
      boxExtents=<20 30 30>
      constraintSetup {angularLimitsMax=<15 15 15>, angularLimitsMin=<-15 -15 -15>, angularTarget=<1 0 0>}
      localJointOffset=<4 0 0>
     }]
   }
   n52 = Anim Dynamics - Bone: dyn_LoopC_dw(ComponentPose=@n51, Alpha=1, ExternalForce="0,0,0") {
    angularDampingOverride=1
    angularSpringConstant=20000
    bAngularSpring=true
    bOverrideAngularDamping=true
    bUseGravityOverride=true
    gravityScale=0.35
    simulationSpace=RootRelative
    boundBone {boneName=dyn_LoopC_dw}
    componentLinearAccScale=<1 1 1>
    componentLinearVelScale=<1 1 1>
    gravityOverride {z=-980}
    physicsBodyDefinitions [{
      boundBone {boneName=dyn_LoopC_dw}
      boxExtents=<20 30 30>
      constraintSetup {angularLimitsMax=<15 15 15>, angularLimitsMin=<-15 -15 -15>, angularTarget=<1 0 0>}
      localJointOffset=<4 0 0>
     }]
   }
   n53 = Anim Dynamics - Bone: base_l_padChain1(ComponentPose=@n52, Alpha=1, ExternalForce="0,0,0") {
    angularDampingOverride=1
    angularSpringConstant=3000
    bAngularSpring=true
    bOverrideAngularDamping=true
    bUseGravityOverride=true
    gravityScale=0.35
    simulationSpace=RootRelative
    boundBone {boneName=base_l_padChain1}
    componentAppliedLinearAccClamp=<4000 4000 4000>
    componentLinearAccScale=<0.45 0.45 0.5>
    componentLinearVelScale=<0.45 0.45 0.5>
    gravityOverride {z=-980}
    physicsBodyDefinitions [{
      boundBone {boneName=base_l_padChain1}
      boxExtents=<10 12 10>
      constraintSetup {angularLimitsMax=<0 3 7>, angularLimitsMin=<0 -60 -7>, angularTarget=<1 0 0>}
      localJointOffset=<4 0 0>
     }]
    simSpaceSettings {maxAngularAcceleration=100, maxAngularVelocity=10, simSpaceAngularAlpha=1}
   }
   n54 = Anim Dynamics - Bone: base_l_padChain1A(ComponentPose=@n53, Alpha=1, ExternalForce="0,0,0") {
    angularDampingOverride=1
    angularSpringConstant=3000
    bAngularSpring=true
    bOverrideAngularDamping=true
    bUseGravityOverride=true
    gravityScale=0.35
    simulationSpace=RootRelative
    boundBone {boneName=base_l_padChain1A}
    componentAppliedLinearAccClamp=<4000 4000 4000>
    componentLinearAccScale=<0.45 0.45 0.5>
    componentLinearVelScale=<0.45 0.45 0.5>
    gravityOverride {z=-980}
    physicsBodyDefinitions [{
      boundBone {boneName=base_l_padChain1A}
      boxExtents=<10 12 10>
      constraintSetup {angularLimitsMax=<0 3 7>, angularLimitsMin=<0 -60 -7>, angularTarget=<1 0 0>}
      localJointOffset=<4 0 0>
     }]
    simSpaceSettings {maxAngularAcceleration=100, maxAngularVelocity=10, simSpaceAngularAlpha=1}
   }
   n55 = Anim Dynamics - Bone: base_l_padChain2(ComponentPose=@n54, Alpha=1, ExternalForce="0,0,0") {
    angularDampingOverride=1
    angularSpringConstant=3000
    bAngularSpring=true
    bOverrideAngularDamping=true
    bUseGravityOverride=true
    gravityScale=0.35
    simulationSpace=RootRelative
    boundBone {boneName=base_l_padChain2}
    componentAppliedLinearAccClamp=<4000 4000 4000>
    componentLinearAccScale=<0.45 0.45 0.5>
    componentLinearVelScale=<0.45 0.45 0.5>
    gravityOverride {z=-980}
    physicsBodyDefinitions [{
      boundBone {boneName=base_l_padChain2}
      boxExtents=<10 12 10>
      constraintSetup {angularLimitsMax=<0 3 7>, angularLimitsMin=<0 -60 -7>, angularTarget=<1 0 0>}
      localJointOffset=<4 0 0>
     }]
    simSpaceSettings {maxAngularAcceleration=100, maxAngularVelocity=10, simSpaceAngularAlpha=1}
   }
   n56 = Anim Dynamics - Bone: base_l_padChain2A(ComponentPose=@n55, Alpha=1, ExternalForce="0,0,0") {
    angularDampingOverride=1
    angularSpringConstant=3000
    bAngularSpring=true
    bOverrideAngularDamping=true
    bUseGravityOverride=true
    gravityScale=0.35
    simulationSpace=RootRelative
    boundBone {boneName=base_l_padChain2A}
    componentAppliedLinearAccClamp=<4000 4000 4000>
    componentLinearAccScale=<0.45 0.45 0.5>
    componentLinearVelScale=<0.45 0.45 0.5>
    gravityOverride {z=-980}
    physicsBodyDefinitions [{
      boundBone {boneName=base_l_padChain2A}
      boxExtents=<10 12 10>
      constraintSetup {angularLimitsMax=<0 3 7>, angularLimitsMin=<0 -60 -7>, angularTarget=<1 0 0>}
      localJointOffset=<4 0 0>
     }]
    simSpaceSettings {maxAngularAcceleration=100, maxAngularVelocity=10, simSpaceAngularAlpha=1}
   }
   n57 = Anim Dynamics - Bone: base_l_padChain3(ComponentPose=@n56, Alpha=1, ExternalForce="0,0,0") {
    angularDampingOverride=1
    angularSpringConstant=3000
    bAngularSpring=true
    bOverrideAngularDamping=true
    bUseGravityOverride=true
    gravityScale=0.35
    simulationSpace=RootRelative
    boundBone {boneName=base_l_padChain3}
    componentAppliedLinearAccClamp=<4000 4000 4000>
    componentLinearAccScale=<0.45 0.45 0.5>
    componentLinearVelScale=<0.45 0.45 0.5>
    gravityOverride {z=-980}
    physicsBodyDefinitions [{
      boundBone {boneName=base_l_padChain3}
      boxExtents=<10 12 10>
      constraintSetup {angularLimitsMax=<0 3 7>, angularLimitsMin=<0 -60 -7>, angularTarget=<1 0 0>}
      localJointOffset=<4 0 0>
     }]
    simSpaceSettings {maxAngularAcceleration=100, maxAngularVelocity=10, simSpaceAngularAlpha=1}
   }
   n58 = Anim Dynamics - Bone: base_l_padChain3A(ComponentPose=@n57, Alpha=1, ExternalForce="0,0,0") {
    angularDampingOverride=1
    angularSpringConstant=3000
    bAngularSpring=true
    bOverrideAngularDamping=true
    bUseGravityOverride=true
    gravityScale=0.35
    simulationSpace=RootRelative
    boundBone {boneName=base_l_padChain3A}
    componentAppliedLinearAccClamp=<4000 4000 4000>
    componentLinearAccScale=<0.45 0.45 0.5>
    componentLinearVelScale=<0.45 0.45 0.5>
    gravityOverride {z=-980}
    physicsBodyDefinitions [{
      boundBone {boneName=base_l_padChain3A}
      boxExtents=<10 12 10>
      constraintSetup {angularLimitsMax=<0 3 7>, angularLimitsMin=<0 -60 -7>, angularTarget=<1 0 0>}
      localJointOffset=<4 0 0>
     }]
    simSpaceSettings {maxAngularAcceleration=100, maxAngularVelocity=10, simSpaceAngularAlpha=1}
   }
   n59 = Anim Dynamics - Bone: base_l_padChain4(ComponentPose=@n58, Alpha=1, ExternalForce="0,0,0") {
    angularDampingOverride=1
    angularSpringConstant=3000
    bAngularSpring=true
    bOverrideAngularDamping=true
    bUseGravityOverride=true
    gravityScale=0.35
    simulationSpace=RootRelative
    boundBone {boneName=base_l_padChain4}
    componentAppliedLinearAccClamp=<4000 4000 4000>
    componentLinearAccScale=<0.45 0.45 0.5>
    componentLinearVelScale=<0.45 0.45 0.5>
    gravityOverride {z=-980}
    physicsBodyDefinitions [{
      boundBone {boneName=base_l_padChain4}
      boxExtents=<10 12 10>
      constraintSetup {angularLimitsMax=<0 3 7>, angularLimitsMin=<0 -60 -7>, angularTarget=<1 0 0>}
      localJointOffset=<4 0 0>
     }]
    simSpaceSettings {maxAngularAcceleration=100, maxAngularVelocity=10, simSpaceAngularAlpha=1}
   }
   n60 = Anim Dynamics - Bone: base_l_padChain5(ComponentPose=@n59, Alpha=1, ExternalForce="0,0,0") {
    angularDampingOverride=1
    angularSpringConstant=3000
    bAngularSpring=true
    bOverrideAngularDamping=true
    bUseGravityOverride=true
    gravityScale=0.35
    simulationSpace=RootRelative
    boundBone {boneName=base_l_padChain5}
    componentAppliedLinearAccClamp=<4000 4000 4000>
    componentLinearAccScale=<0.45 0.45 0.5>
    componentLinearVelScale=<0.45 0.45 0.5>
    gravityOverride {z=-980}
    physicsBodyDefinitions [{
      boundBone {boneName=base_l_padChain5}
      boxExtents=<10 12 10>
      constraintSetup {angularLimitsMax=<0 3 7>, angularLimitsMin=<0 -60 -7>, angularTarget=<1 0 0>}
      localJointOffset=<4 0 0>
     }]
    simSpaceSettings {maxAngularAcceleration=100, maxAngularVelocity=10, simSpaceAngularAlpha=1}
   }
   n61 = Anim Dynamics - Bone: base_l_padChain6(ComponentPose=@n60, Alpha=1, ExternalForce="0,0,0") {
    angularDampingOverride=1
    angularSpringConstant=3000
    bAngularSpring=true
    bOverrideAngularDamping=true
    bUseGravityOverride=true
    gravityScale=0.35
    simulationSpace=RootRelative
    boundBone {boneName=base_l_padChain6}
    componentAppliedLinearAccClamp=<4000 4000 4000>
    componentLinearAccScale=<0.45 0.45 0.5>
    componentLinearVelScale=<0.45 0.45 0.5>
    gravityOverride {z=-980}
    physicsBodyDefinitions [{
      boundBone {boneName=base_l_padChain6}
      boxExtents=<10 12 10>
      constraintSetup {angularLimitsMax=<0 3 7>, angularLimitsMin=<0 -60 -7>, angularTarget=<1 0 0>}
      localJointOffset=<4 0 0>
     }]
    simSpaceSettings {maxAngularAcceleration=100, maxAngularVelocity=10, simSpaceAngularAlpha=1}
   }
   n62 = Anim Dynamics - Bone: base_l_padChain7(ComponentPose=@n61, Alpha=1, ExternalForce="0,0,0") {
    angularDampingOverride=1
    angularSpringConstant=3000
    bAngularSpring=true
    bOverrideAngularDamping=true
    bUseGravityOverride=true
    gravityScale=0.35
    simulationSpace=RootRelative
    boundBone {boneName=base_l_padChain7}
    componentAppliedLinearAccClamp=<4000 4000 4000>
    componentLinearAccScale=<0.45 0.45 0.5>
    componentLinearVelScale=<0.45 0.45 0.5>
    gravityOverride {z=-980}
    physicsBodyDefinitions [{
      boundBone {boneName=base_l_padChain7}
      boxExtents=<10 12 10>
      constraintSetup {angularLimitsMax=<0 3 7>, angularLimitsMin=<0 -60 -7>, angularTarget=<1 0 0>}
      localJointOffset=<4 0 0>
     }]
    simSpaceSettings {maxAngularAcceleration=100, maxAngularVelocity=10, simSpaceAngularAlpha=1}
   }
   n63 = Anim Dynamics - Bone: base_l_padChain8(ComponentPose=@n62, Alpha=1, ExternalForce="0,0,0") {
    angularDampingOverride=1
    angularSpringConstant=3000
    bAngularSpring=true
    bOverrideAngularDamping=true
    bUseGravityOverride=true
    gravityScale=0.35
    simulationSpace=RootRelative
    boundBone {boneName=base_l_padChain8}
    componentAppliedLinearAccClamp=<4000 4000 4000>
    componentLinearAccScale=<0.45 0.45 0.5>
    componentLinearVelScale=<0.45 0.45 0.5>
    gravityOverride {z=-980}
    physicsBodyDefinitions [{
      boundBone {boneName=base_l_padChain8}
      boxExtents=<10 12 10>
      constraintSetup {angularLimitsMax=<0 3 7>, angularLimitsMin=<0 -60 -7>, angularTarget=<1 0 0>}
      localJointOffset=<4 0 0>
     }]
    simSpaceSettings {maxAngularAcceleration=100, maxAngularVelocity=10, simSpaceAngularAlpha=1}
   }
   n64 = Anim Dynamics - Bone: base_l_padChain9(ComponentPose=@n63, Alpha=1, ExternalForce="0,0,0") {
    angularDampingOverride=1
    angularSpringConstant=3000
    bAngularSpring=true
    bOverrideAngularDamping=true
    bUseGravityOverride=true
    gravityScale=0.35
    simulationSpace=RootRelative
    boundBone {boneName=base_l_padChain9}
    componentAppliedLinearAccClamp=<4000 4000 4000>
    componentLinearAccScale=<0.45 0.45 0.5>
    componentLinearVelScale=<0.45 0.45 0.5>
    gravityOverride {z=-980}
    physicsBodyDefinitions [{
      boundBone {boneName=base_l_padChain9}
      boxExtents=<10 12 10>
      constraintSetup {angularLimitsMax=<0 3 7>, angularLimitsMin=<0 -60 -7>, angularTarget=<1 0 0>}
      localJointOffset=<4 0 0>
     }]
    simSpaceSettings {maxAngularAcceleration=100, maxAngularVelocity=10, simSpaceAngularAlpha=1}
   }
   n65 = Anim Dynamics - Bone: base_l_padChain10(ComponentPose=@n64, Alpha=1, ExternalForce="0,0,0") {
    angularDampingOverride=1
    angularSpringConstant=3000
    bAngularSpring=true
    bOverrideAngularDamping=true
    bUseGravityOverride=true
    gravityScale=0.35
    simulationSpace=RootRelative
    boundBone {boneName=base_l_padChain10}
    componentAppliedLinearAccClamp=<4000 4000 4000>
    componentLinearAccScale=<0.45 0.45 0.5>
    componentLinearVelScale=<0.45 0.45 0.5>
    gravityOverride {z=-980}
    physicsBodyDefinitions [{
      boundBone {boneName=base_l_padChain10}
      boxExtents=<10 12 10>
      constraintSetup {angularLimitsMax=<0 3 7>, angularLimitsMin=<0 -60 -7>, angularTarget=<1 0 0>}
      localJointOffset=<4 0 0>
     }]
    simSpaceSettings {maxAngularAcceleration=100, maxAngularVelocity=10, simSpaceAngularAlpha=1}
   }
   n66 = Anim Dynamics - Bone: base_l_padChain11(ComponentPose=@n65, Alpha=1, ExternalForce="0,0,0") {
    angularDampingOverride=1
    angularSpringConstant=3000
    bAngularSpring=true
    bOverrideAngularDamping=true
    bUseGravityOverride=true
    gravityScale=0.35
    simulationSpace=RootRelative
    boundBone {boneName=base_l_padChain11}
    componentAppliedLinearAccClamp=<4000 4000 4000>
    componentLinearAccScale=<0.45 0.45 0.5>
    componentLinearVelScale=<0.45 0.45 0.5>
    gravityOverride {z=-980}
    physicsBodyDefinitions [{
      boundBone {boneName=base_l_padChain11}
      boxExtents=<10 12 10>
      constraintSetup {angularLimitsMax=<0 3 7>, angularLimitsMin=<0 -60 -7>, angularTarget=<1 0 0>}
      localJointOffset=<4 0 0>
     }]
    simSpaceSettings {maxAngularAcceleration=100, maxAngularVelocity=10, simSpaceAngularAlpha=1}
   }
   n67 = Anim Dynamics - Bone: base_l_padChain12(ComponentPose=@n66, Alpha=1, ExternalForce="0,0,0") {
    angularDampingOverride=1
    angularSpringConstant=3000
    bAngularSpring=true
    bOverrideAngularDamping=true
    bUseGravityOverride=true
    gravityScale=0.35
    simulationSpace=RootRelative
    boundBone {boneName=base_l_padChain12}
    componentAppliedLinearAccClamp=<4000 4000 4000>
    componentLinearAccScale=<0.45 0.45 0.5>
    componentLinearVelScale=<0.45 0.45 0.5>
    gravityOverride {z=-980}
    physicsBodyDefinitions [{
      boundBone {boneName=base_l_padChain12}
      boxExtents=<10 12 10>
      constraintSetup {angularLimitsMax=<0 3 7>, angularLimitsMin=<0 -60 -7>, angularTarget=<1 0 0>}
      localJointOffset=<4 0 0>
     }]
    simSpaceSettings {maxAngularAcceleration=100, maxAngularVelocity=10, simSpaceAngularAlpha=1}
   }
   n68 = Anim Dynamics - Bone: base_l_padChain13(ComponentPose=@n67, Alpha=1, ExternalForce="0,0,0") {
    angularDampingOverride=1
    angularSpringConstant=3000
    bAngularSpring=true
    bOverrideAngularDamping=true
    bUseGravityOverride=true
    gravityScale=0.35
    simulationSpace=RootRelative
    boundBone {boneName=base_l_padChain13}
    componentAppliedLinearAccClamp=<4000 4000 4000>
    componentLinearAccScale=<0.45 0.45 0.5>
    componentLinearVelScale=<0.45 0.45 0.5>
    gravityOverride {z=-980}
    physicsBodyDefinitions [{
      boundBone {boneName=base_l_padChain13}
      boxExtents=<10 12 10>
      constraintSetup {angularLimitsMax=<0 3 7>, angularLimitsMin=<0 -60 -7>, angularTarget=<1 0 0>}
      localJointOffset=<4 0 0>
     }]
    simSpaceSettings {maxAngularAcceleration=100, maxAngularVelocity=10, simSpaceAngularAlpha=1}
   }
   n69 = Anim Dynamics - Bone: base_l_padChain14(ComponentPose=@n68, Alpha=1, ExternalForce="0,0,0") {
    angularDampingOverride=1
    angularSpringConstant=3000
    bAngularSpring=true
    bOverrideAngularDamping=true
    bUseGravityOverride=true
    gravityScale=0.35
    simulationSpace=RootRelative
    boundBone {boneName=base_l_padChain14}
    componentAppliedLinearAccClamp=<4000 4000 4000>
    componentLinearAccScale=<0.45 0.45 0.5>
    componentLinearVelScale=<0.45 0.45 0.5>
    gravityOverride {z=-980}
    physicsBodyDefinitions [{
      boundBone {boneName=base_l_padChain14}
      boxExtents=<10 12 10>
      constraintSetup {angularLimitsMax=<0 3 7>, angularLimitsMin=<0 -60 -7>, angularTarget=<1 0 0>}
      localJointOffset=<4 0 0>
     }]
    simSpaceSettings {maxAngularAcceleration=100, maxAngularVelocity=10, simSpaceAngularAlpha=1}
   }
   n70 = Anim Dynamics - Bone: base_l_padChain15(ComponentPose=@n69, Alpha=1, ExternalForce="0,0,0") {
    angularDampingOverride=1
    angularSpringConstant=3000
    bAngularSpring=true
    bOverrideAngularDamping=true
    bUseGravityOverride=true
    gravityScale=0.35
    simulationSpace=RootRelative
    boundBone {boneName=base_l_padChain15}
    componentAppliedLinearAccClamp=<4000 4000 4000>
    componentLinearAccScale=<0.45 0.45 0.5>
    componentLinearVelScale=<0.45 0.45 0.5>
    gravityOverride {z=-980}
    physicsBodyDefinitions [{
      boundBone {boneName=base_l_padChain15}
      boxExtents=<10 12 10>
      constraintSetup {angularLimitsMax=<0 3 7>, angularLimitsMin=<0 -60 -7>, angularTarget=<1 0 0>}
      localJointOffset=<4 0 0>
     }]
    simSpaceSettings {maxAngularAcceleration=100, maxAngularVelocity=10, simSpaceAngularAlpha=1}
   }
   n71 = Anim Dynamics - Bone: base_l_padChain16(ComponentPose=@n70, Alpha=1, ExternalForce="0,0,0") {
    angularDampingOverride=1
    angularSpringConstant=3000
    bAngularSpring=true
    bOverrideAngularDamping=true
    bUseGravityOverride=true
    gravityScale=0.35
    simulationSpace=RootRelative
    boundBone {boneName=base_l_padChain16}
    componentAppliedLinearAccClamp=<4000 4000 4000>
    componentLinearAccScale=<0.45 0.45 0.5>
    componentLinearVelScale=<0.45 0.45 0.5>
    gravityOverride {z=-980}
    physicsBodyDefinitions [{
      boundBone {boneName=base_l_padChain16}
      boxExtents=<10 12 10>
      constraintSetup {angularLimitsMax=<0 3 7>, angularLimitsMin=<0 -60 -7>, angularTarget=<1 0 0>}
      localJointOffset=<4 0 0>
     }]
    simSpaceSettings {maxAngularAcceleration=100, maxAngularVelocity=10, simSpaceAngularAlpha=1}
   }
   n72 = Anim Dynamics - Bone: base_l_padChain17(ComponentPose=@n71, Alpha=1, ExternalForce="0,0,0") {
    angularDampingOverride=1
    angularSpringConstant=3000
    bAngularSpring=true
    bOverrideAngularDamping=true
    bUseGravityOverride=true
    gravityScale=0.35
    simulationSpace=RootRelative
    boundBone {boneName=base_l_padChain17}
    componentAppliedLinearAccClamp=<4000 4000 4000>
    componentLinearAccScale=<0.45 0.45 0.5>
    componentLinearVelScale=<0.45 0.45 0.5>
    gravityOverride {z=-980}
    physicsBodyDefinitions [{
      boundBone {boneName=base_l_padChain17}
      boxExtents=<10 12 10>
      constraintSetup {angularLimitsMax=<0 3 7>, angularLimitsMin=<0 -60 -7>, angularTarget=<1 0 0>}
      localJointOffset=<4 0 0>
     }]
    simSpaceSettings {maxAngularAcceleration=100, maxAngularVelocity=10, simSpaceAngularAlpha=1}
   }
   n73 = Anim Dynamics - Bone: base_l_padChain18(ComponentPose=@n72, Alpha=1, ExternalForce="0,0,0") {
    angularDampingOverride=1
    angularSpringConstant=3000
    bAngularSpring=true
    bOverrideAngularDamping=true
    bUseGravityOverride=true
    gravityScale=0.35
    simulationSpace=RootRelative
    boundBone {boneName=base_l_padChain18}
    componentAppliedLinearAccClamp=<4000 4000 4000>
    componentLinearAccScale=<0.45 0.45 0.5>
    componentLinearVelScale=<0.45 0.45 0.5>
    gravityOverride {z=-980}
    physicsBodyDefinitions [{
      boundBone {boneName=base_l_padChain18}
      boxExtents=<10 12 10>
      constraintSetup {angularLimitsMax=<0 3 7>, angularLimitsMin=<0 -60 -7>, angularTarget=<1 0 0>}
      localJointOffset=<4 0 0>
     }]
    simSpaceSettings {maxAngularAcceleration=100, maxAngularVelocity=10, simSpaceAngularAlpha=1}
   }
   n74 = Anim Dynamics - Bone: base_r_padChain1(ComponentPose=@n73, Alpha=1, ExternalForce="0,0,0") {
    angularDampingOverride=1
    angularSpringConstant=4500
    bAngularSpring=true
    bOverrideAngularDamping=true
    bUseGravityOverride=true
    gravityScale=0.35
    simulationSpace=RootRelative
    boundBone {boneName=base_r_padChain1}
    componentAppliedLinearAccClamp=<4000 4000 4000>
    componentLinearAccScale=<0.45 0.45 0.5>
    componentLinearVelScale=<0.45 0.45 0.5>
    gravityOverride {z=-980}
    physicsBodyDefinitions [{
      boundBone {boneName=base_r_padChain1}
      boxExtents=<10 12 10>
      constraintSetup {angularLimitsMax=<0 3 7>, angularLimitsMin=<0 -45 -7>, angularTarget=<1 0 0>}
      localJointOffset=<4 0 0>
     }]
    simSpaceSettings {maxAngularAcceleration=100, maxAngularVelocity=10, simSpaceAngularAlpha=1}
   }
   n75 = Anim Dynamics - Bone: base_r_padChain1A(ComponentPose=@n74, Alpha=1, ExternalForce="0,0,0") {
    angularDampingOverride=1
    angularSpringConstant=4500
    bAngularSpring=true
    bOverrideAngularDamping=true
    bUseGravityOverride=true
    gravityScale=0.35
    simulationSpace=RootRelative
    boundBone {boneName=base_r_padChain1A}
    componentAppliedLinearAccClamp=<4000 4000 4000>
    componentLinearAccScale=<0.45 0.45 0.5>
    componentLinearVelScale=<0.45 0.45 0.5>
    gravityOverride {z=-980}
    physicsBodyDefinitions [{
      boundBone {boneName=base_r_padChain1A}
      boxExtents=<10 12 10>
      constraintSetup {angularLimitsMax=<0 3 7>, angularLimitsMin=<0 -45 -7>, angularTarget=<1 0 0>}
      localJointOffset=<4 0 0>
     }]
    simSpaceSettings {maxAngularAcceleration=100, maxAngularVelocity=10, simSpaceAngularAlpha=1}
   }
   n76 = Anim Dynamics - Bone: base_r_padChain2(ComponentPose=@n75, Alpha=1, ExternalForce="0,0,0") {
    angularDampingOverride=1
    angularSpringConstant=4500
    bAngularSpring=true
    bOverrideAngularDamping=true
    bUseGravityOverride=true
    gravityScale=0.35
    simulationSpace=RootRelative
    boundBone {boneName=base_r_padChain2}
    componentAppliedLinearAccClamp=<4000 4000 4000>
    componentLinearAccScale=<0.45 0.45 0.5>
    componentLinearVelScale=<0.45 0.45 0.5>
    gravityOverride {z=-980}
    physicsBodyDefinitions [{
      boundBone {boneName=base_r_padChain2}
      boxExtents=<10 12 10>
      constraintSetup {angularLimitsMax=<0 3 7>, angularLimitsMin=<0 -45 -7>, angularTarget=<1 0 0>}
      localJointOffset=<4 0 0>
     }]
    simSpaceSettings {maxAngularAcceleration=100, maxAngularVelocity=10, simSpaceAngularAlpha=1}
   }
   n77 = Anim Dynamics - Bone: base_r_padChain2A(ComponentPose=@n76, Alpha=1, ExternalForce="0,0,0") {
    angularDampingOverride=1
    angularSpringConstant=4500
    bAngularSpring=true
    bOverrideAngularDamping=true
    bUseGravityOverride=true
    gravityScale=0.35
    simulationSpace=RootRelative
    boundBone {boneName=base_r_padChain2A}
    componentAppliedLinearAccClamp=<4000 4000 4000>
    componentLinearAccScale=<0.45 0.45 0.5>
    componentLinearVelScale=<0.45 0.45 0.5>
    gravityOverride {z=-980}
    physicsBodyDefinitions [{
      boundBone {boneName=base_r_padChain2A}
      boxExtents=<10 12 10>
      constraintSetup {angularLimitsMax=<0 3 7>, angularLimitsMin=<0 -45 -7>, angularTarget=<1 0 0>}
      localJointOffset=<4 0 0>
     }]
    simSpaceSettings {maxAngularAcceleration=100, maxAngularVelocity=10, simSpaceAngularAlpha=1}
   }
   n78 = Anim Dynamics - Bone: base_r_padChain3(ComponentPose=@n77, Alpha=1, ExternalForce="0,0,0") {
    angularDampingOverride=1
    angularSpringConstant=4500
    bAngularSpring=true
    bOverrideAngularDamping=true
    bUseGravityOverride=true
    gravityScale=0.35
    simulationSpace=RootRelative
    boundBone {boneName=base_r_padChain3}
    componentAppliedLinearAccClamp=<4000 4000 4000>
    componentLinearAccScale=<0.45 0.45 0.5>
    componentLinearVelScale=<0.45 0.45 0.5>
    gravityOverride {z=-980}
    physicsBodyDefinitions [{
      boundBone {boneName=base_r_padChain3}
      boxExtents=<10 12 10>
      constraintSetup {angularLimitsMax=<0 3 7>, angularLimitsMin=<0 -45 -7>, angularTarget=<1 0 0>}
      localJointOffset=<4 0 0>
     }]
    simSpaceSettings {maxAngularAcceleration=100, maxAngularVelocity=10, simSpaceAngularAlpha=1}
   }
   n79 = Anim Dynamics - Bone: base_r_padChain3A(ComponentPose=@n78, Alpha=1, ExternalForce="0,0,0") {
    angularDampingOverride=1
    angularSpringConstant=4500
    bAngularSpring=true
    bOverrideAngularDamping=true
    bUseGravityOverride=true
    gravityScale=0.35
    simulationSpace=RootRelative
    boundBone {boneName=base_r_padChain3A}
    componentAppliedLinearAccClamp=<4000 4000 4000>
    componentLinearAccScale=<0.45 0.45 0.5>
    componentLinearVelScale=<0.45 0.45 0.5>
    gravityOverride {z=-980}
    physicsBodyDefinitions [{
      boundBone {boneName=base_r_padChain3A}
      boxExtents=<10 12 10>
      constraintSetup {angularLimitsMax=<0 3 7>, angularLimitsMin=<0 -45 -7>, angularTarget=<1 0 0>}
      localJointOffset=<4 0 0>
     }]
    simSpaceSettings {maxAngularAcceleration=100, maxAngularVelocity=10, simSpaceAngularAlpha=1}
   }
   n80 = Anim Dynamics - Bone: base_r_padChain4(ComponentPose=@n79, Alpha=1, ExternalForce="0,0,0") {
    angularDampingOverride=1
    angularSpringConstant=4500
    bAngularSpring=true
    bOverrideAngularDamping=true
    bUseGravityOverride=true
    gravityScale=0.35
    simulationSpace=RootRelative
    boundBone {boneName=base_r_padChain4}
    componentAppliedLinearAccClamp=<4000 4000 4000>
    componentLinearAccScale=<0.45 0.45 0.5>
    componentLinearVelScale=<0.45 0.45 0.5>
    gravityOverride {z=-980}
    physicsBodyDefinitions [{
      boundBone {boneName=base_r_padChain4}
      boxExtents=<10 12 10>
      constraintSetup {angularLimitsMax=<0 3 7>, angularLimitsMin=<0 -45 -7>, angularTarget=<1 0 0>}
      localJointOffset=<4 0 0>
     }]
    simSpaceSettings {maxAngularAcceleration=100, maxAngularVelocity=10, simSpaceAngularAlpha=1}
   }
   n81 = Anim Dynamics - Bone: base_r_padChain5(ComponentPose=@n80, Alpha=1, ExternalForce="0,0,0") {
    angularDampingOverride=1
    angularSpringConstant=3000
    bAngularSpring=true
    bOverrideAngularDamping=true
    bUseGravityOverride=true
    gravityScale=0.35
    simulationSpace=RootRelative
    boundBone {boneName=base_r_padChain5}
    componentAppliedLinearAccClamp=<4000 4000 4000>
    componentLinearAccScale=<0.45 0.45 0.5>
    componentLinearVelScale=<0.45 0.45 0.5>
    gravityOverride {z=-980}
    physicsBodyDefinitions [{
      boundBone {boneName=base_r_padChain5}
      boxExtents=<10 12 10>
      constraintSetup {angularLimitsMax=<0 3 5>, angularLimitsMin=<0 -60 -7>, angularTarget=<1 0 0>}
      localJointOffset=<4 0 0>
     }]
    simSpaceSettings {maxAngularAcceleration=100, maxAngularVelocity=10, simSpaceAngularAlpha=1}
   }
   n82 = Anim Dynamics - Bone: base_r_padChain6(ComponentPose=@n81, Alpha=1, ExternalForce="0,0,0") {
    angularDampingOverride=1
    angularSpringConstant=3000
    bAngularSpring=true
    bOverrideAngularDamping=true
    bUseGravityOverride=true
    gravityScale=0.35
    simulationSpace=RootRelative
    boundBone {boneName=base_r_padChain6}
    componentAppliedLinearAccClamp=<4000 4000 4000>
    componentLinearAccScale=<0.45 0.45 0.5>
    componentLinearVelScale=<0.45 0.45 0.5>
    gravityOverride {z=-980}
    physicsBodyDefinitions [{
      boundBone {boneName=base_r_padChain6}
      boxExtents=<10 12 10>
      constraintSetup {angularLimitsMax=<0 3 5>, angularLimitsMin=<0 -60 -7>, angularTarget=<1 0 0>}
      localJointOffset=<4 0 0>
     }]
    simSpaceSettings {maxAngularAcceleration=100, maxAngularVelocity=10, simSpaceAngularAlpha=1}
   }
   n83 = Anim Dynamics - Bone: base_r_padChain7(ComponentPose=@n82, Alpha=1, ExternalForce="0,0,0") {
    angularDampingOverride=1
    angularSpringConstant=3000
    bAngularSpring=true
    bOverrideAngularDamping=true
    bUseGravityOverride=true
    gravityScale=0.35
    simulationSpace=RootRelative
    boundBone {boneName=base_r_padChain7}
    componentAppliedLinearAccClamp=<4000 4000 4000>
    componentLinearAccScale=<0.45 0.45 0.5>
    componentLinearVelScale=<0.45 0.45 0.5>
    gravityOverride {z=-980}
    physicsBodyDefinitions [{
      boundBone {boneName=base_r_padChain7}
      boxExtents=<10 12 10>
      constraintSetup {angularLimitsMax=<0 3 15>, angularLimitsMin=<0 -60 -7>, angularTarget=<1 0 0>}
      localJointOffset=<4 0 0>
     }]
    simSpaceSettings {maxAngularAcceleration=100, maxAngularVelocity=10, simSpaceAngularAlpha=1}
   }
   n84 = Anim Dynamics - Bone: base_r_padChain8(ComponentPose=@n83, Alpha=1, ExternalForce="0,0,0") {
    angularDampingOverride=1
    angularSpringConstant=3000
    bAngularSpring=true
    bOverrideAngularDamping=true
    bUseGravityOverride=true
    gravityScale=0.35
    simulationSpace=RootRelative
    boundBone {boneName=base_r_padChain8}
    componentAppliedLinearAccClamp=<4000 4000 4000>
    componentLinearAccScale=<0.45 0.45 0.5>
    componentLinearVelScale=<0.45 0.45 0.5>
    gravityOverride {z=-980}
    physicsBodyDefinitions [{
      boundBone {boneName=base_r_padChain8}
      boxExtents=<10 12 10>
      constraintSetup {angularLimitsMax=<0 3 15>, angularLimitsMin=<0 -60 -7>, angularTarget=<1 0 0>}
      localJointOffset=<4 0 0>
     }]
    simSpaceSettings {maxAngularAcceleration=100, maxAngularVelocity=10, simSpaceAngularAlpha=1}
   }
   n85 = Anim Dynamics - Bone: base_r_padChain9(ComponentPose=@n84, Alpha=1, ExternalForce="0,0,0") {
    angularDampingOverride=1
    angularSpringConstant=3000
    bAngularSpring=true
    bOverrideAngularDamping=true
    bUseGravityOverride=true
    gravityScale=0.35
    simulationSpace=RootRelative
    boundBone {boneName=base_r_padChain9}
    componentAppliedLinearAccClamp=<4000 4000 4000>
    componentLinearAccScale=<0.45 0.45 0.5>
    componentLinearVelScale=<0.45 0.45 0.5>
    gravityOverride {z=-980}
    physicsBodyDefinitions [{
      boundBone {boneName=base_r_padChain9}
      boxExtents=<10 12 10>
      constraintSetup {angularLimitsMax=<0 3 15>, angularLimitsMin=<0 -60 -7>, angularTarget=<1 0 0>}
      localJointOffset=<4 0 0>
     }]
    simSpaceSettings {maxAngularAcceleration=100, maxAngularVelocity=10, simSpaceAngularAlpha=1}
   }
   n86 = Anim Dynamics - Bone: base_r_padChain10(ComponentPose=@n85, Alpha=1, ExternalForce="0,0,0") {
    angularDampingOverride=1
    angularSpringConstant=3000
    bAngularSpring=true
    bOverrideAngularDamping=true
    bUseGravityOverride=true
    gravityScale=0.35
    simulationSpace=RootRelative
    boundBone {boneName=base_r_padChain10}
    componentAppliedLinearAccClamp=<4000 4000 4000>
    componentLinearAccScale=<0.45 0.45 0.5>
    componentLinearVelScale=<0.45 0.45 0.5>
    gravityOverride {z=-980}
    physicsBodyDefinitions [{
      boundBone {boneName=base_r_padChain10}
      boxExtents=<10 12 10>
      constraintSetup {angularLimitsMax=<0 3 15>, angularLimitsMin=<0 -60 -7>, angularTarget=<1 0 0>}
      localJointOffset=<4 0 0>
     }]
    simSpaceSettings {maxAngularAcceleration=100, maxAngularVelocity=10, simSpaceAngularAlpha=1}
   }
   n87 = Anim Dynamics - Bone: base_r_padChain11(ComponentPose=@n86, Alpha=1, ExternalForce="0,0,0") {
    angularDampingOverride=1
    angularSpringConstant=3000
    bAngularSpring=true
    bOverrideAngularDamping=true
    bUseGravityOverride=true
    gravityScale=0.35
    simulationSpace=RootRelative
    boundBone {boneName=base_r_padChain11}
    componentAppliedLinearAccClamp=<4000 4000 4000>
    componentLinearAccScale=<0.45 0.45 0.5>
    componentLinearVelScale=<0.45 0.45 0.5>
    gravityOverride {z=-980}
    physicsBodyDefinitions [{
      boundBone {boneName=base_r_padChain11}
      boxExtents=<10 12 10>
      constraintSetup {angularLimitsMax=<0 3 7>, angularLimitsMin=<0 -60 -7>, angularTarget=<1 0 0>}
      localJointOffset=<4 0 0>
     }]
    simSpaceSettings {maxAngularAcceleration=100, maxAngularVelocity=10, simSpaceAngularAlpha=1}
   }
   n88 = Anim Dynamics - Bone: base_r_padChain12(ComponentPose=@n87, Alpha=1, ExternalForce="0,0,0") {
    angularDampingOverride=1
    angularSpringConstant=3000
    bAngularSpring=true
    bOverrideAngularDamping=true
    bUseGravityOverride=true
    gravityScale=0.35
    simulationSpace=RootRelative
    boundBone {boneName=base_r_padChain12}
    componentAppliedLinearAccClamp=<4000 4000 4000>
    componentLinearAccScale=<0.45 0.45 0.5>
    componentLinearVelScale=<0.45 0.45 0.5>
    gravityOverride {z=-980}
    physicsBodyDefinitions [{
      boundBone {boneName=base_r_padChain12}
      boxExtents=<10 12 10>
      constraintSetup {angularLimitsMax=<0 3 7>, angularLimitsMin=<0 -60 -7>, angularTarget=<1 0 0>}
      localJointOffset=<4 0 0>
     }]
    simSpaceSettings {maxAngularAcceleration=100, maxAngularVelocity=10, simSpaceAngularAlpha=1}
   }
   n89 = Anim Dynamics - Bone: base_r_padChain13(ComponentPose=@n88, Alpha=1, ExternalForce="0,0,0") {
    angularDampingOverride=1
    angularSpringConstant=3000
    bAngularSpring=true
    bOverrideAngularDamping=true
    bUseGravityOverride=true
    gravityScale=0.35
    simulationSpace=RootRelative
    boundBone {boneName=base_r_padChain13}
    componentAppliedLinearAccClamp=<4000 4000 4000>
    componentLinearAccScale=<0.45 0.45 0.5>
    componentLinearVelScale=<0.45 0.45 0.5>
    gravityOverride {z=-980}
    physicsBodyDefinitions [{
      boundBone {boneName=base_r_padChain13}
      boxExtents=<10 12 10>
      constraintSetup {angularLimitsMax=<0 3 7>, angularLimitsMin=<0 -60 -7>, angularTarget=<1 0 0>}
      localJointOffset=<4 0 0>
     }]
    simSpaceSettings {maxAngularAcceleration=100, maxAngularVelocity=10, simSpaceAngularAlpha=1}
   }
   n90 = Anim Dynamics - Bone: base_r_padChain14(ComponentPose=@n89, Alpha=1, ExternalForce="0,0,0") {
    angularDampingOverride=1
    angularSpringConstant=3000
    bAngularSpring=true
    bOverrideAngularDamping=true
    bUseGravityOverride=true
    gravityScale=0.35
    simulationSpace=RootRelative
    boundBone {boneName=base_r_padChain14}
    componentAppliedLinearAccClamp=<4000 4000 4000>
    componentLinearAccScale=<0.45 0.45 0.5>
    componentLinearVelScale=<0.45 0.45 0.5>
    gravityOverride {z=-980}
    physicsBodyDefinitions [{
      boundBone {boneName=base_r_padChain14}
      boxExtents=<10 12 10>
      constraintSetup {angularLimitsMax=<0 3 7>, angularLimitsMin=<0 -60 -7>, angularTarget=<1 0 0>}
      localJointOffset=<4 0 0>
     }]
    simSpaceSettings {maxAngularAcceleration=100, maxAngularVelocity=10, simSpaceAngularAlpha=1}
   }
   n91 = Anim Dynamics - Bone: base_r_padChain15(ComponentPose=@n90, Alpha=1, ExternalForce="0,0,0") {
    angularDampingOverride=1
    angularSpringConstant=3000
    bAngularSpring=true
    bOverrideAngularDamping=true
    bUseGravityOverride=true
    gravityScale=0.35
    simulationSpace=RootRelative
    boundBone {boneName=base_r_padChain15}
    componentAppliedLinearAccClamp=<4000 4000 4000>
    componentLinearAccScale=<0.45 0.45 0.5>
    componentLinearVelScale=<0.45 0.45 0.5>
    gravityOverride {z=-980}
    physicsBodyDefinitions [{
      boundBone {boneName=base_r_padChain15}
      boxExtents=<10 12 10>
      constraintSetup {angularLimitsMax=<0 3 7>, angularLimitsMin=<0 -60 -7>, angularTarget=<1 0 0>}
      localJointOffset=<4 0 0>
     }]
    simSpaceSettings {maxAngularAcceleration=100, maxAngularVelocity=10, simSpaceAngularAlpha=1}
   }
   n92 = Anim Dynamics - Bone: base_r_padChain16(ComponentPose=@n91, Alpha=1, ExternalForce="0,0,0") {
    angularDampingOverride=1
    angularSpringConstant=3000
    bAngularSpring=true
    bOverrideAngularDamping=true
    bUseGravityOverride=true
    gravityScale=0.35
    simulationSpace=RootRelative
    boundBone {boneName=base_r_padChain16}
    componentAppliedLinearAccClamp=<4000 4000 4000>
    componentLinearAccScale=<0.45 0.45 0.5>
    componentLinearVelScale=<0.45 0.45 0.5>
    gravityOverride {z=-980}
    physicsBodyDefinitions [{
      boundBone {boneName=base_r_padChain16}
      boxExtents=<10 12 10>
      constraintSetup {angularLimitsMax=<0 3 7>, angularLimitsMin=<0 -60 -7>, angularTarget=<1 0 0>}
      localJointOffset=<4 0 0>
     }]
    simSpaceSettings {maxAngularAcceleration=100, maxAngularVelocity=10, simSpaceAngularAlpha=1}
   }
   n93 = Anim Dynamics - Bone: base_r_padChain17(ComponentPose=@n92, Alpha=1, ExternalForce="0,0,0") {
    angularDampingOverride=1
    angularSpringConstant=3000
    bAngularSpring=true
    bOverrideAngularDamping=true
    bUseGravityOverride=true
    gravityScale=0.35
    simulationSpace=RootRelative
    boundBone {boneName=base_r_padChain17}
    componentAppliedLinearAccClamp=<4000 4000 4000>
    componentLinearAccScale=<0.45 0.45 0.5>
    componentLinearVelScale=<0.45 0.45 0.5>
    gravityOverride {z=-980}
    physicsBodyDefinitions [{
      boundBone {boneName=base_r_padChain17}
      boxExtents=<10 12 10>
      constraintSetup {angularLimitsMax=<0 3 7>, angularLimitsMin=<0 -60 -7>, angularTarget=<1 0 0>}
      localJointOffset=<4 0 0>
     }]
    simSpaceSettings {maxAngularAcceleration=100, maxAngularVelocity=10, simSpaceAngularAlpha=1}
   }
   n94 = Anim Dynamics - Bone: base_r_padChain18(ComponentPose=@n93, Alpha=1, ExternalForce="0,0,0") {
    angularDampingOverride=1
    angularSpringConstant=3000
    bAngularSpring=true
    bOverrideAngularDamping=true
    bUseGravityOverride=true
    gravityScale=0.35
    simulationSpace=RootRelative
    boundBone {boneName=base_r_padChain18}
    componentAppliedLinearAccClamp=<4000 4000 4000>
    componentLinearAccScale=<0.45 0.45 0.5>
    componentLinearVelScale=<0.45 0.45 0.5>
    gravityOverride {z=-980}
    physicsBodyDefinitions [{
      boundBone {boneName=base_r_padChain18}
      boxExtents=<10 12 10>
      constraintSetup {angularLimitsMax=<0 3 7>, angularLimitsMin=<0 -60 -7>, angularTarget=<1 0 0>}
      localJointOffset=<4 0 0>
     }]
    simSpaceSettings {maxAngularAcceleration=100, maxAngularVelocity=10, simSpaceAngularAlpha=1}
   }
   Save cached pose 'Second_Anim_Dynamics'(Pose=Component To Local(ComponentPose=@n94)) {cacheName=Second_Anim_Dynamics}
  ]}
 EventGraph {nodes [
   Event BlueprintUpdateAnimation()
   > n1 = Sequence!(execute) -> (then_0, then_1, then_2:Exec, then_3:Exec, then_4:Exec, then_5:Exec, then_6:Exec, then_7:Exec, then_8:Exec, then_9:Exec, then_10:Exec)
   Set Is Custom Idle Playing Float(execute=@n1.then_2, Is Custom Idle Playing Float:double=GetCurveValue(self=GetOwnerFrontEndAnimInstance(), CurveName=Frontend_Idle_Bypass)) -> (Output_Get:double)
   Set Is_Parachute_Open(execute=@n1.then_5, Is_Parachute_Open:double="To Float (Boolean)"(InBool=Get bIsParachuteOpened())) -> (Output_Get:double)
   n2 = Get bIsKneeSliding()
   n3 = Branch(execute=@n1.then_7, Condition=AND Boolean(A=Get bIsCrouching(), B=NOT Boolean(A=@n2)))
   > Set Crouching(Crouching:double=1) -> (Output_Get:double)
   Set Crouching(execute=@n3.else, Crouching=0)
   n4 = Get bIsMoving2D()
   n5 = Get bIsCrouchSprinting()
   n6 = Branch(execute=@n1.then_4, Condition=AND Boolean!(A=Get bIsCrouching(), B=NOT Boolean(A=@n5), C:bool=NOT Boolean(A=@n4)) -> (ReturnValue))
   > Set Crouching_not_Sprin(Crouching_not_Sprin:double=1) -> (Output_Get:double)
   Set Crouching_not_Sprin(execute=@n6.else, Crouching_not_Sprin=0)
   n7 = Branch(execute=@n1.then_8, Condition=AND Boolean!(A=Get bIsTacticalSprinting(), B=Get bIsJumping()) -> (ReturnValue))
   > Delay(Duration=0.1)
   > Set isTacticalSprinting(isTacticalSprinting:double=1) -> (Output_Get:double)
   Set isTacticalSprinting(execute=@n7.else, isTacticalSprinting=0)
   n8 = Sequence!(execute=@n1.then_10) -> (then_0, then_1, then_2, then_3, then_4, then_5)
   > n9 = Branch(Condition=AND Boolean(A=Get bIsCrouching(), B=Get bIsMoving2D()))
   > Set CrouchingSprinting(CrouchingSprinting:double=1) -> (Output_Get:double)
   Set CrouchingSprinting(execute=@n9.else, CrouchingSprinting=0)
   n10 = GetAnimInstance(self=GetOwnerSkeletalMeshComponent())
   Is Valid(exec=@n8.then_3, InputObject=@n10)
   > n11 = Set isBreakinDance(isBreakinDance:bool=Montage_IsPlaying(self=@n10, Montage=/Game/Animation/Game/MainPlayer/Emotes/Emote_Breakdance_Female_M)) -> (Output_Get:bool)
   > n12 = Branch(Condition=@n11)
   > Set is BreakDancing(is BreakDancing:double=1) -> (Output_Get:double)
   Set is BreakDancing(execute=@n12.else, is BreakDancing=0)
   n13 = GetAnimInstance(self=GetOwnerSkeletalMeshComponent())
   Is Valid(exec=@n8.then_5, InputObject=@n13)
   > n14 = Set BalletSpin(BalletSpin:bool=Montage_IsPlaying(self=@n13, Montage=/BRCosmetics/Animation/Game/MainPlayer/Emotes/BalletSpin/Emote_BalletSpin_CMF_M)) -> (Output_Get:bool)
   > n15 = Branch(Condition=@n14)
   > Set IsBallet_Spin(IsBallet_Spin:double=1) -> (Output_Get:double)
   Set IsBallet_Spin(execute=@n15.else, IsBallet_Spin=0)
   n16 = GetAnimInstance(self=GetOwnerSkeletalMeshComponent())
   Is Valid(exec=@n8.then_4, InputObject=@n16)
   > n17 = Set LobbyIddle(LobbyIddle:bool=Montage_IsPlaying(self=@n16, Montage=/BRCosmetics/Animation/Game/MainPlayer/Menu/BR/CustomSkins/_s39/LowerDrop_F/LowerDrop_Female_Idle_Montage)) -> (Output_Get:bool)
   n18 = Get bIsPlayingEmote()
   > n19 = Branch(Condition=AND Boolean(A=@n17, B=NOT Boolean(A=@n18)))
   > Set Lobby Iddle(Lobby Iddle:double=1) -> (Output_Get:double)
   Set Lobby Iddle(execute=@n19.else, Lobby Iddle=0)
   n20 = Branch(execute=@n8.then_1, Condition=Get bIsSkydiveFloating())
   > Set Is_skyFloating(Is_skyFloating:double=1) -> (Output_Get:double)
   Set Is_skyFloating(execute=@n20.else, Is_skyFloating=0)
   n21 = GetAnimInstance(self=GetOwnerSkeletalMeshComponent())
   Is Valid(exec=@n8.then_2, InputObject=@n21)
   > n22 = Set is Celebrating(is Celebrating:bool=Montage_IsPlaying(self=@n21, Montage=/BRCosmetics/Animation/Game/MainPlayer/Emotes/Emote_Celebration_Female_M)) -> (Output_Get:bool)
   > n23 = Branch(Condition=@n22)
   > Set isCelebrating(isCelebrating:double=1) -> (Output_Get:double)
   Set isCelebrating(execute=@n23.else, isCelebrating=0)
   n24 = PlayerMovement(execute=@n1.then_6, AnimInstance=Self-Reference(), XDotProductCoefficient=1, YDotProductCoefficient=1, ZDotProductCoefficient=1)
   n25 = BreakVector(InVec=MakeVector(X=@n24, Y=@n24.YMovement))
   > Set strafeLeft float(strafeLeft float:double="Absolute (Float)"(A="Clamp (Float)"(Value=@n25, Min=-1, Max=0))) -> (Output_Get:double)
   > Set strafeRight float(strafeRight float:double="Clamp (Float)"(Value=@n25, Min=0, Max=1)) -> (Output_Get:double)
   > Set moveForward(moveForward:double="Clamp (Float)"(Value=@n25.Y, Min=0, Max=1)) -> (Output_Get:double)
   > Set moveBackward(moveBackward:double="Absolute (Float)"(A="Clamp (Float)"(Value=@n25.Y, Min=-1, Max=0))) -> (Output_Get:double)
   n26 = Event BlueprintProcessAnimInputEvent()
   n27 = Branch(
    execute=@n1.then_3
    Condition=OR Boolean!(
     A=IsNamed(self=@n26.AnimInputEvent, InEventName=aie.Player.Zipline.Entered)
     B=Get bIsOnGrindRail()
     C:bool=Get bIsInFrontEnd()
     D:bool=Get bIsPlayingEmote()
    ) -> (ReturnValue)
   )
   > Set OnZipline(OnZipline:bool=1) -> (Output_Get:bool)
   > Set ZiplineAlpha(ZiplineAlpha:double=0.2) -> (Output_Get:double)
   Set ZiplineAlpha(execute=@n27.else, ZiplineAlpha=1)
   n28 = GetAnimInstance(self=GetOwnerSkeletalMeshComponent())
   Is Valid(exec=@n1.then_1, InputObject=@n28)
   > Set BalletSpinEmote(BalletSpinEmote:bool=NOT Boolean(A=Montage_IsPlaying(self=@n28, Montage=&Emote_BalletSpin_CMF_M))) -> (Output_Get:bool)
   n29 = "Gravity Override State Function (Parameter Asset) (Nativized)"(
    execute=@n1.then_0
    CharacterPartAnimInstance=Self-Reference()
    GameTimeInSeconds=GetGameTimeInSeconds()
    DebugJointName=spine_05
    Parameters=&/F_MED_LowerDrop_Parameters
   )
   > Set Gravity Override(Gravity Override=@n29)
   > Set RBAN Settings Out(RBAN Settings Out=@n29.OutRigidBodyAnimNodeParameters)
   n30 = OR Boolean!(A=Get bIsSkydiveFloating(), B=Get bIsSkyDiving()) -> (ReturnValue)
   n31 = Get bIsJumping()
   n32 = Get bIsFalling()
   n33 = Branch(execute=@n1.then_9, Condition=AND Boolean(A=OR Boolean(A=@n31, B=@n32), B=NOT Boolean(A=@n30)))
   > Set Is Jumping(Is Jumping:double=1) -> (Output_Get:double)
   Set Is Jumping(execute=@n33.else, Is Jumping=0)
   Event OnBeginSliding() {is=title K2Node_Event}
   > Set isRegular_Sliding(isRegular_Sliding:double=1) -> (Output_Get:double)
   Event OnEndSliding() {is=title K2Node_Event}
   > Set isRegular_Sliding(isRegular_Sliding=0)
   Event OnBeginBaseballSliding() {is=title K2Node_Event}
   > Set IsBaseball_Sliding(IsBaseball_Sliding:double=1) -> (Output_Get:double)
   Event OnWeaponChanged()
   > Set isAR Equipped(isAR Equipped:bool=false) -> (Output_Get:bool)
   > Set isPistol Equipped(isPistol Equipped:bool=false) -> (Output_Get:bool)
   > Set Is_outLaw_SG(Is_outLaw_SG:bool=false) -> (Output_Get:bool)
   > Set SMG_SG(SMG_SG:bool=false) -> (Output_Get:bool)
   > n34 = Switch on EFortWeaponCoreAnimation!(execute, Selection:EFortWeaponCoreAnimation Enum=Get WeaponCoreAnim!() -> (WeaponCoreAnim)) -> (
    Melee:Exec
    Pistol:Exec
    Shotgun:Exec
    PaperBlueprint:Exec
    Rifle:Exec
    MeleeOneHand:Exec
    MachinePistol:Exec
    RocketLauncher:Exec
    GrenadeLauncher:Exec
    GoingCommando:Exec
    AssaultRifle:Exec
    TacticalShotgun:Exec
    SniperRifle:Exec
    TrapPlacement:Exec
    ShoulderLauncher:Exec
    AbilityDecoTool:Exec
    Crossbow:Exec
    C4:Exec
    RemoteControl:Exec
    DualWield:Exec
    AR_BullPup:Exec
    AR_ForwardGrip:Exec
    MedPackPaddles:Exec
    SMG_P90:Exec
    AR_DrumGun:Exec
    Consumable_Small:Exec
    Consumable_Large:Exec
    Balloon:Exec
    MountedTurret:Exec
    CreativeTool:Exec
    ExplosiveBow:Exec
    AshtonIndigo:Exec
    AshtonChicago:Exec
    MeleeDualWield:Exec
    Unarmed:Exec
   )
   Set isAR Equipped(execute=@n34.AssaultRifle, isAR Equipped=true)
   Set isPistol Equipped(execute=@n34.Pistol, isPistol Equipped=true)
   Set SMG_SG(execute=@n34.SMG_P90, SMG_SG=true)
   Set Is_outLaw_SG(execute=@n34.TacticalShotgun, Is_outLaw_SG=true)
   Event OnEndBaseballSliding() {is=title K2Node_Event}
   > Set IsBaseball_Sliding(IsBaseball_Sliding=0)
  ]}
]
variables [
 BalletSpin {type=boolean}
 BalletSpinEmote {type=boolean}
 Crouching {type=number}
 Crouching_not_Sprin {type=number}
 Crouching_sides {type=number}
 CrouchingSprinting {type=number}
 Gravity Override {type=Vector}
 is BreakDancing {type=number}
 is Celebrating {type=boolean}
 Is Custom Idle Playing Float {type=number}
 Is Jumping {type=number}
 Is_outLaw_SG {type=boolean}
 Is_Parachute_Open {type=number}
 Is_skyFloating {type=number}
 isAR Equipped {type=boolean}
 IsBallet_Spin {type=number}
 IsBaseball_Sliding {type=number}
 isBreakinDance {type=boolean}
 isCelebrating {type=number}
 isKeytar_Ballerina {type=number}
 isPistol Equipped {type=boolean}
 isRegular_Sliding {type=number}
 isTacticalSprinting {type=number}
 Keytar_ballerina {type=boolean}
 Lobby Iddle {type=number}
 LobbyIddle {type=boolean}
 moveBackward {type=number}
 moveForward {type=number}
 OnZipline {type=boolean}
 RBAN Settings Out {type=FortRigidBodyAnimNodeParameters}
 SMG_SG {type=boolean}
 strafeLeft float {type=number}
 strafeRight float {type=number}
 ZiplineAlpha {type=number, default=1}
]
