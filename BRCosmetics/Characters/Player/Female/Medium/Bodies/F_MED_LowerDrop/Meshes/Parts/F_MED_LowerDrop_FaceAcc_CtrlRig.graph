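# Control Rig for the LowerDrop female face accessory, which makes three long hair strands, other hair and front-chain bones follow body motion.
# Weighted parent constraints blend strand control nulls from head down to pelvis, while twist-bone and B-spline IK functions pose the strands.
# Spherical pose readers and delta-rotation functions weight additive correctives for head, spine, clavicle and arm poses, some gated by gameplay flags.
# It also writes head-look curves during FE Idle.
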ControlRigBlueprint /BRCosmetics/Characters/Player/Female/Medium/Bodies/F_MED_LowerDrop/Meshes/Parts/F_MED_LowerDrop_FaceAcc_CtrlRig
graphs [
 CollapseNode_ContainedGraph {nodes [
   RigVMFunctionEntryNode()
   > &RigUnit_ParentConstraint(
    AdvancedSettings=(RotationOrderForFilter=XZY)
    Child=(Type=Null,Name="dyn_L_hair_1_ctrl_null")
    Filter=(TranslationFilter=(bX=True,bY=True,bZ=True),RotationFilter=(bX=True,bY=True,bZ=True),ScaleFilter=(bX=False,bY=False,bZ=False))
    Parents=((Item=(Type=Bone,Name="head"),Weight=1))
    Weight=1
   )
   > &RigUnit_ParentConstraint(
    AdvancedSettings=(RotationOrderForFilter=XZY)
    Child=(Type=Null,Name="dyn_C_hair_1_ctrl_null")
    Filter=(TranslationFilter=(bX=True,bY=True,bZ=True),RotationFilter=(bX=True,bY=True,bZ=True),ScaleFilter=(bX=False,bY=False,bZ=False))
    Parents=((Item=(Type=Bone,Name="head"),Weight=1))
    Weight=1
   )
   > &RigUnit_ParentConstraint(
    AdvancedSettings=(RotationOrderForFilter=XZY)
    Child=(Type=Null,Name="dyn_R_hair_1_ctrl_null")
    Filter=(TranslationFilter=(bX=True,bY=True,bZ=True),RotationFilter=(bX=True,bY=True,bZ=True),ScaleFilter=(bX=False,bY=False,bZ=False))
    Parents=((Item=(Type=Bone,Name="head"),Weight=1))
    Weight=1
   )
   > &RigUnit_ParentConstraint(
    AdvancedSettings=(RotationOrderForFilter=XZY)
    Child=(Type=Null,Name="dyn_L_hair_2_ctrl_null")
    Filter=(TranslationFilter=(bX=True,bY=True,bZ=True),RotationFilter=(bX=True,bY=True,bZ=True),ScaleFilter=(bX=False,bY=False,bZ=False))
    Parents=((Item=(Type=Bone,Name="head"),Weight=0.5),(Item=(Type=Bone,Name="spine_05"),Weight=0.5))
    Weight=1
   )
   > &RigUnit_ParentConstraint(
    AdvancedSettings=(RotationOrderForFilter=XZY)
    Child=(Type=Null,Name="dyn_C_hair_2_ctrl_null")
    Filter=(TranslationFilter=(bX=True,bY=True,bZ=True),RotationFilter=(bX=True,bY=True,bZ=True),ScaleFilter=(bX=False,bY=False,bZ=False))
    Parents=((Item=(Type=Bone,Name="head"),Weight=0.5),(Item=(Type=Bone,Name="spine_05"),Weight=0.5))
    Weight=1
   )
   > &RigUnit_ParentConstraint(
    AdvancedSettings=(RotationOrderForFilter=XZY)
    Child=(Type=Null,Name="dyn_R_hair_2_ctrl_null")
    Filter=(TranslationFilter=(bX=True,bY=True,bZ=True),RotationFilter=(bX=True,bY=True,bZ=True),ScaleFilter=(bX=False,bY=False,bZ=False))
    Parents=((Item=(Type=Bone,Name="head"),Weight=0.5),(Item=(Type=Bone,Name="spine_05"),Weight=0.5))
    Weight=1
   )
   > &RigUnit_ParentConstraint(
    AdvancedSettings=(RotationOrderForFilter=XZY)
    Child=(Type=Null,Name="dyn_L_hair_3_ctrl_null")
    Filter=(TranslationFilter=(bX=True,bY=True,bZ=True),RotationFilter=(bX=True,bY=True,bZ=True),ScaleFilter=(bX=False,bY=False,bZ=False))
    Parents=((Item=(Type=Bone,Name="upperarm_twist_01_l"),Weight=0.05),(Item=(Type=Bone,Name="spine_05"),Weight=0.35),(Item=(Type=Bone,Name="clavicle_l"),Weight=0.6))
    Weight=1
   )
   > &RigUnit_ParentConstraint(
    AdvancedSettings=(RotationOrderForFilter=XZY)
    Child=(Type=Null,Name="dyn_C_hair_3_ctrl_null")
    Filter=(TranslationFilter=(bX=True,bY=True,bZ=True),RotationFilter=(bX=True,bY=True,bZ=True),ScaleFilter=(bX=False,bY=False,bZ=False))
    Parents=((Item=(Type=Bone,Name="clavicle_l"),Weight=0.05),(Item=(Type=Bone,Name="clavicle_r"),Weight=0.05),(Item=(Type=Bone,Name="spine_05"),Weight=0.9))
    Weight=1
   )
   > &RigUnit_ParentConstraint(
    AdvancedSettings=(RotationOrderForFilter=XZY)
    Child=(Type=Null,Name="dyn_R_hair_3_ctrl_null")
    Filter=(TranslationFilter=(bX=True,bY=True,bZ=True),RotationFilter=(bX=True,bY=True,bZ=True),ScaleFilter=(bX=False,bY=False,bZ=False))
    Parents=((Item=(Type=Bone,Name="upperarm_twist_01_r"),Weight=0.05),(Item=(Type=Bone,Name="spine_05"),Weight=0.35),(Item=(Type=Bone,Name="clavicle_r"),Weight=0.6))
    Weight=1
   )
   > &RigUnit_ParentConstraint(
    AdvancedSettings=(RotationOrderForFilter=XZY)
    Child=(Type=Null,Name="dyn_L_hair_4_ctrl_null")
    Filter=(TranslationFilter=(bX=True,bY=True,bZ=True),RotationFilter=(bX=True,bY=True,bZ=True),ScaleFilter=(bX=False,bY=False,bZ=False))
    Parents="((Item=(Type=Bone,Name=\"spine_02\"),Weight=0.0014),(Item=(Type=Bone,Name=\"spine_03\"),Weight=0.0235),(Item=(Type=Bone,Name=\"spine_04\"),Weight=0.0379),(Item=(Type=Bone,Name=\"upperarm_twist_01_l\"),Weight=0.0733),(Item=(Type=Bone,Name=\"spine_05\"),Weight=0.3523) ...(truncated)"
    Weight=1
   )
   > &RigUnit_ParentConstraint(
    AdvancedSettings=(RotationOrderForFilter=XZY)
    Child=(Type=Null,Name="dyn_C_hair_4_ctrl_null")
    Filter=(TranslationFilter=(bX=True,bY=True,bZ=True),RotationFilter=(bX=True,bY=True,bZ=True),ScaleFilter=(bX=False,bY=False,bZ=False))
    Parents="((Item=(Type=Bone,Name=\"spine_01\"),Weight=0.0001),(Item=(Type=Bone,Name=\"spine_02\"),Weight=0.0068),(Item=(Type=Bone,Name=\"clavicle_l\"),Weight=0.063),(Item=(Type=Bone,Name=\"clavicle_r\"),Weight=0.0705),(Item=(Type=Bone,Name=\"spine_03\"),Weight=0.1021),(Item=( ...(truncated)"
    Weight=1
   )
   > &RigUnit_ParentConstraint(
    AdvancedSettings=(RotationOrderForFilter=XZY)
    Child=(Type=Null,Name="dyn_R_hair_4_ctrl_null")
    Filter=(TranslationFilter=(bX=True,bY=True,bZ=True),RotationFilter=(bX=True,bY=True,bZ=True),ScaleFilter=(bX=False,bY=False,bZ=False))
    Parents="((Item=(Type=Bone,Name=\"spine_02\"),Weight=0.0014),(Item=(Type=Bone,Name=\"spine_03\"),Weight=0.0235),(Item=(Type=Bone,Name=\"spine_04\"),Weight=0.038),(Item=(Type=Bone,Name=\"upperarm_twist_01_r\"),Weight=0.0733),(Item=(Type=Bone,Name=\"spine_05\"),Weight=0.3528), ...(truncated)"
    Weight=1
   )
   > &RigUnit_ParentConstraint(
    AdvancedSettings=(RotationOrderForFilter=XZY)
    Child=(Type=Null,Name="dyn_L_hair_5_ctrl_null")
    Filter=(TranslationFilter=(bX=True,bY=True,bZ=True),RotationFilter=(bX=True,bY=True,bZ=True),ScaleFilter=(bX=False,bY=False,bZ=False))
    Parents=((Item=(Type=Bone,Name="clavicle_l"),Weight=0.1),(Item=(Type=Bone,Name="spine_04"),Weight=0.15),(Item=(Type=Bone,Name="spine_05"),Weight=0.15),(Item=(Type=Bone,Name="spine_02"),Weight=0.2),(Item=(Type=Bone,Name="spine_03"),Weight=0.4))
    Weight=1
   )
   > &RigUnit_ParentConstraint(
    AdvancedSettings=(RotationOrderForFilter=XZY)
    Child=(Type=Null,Name="dyn_C_hair_5_ctrl_null")
    Filter=(TranslationFilter=(bX=True,bY=True,bZ=True),RotationFilter=(bX=True,bY=True,bZ=True),ScaleFilter=(bX=False,bY=False,bZ=False))
    Parents=((Item=(Type=Bone,Name="spine_04"),Weight=0.15),(Item=(Type=Bone,Name="spine_05"),Weight=0.1),(Item=(Type=Bone,Name="spine_02"),Weight=0.25),(Item=(Type=Bone,Name="spine_03"),Weight=0.5))
    Weight=1
   )
   > &RigUnit_ParentConstraint(
    AdvancedSettings=(RotationOrderForFilter=XZY)
    Child=(Type=Null,Name="dyn_R_hair_5_ctrl_null")
    Filter=(TranslationFilter=(bX=True,bY=True,bZ=True),RotationFilter=(bX=True,bY=True,bZ=True),ScaleFilter=(bX=False,bY=False,bZ=False))
    Parents=((Item=(Type=Bone,Name="clavicle_r"),Weight=0.1),(Item=(Type=Bone,Name="spine_04"),Weight=0.15),(Item=(Type=Bone,Name="spine_05"),Weight=0.15),(Item=(Type=Bone,Name="spine_02"),Weight=0.2),(Item=(Type=Bone,Name="spine_03"),Weight=0.4))
    Weight=1
   )
   > &RigUnit_ParentConstraint(
    AdvancedSettings=(RotationOrderForFilter=XZY)
    Child=(Type=Null,Name="dyn_L_hair_6_ctrl_null")
    Filter=(TranslationFilter=(bX=True,bY=True,bZ=True),RotationFilter=(bX=True,bY=True,bZ=True),ScaleFilter=(bX=False,bY=False,bZ=False))
    Parents=((Item=(Type=Bone,Name="spine_02"),Weight=0.075),(Item=(Type=Bone,Name="spine_01"),Weight=0.425),(Item=(Type=Bone,Name="pelvis"),Weight=0.5))
    Weight=1
   )
   > &RigUnit_ParentConstraint(
    AdvancedSettings=(RotationOrderForFilter=XZY)
    Child=(Type=Null,Name="dyn_C_hair_6_ctrl_null")
    Filter=(TranslationFilter=(bX=True,bY=True,bZ=True),RotationFilter=(bX=True,bY=True,bZ=True),ScaleFilter=(bX=False,bY=False,bZ=False))
    Parents=((Item=(Type=Bone,Name="spine_02"),Weight=0.075),(Item=(Type=Bone,Name="spine_01"),Weight=0.425),(Item=(Type=Bone,Name="pelvis"),Weight=0.5))
    Weight=1
   )
   > &RigUnit_ParentConstraint(
    AdvancedSettings=(RotationOrderForFilter=XZY)
    Child=(Type=Null,Name="dyn_R_hair_6_ctrl_null")
    Filter=(TranslationFilter=(bX=True,bY=True,bZ=True),RotationFilter=(bX=True,bY=True,bZ=True),ScaleFilter=(bX=False,bY=False,bZ=False))
    Parents=((Item=(Type=Bone,Name="spine_02"),Weight=0.075),(Item=(Type=Bone,Name="spine_01"),Weight=0.425),(Item=(Type=Bone,Name="pelvis"),Weight=0.5))
    Weight=1
   )
   > &RigUnit_ParentConstraint(
    AdvancedSettings=(RotationOrderForFilter=XZY)
    Child=(Type=Null,Name="dyn_L_hair_7_ctrl_null")
    Filter=(TranslationFilter=(bX=True,bY=True,bZ=True),RotationFilter=(bX=True,bY=True,bZ=True),ScaleFilter=(bX=False,bY=False,bZ=False))
    Parents=((Item=(Type=Bone,Name="spine_01"),Weight=0.02),(Item=(Type=Bone,Name="pelvis"),Weight=0.98))
    Weight=1
   )
   > &RigUnit_ParentConstraint(
    AdvancedSettings=(RotationOrderForFilter=XZY)
    Child=(Type=Null,Name="dyn_C_hair_7_ctrl_null")
    Filter=(TranslationFilter=(bX=True,bY=True,bZ=True),RotationFilter=(bX=True,bY=True,bZ=True),ScaleFilter=(bX=False,bY=False,bZ=False))
    Parents=((Item=(Type=Bone,Name="spine_01"),Weight=0.02),(Item=(Type=Bone,Name="pelvis"),Weight=0.98))
    Weight=1
   )
   > &RigUnit_ParentConstraint(
    AdvancedSettings=(RotationOrderForFilter=XZY)
    Child=(Type=Null,Name="dyn_R_hair_7_ctrl_null")
    Filter=(TranslationFilter=(bX=True,bY=True,bZ=True),RotationFilter=(bX=True,bY=True,bZ=True),ScaleFilter=(bX=False,bY=False,bZ=False))
    Parents=((Item=(Type=Bone,Name="spine_01"),Weight=0.02),(Item=(Type=Bone,Name="pelvis"),Weight=0.98))
    Weight=1
   )
   > RigVMFunctionReturnNode()
  ]}
 CollapseNode_ContainedGraph {nodes [
   RigVMFunctionEntryNode()
   n1 = &RigVMFunction_AlphaInterp(
    InterpSpeedDecreasing=15
    InterpSpeedIncreasing=15
    Scale=1
    bClampResult=False
    Value=&RigVMFunction_MathBoolToFloat(Value=RigVMVariableNode() {variable=pub isPistolEquipped}.Value).Result
   )
   > RigVMVariableNode(Value=@n1.Result) {variable=is_PistolEquipped lerp}
   > RigVMFunctionReturnNode()
  ]}
 Multiply_3_ContainedGraph {nodes [
   n1 = RigVMFunctionEntryNode()
   RigVMFunctionReturnNode(Result=&RigVMFunction_MathDoubleMul(B=@n1.C, A=&RigVMFunction_MathDoubleMul(A=@n1.A, B=@n1.B).Result).Result)
  ]}
 Multiply_4_ContainedGraph {nodes [
   n1 = RigVMFunctionEntryNode()
   RigVMFunctionReturnNode(Result=&RigVMFunction_MathDoubleMul(B=@n1.C, A=&RigVMFunction_MathDoubleMul(A=@n1.A, B=@n1.B).Result).Result)
  ]}
 New Function_ContainedGraph {nodes [
   n1 = RigVMFunctionEntryNode()
   RigVMFunction_ControlFlowBranch = &RigVMFunction_ControlFlowBranch(ExecuteContext=@n1.ExecuteContext, Condition=@n1.debug)
   RigVMFunctionReturnNode(ExecuteContext=@RigVMFunction_ControlFlowBranch.Completed)
   RigUnit_ItemArray = &RigUnit_ItemArray(Items="((Type=Bone,Name=\"spine_01\"),(Type=Bone,Name=\"spine_02\"),(Type=Bone,Name=\"spine_03\"),(Type=Bone,Name=\"spine_04\"),(Type=Bone,Name=\"spine_05\"),(Type=Bone,Name=\"neck_01\"),(Type=Bone,Name=\"neck_02\"),(Type=Bone,Name=\"head\"),(Type=Bone,Name=\"clavicle_l\"),(Type=B ...(truncated)")
   DISPATCH_RigVMDispatch_ArrayIterator = RigVMDispatchNode(Array=@RigUnit_ItemArray.Items, ExecuteContext=@RigVMFunction_ControlFlowBranch.True) {template="DISPATCH_RigVMDispatch_ArrayIterator(in Array,out Element,out Index,out Count,out Ratio)"}
   Concat = /Script/RigVM.RigVMFunction_NameConcat(B=_ctrl, A=@DISPATCH_RigVMDispatch_ArrayIterator.Element.Name) {is=struct RigVMUnitNode, template="Concat::Execute(in A,in B,out Result)"}
   > &RigUnit_SetTransform(
    Item=(Type=Bone,Name="spine_01")
    Weight=1
    bInitial=False
    Item.Name=@DISPATCH_RigVMDispatch_ArrayIterator.Element.Name
    Value=&RigUnit_GetTransform(Item=(Type=Control,Name="spine_01_ctrl"), bInitial=False, Item.Name=@Concat.Result).Transform
   )
   &RigUnit_SetTransform(
    Item=(Type=Bone,Name="pelvis")
    Weight=1
    bInitial=False
    bPropagateToChildren=false
    Value=&RigUnit_GetTransform(Item=(Type=Control,Name="pelvis_ctrl"), bInitial=False).Transform
    ExecutePin=@DISPATCH_RigVMDispatch_ArrayIterator.Completed
   )
  ]}
 New Function_ContainedGraph {nodes [
   RigVMFunctionEntryNode()
   > RigVMCollapseNode()
   > RigVMFunction_Sequence = RigVMAggregateNode()
   TwistBones_2 = RigVMFunctionReferenceNode(
    Axis to Twist=(bX=true,bY=true,bZ=true)
    Bones to Twist=((Type=Bone,Name="dyn_C_hair_3"),(Type=Bone,Name="dyn_C_hair_4"),(Type=Bone,Name="dyn_C_hair_5"),(Type=Bone,Name="dyn_C_hair_6"),(Type=Bone,Name="dyn_C_hair_7"))
    End=(Type=Bone,Name="neck_01")
    Start=(Type=Bone,Name="pelvis")
    ExecuteContext=@RigVMFunction_Sequence.A
   )
   > TwistBones_2_1 = RigVMFunctionReferenceNode(
    Axis to Twist=(bX=True,bY=true,bZ=true)
    Bones to Twist=((Type=Bone,Name="dyn_L_hair_3"),(Type=Bone,Name="dyn_L_hair_4"),(Type=Bone,Name="dyn_L_hair_5"),(Type=Bone,Name="dyn_L_hair_6"),(Type=Bone,Name="dyn_L_hair_7"))
    End=(Type=Bone,Name="neck_01")
    Start=(Type=Bone,Name="pelvis")
   )
   > TwistBones_2_1_1 = RigVMFunctionReferenceNode(
    Axis to Twist=(bX=True,bY=true,bZ=true)
    Bones to Twist=((Type=Bone,Name="dyn_R_hair_3"),(Type=Bone,Name="dyn_R_hair_4"),(Type=Bone,Name="dyn_R_hair_5"),(Type=Bone,Name="dyn_R_hair_6"),(Type=Bone,Name="dyn_R_hair_7"))
    End=(Type=Bone,Name="neck_01")
    Start=(Type=Bone,Name="pelvis")
   )
   n1 = &RigUnit_ItemArray(Items=((Type=Bone,Name="dyn_C_hair_1"),(Type=Bone,Name="dyn_C_hair_2"),(Type=Bone,Name="dyn_C_hair_3"),(Type=Bone,Name="dyn_C_hair_4"),(Type=Bone,Name="dyn_C_hair_5"),(Type=Bone,Name="dyn_C_hair_6"),(Type=Bone,Name="dyn_C_hair_7")))
   SplineIK_2 = RigVMFunctionReferenceNode(
    Primary Axis=(X=1,Y=0,Z=0)
    Samples Per Segment=1
    Secondary Spline Direction=(X=0,Y=0,Z=0)
    Spline Mode=BSpline
    Stretch=false
    Up Axis=(X=0,Y=0,Z=0)
    debug=false
    Bones=@n1.Items
    ExecuteContext=@RigVMFunction_Sequence.B
    Controls=RigVMVariableNode() {variable=c_dyn_hair_controls}.Value
   )
   RigUnit_ItemArray = &RigUnit_ItemArray(Items=((Type=Bone,Name="dyn_L_hair_1"),(Type=Bone,Name="dyn_L_hair_2"),(Type=Bone,Name="dyn_L_hair_3"),(Type=Bone,Name="dyn_L_hair_4"),(Type=Bone,Name="dyn_L_hair_5"),(Type=Bone,Name="dyn_L_hair_6"),(Type=Bone,Name="dyn_L_hair_7")))
   > SplineIK_2_1 = RigVMFunctionReferenceNode(
    Primary Axis=(X=1,Y=0,Z=0)
    Samples Per Segment=1
    Secondary Spline Direction=(X=0,Y=0,Z=0)
    Spline Mode=BSpline
    Stretch=false
    Up Axis=(X=0,Y=0,Z=0)
    debug=false
    Bones=@RigUnit_ItemArray.Items
    Controls=RigVMVariableNode() {variable=l_dyn_hair_controls}.Value
   )
   n2 = &RigUnit_ItemArray(Items=((Type=Bone,Name="dyn_R_hair_1"),(Type=Bone,Name="dyn_R_hair_2"),(Type=Bone,Name="dyn_R_hair_3"),(Type=Bone,Name="dyn_R_hair_4"),(Type=Bone,Name="dyn_R_hair_5"),(Type=Bone,Name="dyn_R_hair_6"),(Type=Bone,Name="dyn_R_hair_7")))
   > SplineIK_2_1_1 = RigVMFunctionReferenceNode(
    Primary Axis=(X=1,Y=0,Z=0)
    Samples Per Segment=1
    Secondary Spline Direction=(X=0,Y=0,Z=0)
    Spline Mode=BSpline
    Stretch=false
    Up Axis=(X=0,Y=0,Z=0)
    debug=false
    Bones=@n2.Items
    Controls=RigVMVariableNode() {variable=r_dyn_hair_controls}.Value
   )
   DeltaRotation = RigVMFunctionReferenceNode(
    Rotation Order=ZYX
    child=spine_02
    max=80
    parent=spine_02_null
    parentType=Null
    primAxis=2
    twistAxis=(X=1,Y=0,Z=0)
    ExecuteContext=@RigVMFunction_Sequence.C
   )
   RigVMFunction_MathQuaternionFromEuler = &RigVMFunction_MathQuaternionFromEuler(Euler=(X=0,Y=10,Z=0))
   n3 = /Script/RigVM.RigVMFunction_MathVectorMake(X=3, Y=0, Z=-2) {is=struct RigVMUnitNode}
   > &RigUnit_ModifyTransforms(
    ItemToModify="((Item=(Type=Bone,Name=\"dyn_C_hair_6\"),Transform=(Rotation=(X=0,Y=0,Z=0,W=1),Translation=(X=0,Y=0,Z=0),Scale3D=(X=1,Y=1,Z=1))),(Item=(Type=Bone,Name=\"dyn_L_hair_6\"),Transform=(Rotation=(X=0,Y=0,Z=0,W=1),Translation=(X=0,Y=0,Z=0),Scale3D=(X=1,Y=1,Z=1))),(It ...(truncated)"
    WeightMinimum=0
    Weight=@DeltaRotation.outputWeight
    ItemToModify.0.Transform.Rotation=@RigVMFunction_MathQuaternionFromEuler.Result
    ItemToModify.1.Transform.Rotation=@RigVMFunction_MathQuaternionFromEuler.Result
    ItemToModify.2.Transform.Rotation=@RigVMFunction_MathQuaternionFromEuler.Result
    ItemToModify.0.Transform.Translation=@n3.Result
    ItemToModify.1.Transform.Translation=@n3.Result
    ItemToModify.2.Transform.Translation=@n3.Result
   )
   RigVMFunctionReturnNode(ExecuteContext=@RigVMFunction_Sequence.D)
   &RigUnit_SphericalPoseReader(
    ActiveRegionScaleFactors=(PositiveWidth=1,NegativeWidth=1,PositiveHeight=1,NegativeHeight=1)
    ActiveRegionSize=0.15
    Debug=(bDrawDebug=false,bDraw2D=False,bDrawLocalAxes=False,DebugScale=25,DebugSegments=20,DebugThickness=0.25)
    DriverAxis=(X=1,Y=0,Z=0)
    DriverItem=(Type=Bone,Name="spine_02")
    ExecutePin=()
    FalloffRegionScaleFactors=(PositiveWidth=1,NegativeWidth=1,PositiveHeight=1,NegativeHeight=1)
    FalloffSize=0.35
    FlipHeightScaling=False
    FlipWidthScaling=False
    OptionalParentItem=(Type=Bone,Name="pelvis")
    RotationOffset=(X=90,Y=0,Z=0)
   )
  ]}
 New Function_ContainedGraph {nodes [
   DeltaRotation_1_3_1 = RigVMFunctionReferenceNode(
    ExecuteContext=()
    Rotation Order=ZYX
    child=head
    debugRotation=false
    max=-43
    min=0
    parent=head_null
    parentType=Null
    primAxis=2
    twistAxis=(X=1,Y=0,Z=0)
   )
   > alterTransform_4_2_6 = RigVMFunctionReferenceNode(
    type=Bone
    transforms="((Bone=\"dyn_C_hair_1\",Transform=(Rotation=(X=0,Y=0.1736,Z=0,W=0.9848),Translation=(X=0,Y=0,Z=-8),Scale3D=(X=1,Y=1,Z=1))),(Bone=\"dyn_L_hair_1\",Transform=(Rotation=(X=0,Y=0.1736,Z=0,W=0.9848),Translation=(X=0,Y=0,Z=-8),Scale3D=(X=1,Y=1,Z=1))),(Bone=\"dyn_R_ha ...(truncated)"
   )
   > alterTransform_4_1_5 = RigVMFunctionReferenceNode(
    type=Bone
    transforms="((Bone=\"dyn_C_hair_2\",Transform=(Rotation=(X=0,Y=0.1736,Z=0,W=0.9848),Translation=(X=0,Y=0,Z=-6),Scale3D=(X=1,Y=1,Z=1))),(Bone=\"dyn_L_hair_2\",Transform=(Rotation=(X=0,Y=0.1736,Z=0,W=0.9848),Translation=(X=0,Y=0,Z=-6),Scale3D=(X=1,Y=1,Z=1))),(Bone=\"dyn_R_ha ...(truncated)"
   )
   > alterTransform_4_1_1_1 = RigVMFunctionReferenceNode(
    type=Bone
    transforms="((Bone=\"dyn_C_hair_3\",Transform=(Rotation=(X=0,Y=0,Z=0,W=1),Translation=(X=0,Y=0,Z=-1),Scale3D=(X=1,Y=1,Z=1))),(Bone=\"dyn_L_hair_3\",Transform=(Rotation=(X=0,Y=0,Z=0,W=1),Translation=(X=0,Y=0,Z=-1),Scale3D=(X=1,Y=1,Z=1))),(Bone=\"dyn_R_hair_3\",Transform=(Rot ...(truncated)"
   )
   DeltaRotation_3 = RigVMFunctionReferenceNode(
    ExecuteContext=()
    Rotation Order=ZYX
    child=spine_05
    debugRotation=false
    max=-50
    min=-10
    onlyTwist=true
    parent=spine_01
    parentType=Bone
    twistAxis=(X=1,Y=0,Z=0)
   )
   n1 = &RigVMFunction_AnimEvalRichCurve(
    Curve=(EditorCurveData=(Keys=((InterpMode=RCIM_Cubic),(InterpMode=RCIM_Cubic,TangentMode=RCTM_User,Time=0.5279,Value=0.4056,ArriveTangent=1.3758,LeaveTangent=1.3758),(InterpMode=RCIM_Cubic,Time=1,Value=1))))
    SourceMinimum=0
    TargetMinimum=0
    Value=@DeltaRotation_3.outputWeight
   )
   > alterTransform_4_2_1 = RigVMFunctionReferenceNode(
    type=Bone
    transforms="((Bone=\"dyn_L_hair_2\",Transform=(Rotation=(X=0,Y=0,Z=0,W=1),Translation=(X=0,Y=-1,Z=0),Scale3D=(X=1,Y=1,Z=1))),(Bone=\"dyn_L_hair_3\",Transform=(Rotation=(X=0,Y=0,Z=0,W=1),Translation=(X=0,Y=-1,Z=0),Scale3D=(X=1,Y=1,Z=1))),(Bone=\"dyn_L_hair_4\",Transform=(Rot ...(truncated)"
    weight=@n1.Result
   )
   > DeltaRotation_2 = RigVMFunctionReferenceNode(
    Rotation Order=ZYX
    child=spine_05
    debugRotation=false
    max=50
    min=10
    onlyTwist=true
    parent=spine_01
    parentType=Bone
    twistAxis=(X=1,Y=0,Z=0)
   )
   RigVMFunction_AnimEvalRichCurve = &RigVMFunction_AnimEvalRichCurve(
    Curve=(EditorCurveData=(Keys=((InterpMode=RCIM_Cubic),(InterpMode=RCIM_Cubic,TangentMode=RCTM_User,Time=0.5279,Value=0.4056,ArriveTangent=1.3758,LeaveTangent=1.3758),(InterpMode=RCIM_Cubic,Time=1,Value=1))))
    SourceMinimum=0
    TargetMinimum=0
    Value=@DeltaRotation_2.outputWeight
   )
   > alterTransform_4_2_1_1 = RigVMFunctionReferenceNode(
    type=Bone
    transforms="((Bone=\"dyn_R_hair_2\",Transform=(Rotation=(X=0,Y=0,Z=0,W=1),Translation=(X=0,Y=-1,Z=0),Scale3D=(X=1,Y=1,Z=1))),(Bone=\"dyn_R_hair_3\",Transform=(Rotation=(X=0,Y=0,Z=0,W=1),Translation=(X=0,Y=-1,Z=0),Scale3D=(X=1,Y=1,Z=1))),(Bone=\"dyn_R_hair_4\",Transform=(Rot ...(truncated)"
    weight=@RigVMFunction_AnimEvalRichCurve.Result
   )
   RigVMFunctionEntryNode()
   > RigVMFunction_Sequence = RigVMAggregateNode()
   RigVMFunction_ControlFlowBranch = &RigVMFunction_ControlFlowBranch(ExecuteContext=@RigVMFunction_Sequence.B, Condition=RigVMVariableNode() {variable=pub disableUpperLimbsCorrection}.Value)
   n2 = &RigVMFunction_Sequence(ExecuteContext=@RigVMFunction_ControlFlowBranch.False)
   n3 = &RigUnit_SphericalPoseReader(
    ActiveRegionScaleFactors=(PositiveWidth=1,NegativeWidth=1,PositiveHeight=1,NegativeHeight=1)
    ActiveRegionSize=0.1
    Debug=(bDrawDebug=false,bDraw2D=False,bDrawLocalAxes=False,DebugScale=25,DebugSegments=20,DebugThickness=0.25)
    DriverAxis=(X=1,Y=0,Z=0)
    DriverItem=(Type=Bone,Name="upperarm_l")
    FalloffRegionScaleFactors=(PositiveWidth=1,NegativeWidth=1,PositiveHeight=1,NegativeHeight=1)
    FalloffSize=0.3
    FlipHeightScaling=False
    FlipWidthScaling=False
    OptionalParentItem=(Type=Bone,Name="None")
    RotationOffset=(X=-140,Y=50,Z=0)
    ExecutePin=@n2.A
   )
   > alterTransform_2 = RigVMFunctionReferenceNode(
    type=Bone
    transforms="((Bone=\"dyn_L_hair_2\",Transform=(Rotation=(X=0,Y=0,Z=0,W=1),Translation=(X=0,Y=2,Z=0),Scale3D=(X=1,Y=1,Z=1))),(Bone=\"dyn_L_hair_3\",Transform=(Rotation=(X=0,Y=0,Z=0,W=1),Translation=(X=0,Y=4,Z=0),Scale3D=(X=1,Y=1,Z=1))),(Bone=\"dyn_L_hair_4\",Transform=(Rotat ...(truncated)"
    weight=@n3.OutputParam
   )
   > n4 = &RigUnit_SphericalPoseReader(
    ActiveRegionScaleFactors=(PositiveWidth=1,NegativeWidth=1,PositiveHeight=1,NegativeHeight=1)
    ActiveRegionSize=0.1
    Debug=(bDrawDebug=false,DebugScale=25,DebugSegments=20,DebugThickness=0.25)
    DriverAxis=(X=-1,Y=0,Z=0)
    DriverItem=(Type=Bone,Name="upperarm_r")
    FalloffRegionScaleFactors=(PositiveWidth=1,NegativeWidth=1,PositiveHeight=1,NegativeHeight=1)
    FalloffSize=0.3
    FlipHeightScaling=false
    FlipWidthScaling=False
    OptionalParentItem=(Type=Bone,Name="None")
    RotationOffset=(X=40,Y=50,Z=0)
   )
   > alterTransform_3 = RigVMFunctionReferenceNode(
    type=Bone
    transforms="((Bone=\"dyn_R_hair_2\",Transform=(Rotation=(X=0,Y=0,Z=0,W=1),Translation=(X=0,Y=-2,Z=0),Scale3D=(X=1,Y=1,Z=1))),(Bone=\"dyn_R_hair_3\",Transform=(Rotation=(X=0,Y=0,Z=0,W=1),Translation=(X=0,Y=-4,Z=0),Scale3D=(X=1,Y=1,Z=1))),(Bone=\"dyn_R_hair_4\",Transform=(Rot ...(truncated)"
    weight=@n4.OutputParam
   )
   > n5 = &RigUnit_SphericalPoseReader(
    ActiveRegionScaleFactors=(PositiveWidth=1,NegativeWidth=1,PositiveHeight=1,NegativeHeight=1)
    ActiveRegionSize=0.1
    Debug=(bDrawDebug=true,bDraw2D=False,bDrawLocalAxes=False,DebugScale=25,DebugSegments=20,DebugThickness=0.25)
    DriverAxis=(X=0,Y=1,Z=0)
    DriverItem=(Type=Bone,Name="head")
    FalloffRegionScaleFactors=(PositiveWidth=1,NegativeWidth=1,PositiveHeight=1,NegativeHeight=1)
    FalloffSize=0.3
    FlipHeightScaling=False
    FlipWidthScaling=False
    OptionalParentItem=(Type=Bone,Name="spine_05")
    RotationOffset=(X=0,Y=0,Z=0)
   )
   Subtract = &RigVMFunction_MathDoubleSub(B=1, A=@n5.OutputParam)
   Absolute = &RigVMFunction_MathDoubleAbs(Value=@Subtract.Result)
   AlphaInterp = &RigVMFunction_AlphaInterp(
    InterpSpeedDecreasing=10
    InterpSpeedIncreasing=10
    Scale=1
    bClampResult=true
    Value=&RigVMFunction_MathBoolToFloat(Value=RigVMVariableNode() {variable=FE Idle}.Value).Result
   )
   Multiply_3 = RigVMAggregateNode(B=-1, A=@Absolute.Result, C=@AlphaInterp.Result)
   > /Script/ControlRig.RigUnit_SetCurveValue(Curve=SPR Head Look Right FE, Value=@Multiply_3.Result) {is=struct RigVMUnitNode}
   > n6 = &RigUnit_SphericalPoseReader(
    ActiveRegionScaleFactors=(PositiveWidth=1,NegativeWidth=1,PositiveHeight=1,NegativeHeight=1)
    ActiveRegionSize=0.1
    Debug=(bDrawDebug=true,bDraw2D=False,bDrawLocalAxes=False,DebugScale=25,DebugSegments=20,DebugThickness=0.25)
    DriverAxis=(X=0,Y=1,Z=0)
    DriverItem=(Type=Bone,Name="head")
    FalloffRegionScaleFactors=(PositiveWidth=1,NegativeWidth=1,PositiveHeight=1,NegativeHeight=1)
    FalloffSize=0.3
    FlipHeightScaling=False
    FlipWidthScaling=False
    OptionalParentItem=(Type=Bone,Name="spine_05")
    RotationOffset=(X=100,Y=0,Z=0)
   )
   > &RigUnit_SetCurveValue(Curve=SPR Head Look Left FE, Value=&RigVMFunction_MathDoubleMul(A=@n6.OutputParam, B=@AlphaInterp.Result).Result)
   claviclesBackFoward = RigVMFunctionReferenceNode(ExecuteContext=@n2.B)
   > n7 = &RigUnit_SphericalPoseReader(
    ActiveRegionScaleFactors=(PositiveWidth=1,NegativeWidth=1,PositiveHeight=1,NegativeHeight=1)
    ActiveRegionSize=0.1
    Debug=(bDrawDebug=false,bDraw2D=False,bDrawLocalAxes=False,DebugScale=25,DebugSegments=20,DebugThickness=0.25)
    DriverAxis=(X=1,Y=0,Z=0)
    DriverItem=(Type=Bone,Name="clavicle_l")
    FalloffRegionScaleFactors=(PositiveWidth=0.5,NegativeWidth=0.5,PositiveHeight=0.5,NegativeHeight=0.5)
    FalloffSize=0.3
    FlipHeightScaling=False
    FlipWidthScaling=False
    OptionalParentItem=(Type=Bone,Name="None")
    RotationOffset=(X=-45,Y=6,Z=60)
   )
   > alterTransform_5 = RigVMFunctionReferenceNode(
    type=Bone
    transforms=((Bone="dyn_L_hair_2",Transform=(Rotation=(X=0,Y=0,Z=0,W=1),Translation=(X=0,Y=3,Z=-1),Scale3D=(X=1,Y=1,Z=1))),(Bone="dyn_L_hair_3",Transform=(Rotation=(X=0,Y=0,Z=0,W=1),Translation=(X=0,Y=6,Z=-1),Scale3D=(X=1,Y=1,Z=1))))
    weight=@n7.OutputParam
   )
   > n8 = &RigUnit_SphericalPoseReader(
    ActiveRegionScaleFactors=(PositiveWidth=1,NegativeWidth=1,PositiveHeight=1,NegativeHeight=1)
    ActiveRegionSize=0.1
    Debug=(bDrawDebug=false,bDraw2D=False,bDrawLocalAxes=False,DebugScale=25,DebugSegments=20,DebugThickness=0.25)
    DriverAxis=(X=-1,Y=0,Z=0)
    DriverItem=(Type=Bone,Name="clavicle_r")
    FalloffRegionScaleFactors=(PositiveWidth=0.5,NegativeWidth=0.5,PositiveHeight=0.5,NegativeHeight=0.5)
    FalloffSize=0.3
    FlipHeightScaling=False
    FlipWidthScaling=False
    OptionalParentItem=(Type=Bone,Name="None")
    RotationOffset=(X=45,Y=174,Z=30)
   )
   > alterTransform_5_1 = RigVMFunctionReferenceNode(
    type=Bone
    transforms=((Bone="dyn_R_hair_2",Transform=(Rotation=(X=0,Y=0,Z=0,W=1),Translation=(X=0,Y=-3,Z=-1),Scale3D=(X=1,Y=1,Z=1))),(Bone="dyn_R_hair_3",Transform=(Rotation=(X=0,Y=0,Z=0,W=1),Translation=(X=0,Y=-5,Z=-1),Scale3D=(X=1,Y=1,Z=1))))
    weight=@n8.OutputParam
   )
   > DeltaRotation_4 = RigVMFunctionReferenceNode(
    Rotation Order=ZXY
    child=clavicle_r
    debugRotation=false
    max=-15
    parent=clavicle_r_null
    parentType=Null
    primAxis=1
    twistAxis=(X=1,Y=0,Z=0)
   )
   > alterTransform_4_2_1_2_1 = RigVMFunctionReferenceNode(
    type=Bone
    transforms=((Bone="dyn_R_hair_2",Transform=(Rotation=(X=0,Y=0,Z=0,W=1),Translation=(X=-1,Y=0,Z=-1),Scale3D=(X=1,Y=1,Z=1))),(Bone="dyn_R_hair_3",Transform=(Rotation=(X=0,Y=0,Z=0,W=1),Translation=(X=-4,Y=-1,Z=-1),Scale3D=(X=1,Y=1,Z=1))))
    weight=@DeltaRotation_4.outputWeight
   )
   > DeltaRotation_4_1 = RigVMFunctionReferenceNode(
    Rotation Order=ZXY
    child=clavicle_l
    debugRotation=false
    max=-10
    parent=clavicle_l_null
    parentType=Null
    primAxis=1
    twistAxis=(X=1,Y=0,Z=0)
   )
   > alterTransform_4_2_1_2_1_1 = RigVMFunctionReferenceNode(
    type=Bone
    transforms=((Bone="dyn_L_hair_2",Transform=(Rotation=(X=0,Y=0,Z=0,W=1),Translation=(X=-1,Y=0,Z=-1),Scale3D=(X=1,Y=1,Z=1))),(Bone="dyn_L_hair_3",Transform=(Rotation=(X=0,Y=0,Z=0,W=1),Translation=(X=-4,Y=1,Z=-1),Scale3D=(X=1,Y=1,Z=1))))
    weight=@DeltaRotation_4_1.outputWeight
   )
   RigVMFunctionReturnNode(ExecuteContext=@RigVMFunction_Sequence.E)
   RigVMFunction_Sequence_3 = RigVMAggregateNode(ExecuteContext=@RigVMFunction_Sequence.D)
   DeltaRotation_1 = RigVMFunctionReferenceNode(
    Rotation Order=ZYX
    child=neck_01
    debugRotation=false
    max=-23
    min=0
    parent=neck_01_null
    parentType=Null
    primAxis=2
    twistAxis=(X=1,Y=0,Z=0)
    ExecuteContext=@RigVMFunction_Sequence_3.A
   )
   > alterTransform_4_1_3 = RigVMFunctionReferenceNode(
    type=Bone
    transforms="((Bone=\"dyn_C_hair_2\",Transform=(Rotation=(X=0,Y=0,Z=0,W=1),Translation=(X=0,Y=0,Z=-1),Scale3D=(X=1,Y=1,Z=1))),(Bone=\"dyn_L_hair_2\",Transform=(Rotation=(X=0,Y=0,Z=0,W=1),Translation=(X=0,Y=0,Z=-1),Scale3D=(X=1,Y=1,Z=1))),(Bone=\"dyn_R_hair_2\",Transform=(Rot ...(truncated)"
    weight=@DeltaRotation_1.outputWeight
   )
   > alterTransform_4_1_1_2 = RigVMFunctionReferenceNode(
    type=Bone
    transforms="((Bone=\"dyn_C_hair_3\",Transform=(Rotation=(X=0,Y=0,Z=0,W=1),Translation=(X=0,Y=0,Z=-1),Scale3D=(X=1,Y=1,Z=1))),(Bone=\"dyn_L_hair_3\",Transform=(Rotation=(X=0,Y=0,Z=0,W=1),Translation=(X=0,Y=0,Z=-1),Scale3D=(X=1,Y=1,Z=1))),(Bone=\"dyn_R_hair_3\",Transform=(Rot ...(truncated)"
    weight=@DeltaRotation_1.outputWeight
   )
   > DeltaRotation_1_3 = RigVMFunctionReferenceNode(Rotation Order=ZYX, child=head, debugRotation=false, max=-20, min=0, parent=head_null, parentType=Null, primAxis=2, twistAxis=(X=1,Y=0,Z=0))
   > n9 = &RigUnit_SphericalPoseReader(
    ActiveRegionScaleFactors=(PositiveWidth=2,NegativeWidth=2,PositiveHeight=4,NegativeHeight=4)
    ActiveRegionSize=0.03
    Debug=(bDrawDebug=false,bDraw2D=False,bDrawLocalAxes=False,DebugScale=25,DebugSegments=20,DebugThickness=0.25)
    DriverAxis=(X=1,Y=0,Z=0)
    DriverItem=(Type=Bone,Name="upperarm_r")
    FalloffRegionScaleFactors=(PositiveWidth=0.5,NegativeWidth=0.7,PositiveHeight=1.6,NegativeHeight=1)
    FalloffSize=0.25
    FlipHeightScaling=False
    FlipWidthScaling=False
    OptionalParentItem=(Type=Bone,Name="None")
    RotationOffset=(X=75,Y=25,Z=10)
   )
   RigVMFunction_MathBoolToFloat = &RigVMFunction_MathBoolToFloat(Value=RigVMVariableNode() {variable=pub isTargeting}.Value)
   Multiply_2 = RigVMAggregateNode(B=@RigVMFunction_MathBoolToFloat.Result, A=@n9.OutputParam, C=RigVMVariableNode() {variable=is_PistolEquipped lerp}.Value)
   > &RigUnit_ModifyTransforms(
    ItemToModify=((Item=(Type=Bone,Name="base_R_front_hair_01"),Transform=(Rotation=(X=0,Y=0,Z=-0.225,W=0.9744),Translation=(X=0,Y=0,Z=0),Scale3D=(X=1,Y=1,Z=1))))
    WeightMinimum=0
    Weight=@Multiply_2.Result
   )
   n10 = &RigVMFunction_AnimEvalRichCurve(
    Curve=(EditorCurveData=(Keys=((InterpMode=RCIM_Cubic),(InterpMode=RCIM_Cubic,Time=0.3879,Value=0.174,ArriveTangent=1,LeaveTangent=1),(InterpMode=RCIM_Cubic,TangentMode=RCTM_User,Time=1,Value=1,ArriveTangent=1.9233,LeaveTangent=1.9233))))
    SourceMinimum=0
    TargetMinimum=0
    Value=@DeltaRotation_1_3.outputWeight
   )
   > alterTransform_4_2_3_1 = RigVMFunctionReferenceNode(
    type=Bone
    transforms="((Bone=\"dyn_C_hair_1\",Transform=(Rotation=(X=0,Y=0.0872,Z=0,W=0.9962),Translation=(X=0,Y=0,Z=-3),Scale3D=(X=1,Y=1,Z=1))),(Bone=\"dyn_L_hair_1\",Transform=(Rotation=(X=0,Y=0.0872,Z=0,W=0.9962),Translation=(X=0,Y=0,Z=-3),Scale3D=(X=1,Y=1,Z=1))),(Bone=\"dyn_R_ha ...(truncated)"
    weight=@n10.Result
   )
   > alterTransform_4_1_2_1 = RigVMFunctionReferenceNode(
    type=Bone
    transforms="((Bone=\"dyn_C_hair_2\",Transform=(Rotation=(X=0,Y=0.0872,Z=0,W=0.9962),Translation=(X=0,Y=0,Z=-1),Scale3D=(X=1,Y=1,Z=1))),(Bone=\"dyn_L_hair_2\",Transform=(Rotation=(X=0,Y=0.0872,Z=0,W=0.9962),Translation=(X=0,Y=0,Z=-1),Scale3D=(X=1,Y=1,Z=1))),(Bone=\"dyn_R_ha ...(truncated)"
    weight=@n10.Result
   )
   Multiply = &RigVMFunction_MathDoubleMul(B=@DeltaRotation_1.outputWeight, A=@n10.Result)
   > alterTransform_4_1_3_1 = RigVMFunctionReferenceNode(
    type=Bone
    transforms="((Bone=\"dyn_C_hair_2\",Transform=(Rotation=(X=0,Y=-0.0872,Z=0,W=0.9962),Translation=(X=0,Y=0,Z=-1),Scale3D=(X=1,Y=1,Z=1))),(Bone=\"dyn_L_hair_2\",Transform=(Rotation=(X=0,Y=-0.0872,Z=0,W=0.9962),Translation=(X=0,Y=0,Z=-1),Scale3D=(X=1,Y=1,Z=1))),(Bone=\"dyn_R_ ...(truncated)"
    weight=@Multiply.Result
   )
   > &RigUnit_ModifyTransforms(
    ItemToModify="((Item=(Type=Bone,Name=\"dyn_L_back_hair_01\"),Transform=(Rotation=(X=0,Y=-0.3859,Z=0,W=0.9226),Translation=(X=0,Y=0,Z=0),Scale3D=(X=1,Y=1,Z=1))),(Item=(Type=Bone,Name=\"dyn_R_back_hair_01\"),Transform=(Rotation=(X=0,Y=-0.4144,Z=0,W=0.9101),Translation=(X=0,Y= ...(truncated)"
    WeightMinimum=0
    Weight=@n10.Result
   )
   > DeltaRotation = RigVMFunctionReferenceNode(Rotation Order=ZYX, child=head, debugRotation=false, max=60, min=0, parent=head_null, parentType=Null, primAxis=2, twistAxis=(X=1,Y=0,Z=0))
   > alterTransform_4_2_2 = RigVMFunctionReferenceNode(
    type=Bone
    transforms="((Bone=\"dyn_C_hair_2\",Transform=(Rotation=(X=0,Y=0,Z=0,W=1),Translation=(X=0,Y=0,Z=-3),Scale3D=(X=1,Y=1,Z=1))),(Bone=\"dyn_L_hair_2\",Transform=(Rotation=(X=0,Y=0,Z=0,W=1),Translation=(X=0,Y=0,Z=-3),Scale3D=(X=1,Y=1,Z=1))),(Bone=\"dyn_R_hair_2\",Transform=(Rot ...(truncated)"
    weight=@DeltaRotation.outputWeight
   )
   > DeltaRotation_1_1_1_1_1 = RigVMFunctionReferenceNode(
    Rotation Order=ZYX
    child=head
    debugRotation=false
    max=50
    min=15
    onlyTwist=true
    parent=head_null_sp
    parentType=Null
    primAxis=0
    twistAxis=(X=1,Y=0,Z=0)
   )
   > alterTransform_4_2_2_1_1_1_1 = RigVMFunctionReferenceNode(
    type=Bone
    transforms="((Bone=\"dyn_C_hair_1\",Transform=(Rotation=(X=0,Y=0,Z=0,W=1),Translation=(X=0,Y=2,Z=0),Scale3D=(X=1,Y=1,Z=1))),(Bone=\"dyn_R_hair_2\",Transform=(Rotation=(X=0,Y=0,Z=0,W=1),Translation=(X=0,Y=0,Z=-7),Scale3D=(X=1,Y=1,Z=1))),(Bone=\"dyn_R_hair_3\",Transform=(Rota ...(truncated)"
    weight=&RigVMFunction_MathDoubleMul(B=0.5, A=@DeltaRotation_1_1_1_1_1.outputWeight).Result
   )
   DeltaRotation v4 = RigVMFunctionReferenceNode(
    Rotation Order=ZYX
    child=(Type=Bone,Name="head")
    debugRotation=false
    outputMinMax=(X=0,Y=1)
    parent=(Type=Null,Name="head_null")
    premult=1
    rotationMinMax=(X=0,Y=4)
    triggerAxis=Y
    twistAxis=(X=1,Y=0,Z=0)
    ExecuteContext=@RigVMFunction_Sequence.C
   )
   > &RigUnit_ModifyTransforms(
    ItemToModify="((Item=(Type=Bone,Name=\"base_frontChain_c1\"),Transform=(Rotation=(X=0,Y=0,Z=-0.1736,W=0.9848),Translation=(X=-1,Y=0,Z=-5),Scale3D=(X=1,Y=1,Z=1))),(Item=(Type=Bone,Name=\"dyn_frontChain_c1_02\"),Transform=(Rotation=(X=0,Y=0,Z=-0.1736,W=0.9848),Translation=(X= ...(truncated)"
    WeightMinimum=0
    Weight=@"DeltaRotation v4".outputWeight
   )
   > DeltaRotation v4_1 = RigVMFunctionReferenceNode(
    Rotation Order=ZYX
    child=(Type=Bone,Name="head")
    debugRotation=false
    onlyTwist=false
    outputMinMax=(X=0,Y=1)
    parent=(Type=Null,Name="head_null")
    premult=1
    rotationMinMax=(X=7,Y=10)
    triggerAxis=Z
    twistAxis=(X=1,Y=0,Z=0)
   )
   > &RigUnit_ModifyTransforms(
    ItemToModify="((Item=(Type=Bone,Name=\"dyn_frontChain_c1_01\"),Transform=(Rotation=(X=0,Y=0,Z=0,W=1),Translation=(X=0,Y=0,Z=-1),Scale3D=(X=1,Y=1,Z=1))),(Item=(Type=Bone,Name=\"dyn_frontChain_c2_01\"),Transform=(Rotation=(X=0,Y=0,Z=0,W=1),Translation=(X=0,Y=0,Z=-1),Scale3D=( ...(truncated)"
    WeightMinimum=0
    Weight=@"DeltaRotation v4_1".outputWeight
   )
   n11 = &RigVMFunction_ControlFlowBranch(Condition=RigVMVariableNode() {variable=pub diveFloating}.Value, ExecuteContext=@RigVMFunction_Sequence.A)
   claviclesBackFoward_1 = RigVMFunctionReferenceNode(ExecuteContext=@n11.True)
   &RigUnit_ModifyTransforms(
    ExecutePin=()
    ItemToModify="((Item=(Type=Bone,Name=\"base_L_back_hair_01\"),Transform=(Rotation=(X=-0.0433,Y=0.0198,Z=-0.4156,W=0.9083),Translation=(X=0,Y=0,Z=0),Scale3D=(X=1,Y=1,Z=1))),(Item=(Type=Bone,Name=\"base_R_back_hair_01\"),Transform=(Rotation=(X=-0.1058,Y=-0.032,Z=-0.2875,W=0.9 ...(truncated)"
    WeightMinimum=0
    Weight=@DeltaRotation_1_1_1_1_1.outputWeight
   )
   RigVMRerouteNode(Value=@RigVMFunction_Sequence_3.B)
   alterTransform_4_1_2 = RigVMFunctionReferenceNode(
    type=Bone
    ExecuteContext=()
    transforms="((Bone=\"dyn_C_hair_2\",Transform=(Rotation=(X=0,Y=0,Z=0,W=1),Translation=(X=0,Y=0,Z=-4),Scale3D=(X=1,Y=1,Z=1))),(Bone=\"dyn_L_hair_2\",Transform=(Rotation=(X=0,Y=0,Z=0,W=1),Translation=(X=0,Y=0,Z=-4),Scale3D=(X=1,Y=1,Z=1))),(Bone=\"dyn_R_hair_2\",Transform=(Rot ...(truncated)"
   )
   alterTransform_4_2_3 = RigVMFunctionReferenceNode(
    type=Bone
    ExecuteContext=()
    transforms="((Bone=\"dyn_C_hair_1\",Transform=(Rotation=(X=0,Y=0,Z=0,W=1),Translation=(X=0,Y=0,Z=-3),Scale3D=(X=1,Y=1,Z=1))),(Bone=\"dyn_L_hair_1\",Transform=(Rotation=(X=0,Y=0,Z=0,W=1),Translation=(X=0,Y=0,Z=-3),Scale3D=(X=1,Y=1,Z=1))),(Bone=\"dyn_R_hair_1\",Transform=(Rot ...(truncated)"
   )
   lerp bools = RigVMCollapseNode(ExecuteContext=())
  ]}
 New Function_ContainedGraph {nodes [
   n1 = RigVMFunctionEntryNode()
   > DISPATCH_RigVMDispatch_ArrayIterator = RigVMDispatchNode(Array=@n1.transforms) {template="DISPATCH_RigVMDispatch_ArrayIterator(in Array,out Element,out Index,out Count,out Ratio)"}
   Multiply = &RigVMFunction_MathTransformMul(
    A=@DISPATCH_RigVMDispatch_ArrayIterator.Element.Transform
    B=&RigUnit_GetTransform(
     Item=(Type=Bone,Name="dyn_L_hair_3")
     Space=LocalSpace
     bInitial=False
     Item.Name=@DISPATCH_RigVMDispatch_ArrayIterator.Element.Bone
     Item.Type=@n1.type
    ).Transform
   ) {template="Multiply::Execute(in A,in B,out Result)"}
   > &RigUnit_SetTransform(
    Item=(Type=Bone,Name="dyn_L_hair_3")
    Space=LocalSpace
    bInitial=false
    Value=@Multiply.Result
    bPropagateToChildren=@n1.inheritsTransform
    Weight=@n1.weight
    Item.Name=@DISPATCH_RigVMDispatch_ArrayIterator.Element.Bone
    Item.Type=@n1.type
   )
   RigVMFunctionReturnNode(ExecuteContext=@DISPATCH_RigVMDispatch_ArrayIterator.Completed)
   &RigUnit_OffsetTransformForItem(
    ExecutePin=()
    Item=(Type=Bone,Name="None")
    OffsetTransform=(Rotation=(X=0,Y=0,Z=0,W=1),Translation=(X=0,Y=0,Z=0),Scale3D=(X=1,Y=1,Z=1))
    Weight=1
    bPropagateToChildren=True
   )
  ]}
 New Function_ContainedGraph {nodes [
   n1 = RigVMFunctionEntryNode()
   RigVMFunction_MathBoolToInteger = &RigVMFunction_MathBoolToInteger(Value=@n1.onlyTwist)
   RigVMFunction_MathQuaternionSwingTwist = &RigVMFunction_MathQuaternionSwingTwist(
    TwistAxis=@n1.twistAxis
    Input=&RigUnit_GetRelativeTransformForItem(
     Child=(Type=Bone,Name="None")
     Parent=(Type=Bone,Name="None")
     bChildInitial=False
     bParentInitial=False
     Child.Name=@n1.child
     Parent.Name=@n1.parent
     Parent.Type=@n1.parentType
    ).RelativeTransform.Rotation
   )
   RigVMFunction_MathQuaternionToEuler = &RigVMFunction_MathQuaternionToEuler(
    Value=RigVMDispatchNode(
     Values=((),())
     Index=@RigVMFunction_MathBoolToInteger.Result
     Values.0=@RigVMFunction_MathQuaternionSwingTwist.Swing
     Values.1=@RigVMFunction_MathQuaternionSwingTwist.Twist
    ) {template="DISPATCH_RigVMDispatch_SelectInt32(in Index,in Values,out Result)"}.Result
    RotationOrder=@n1.Rotation Order
   )
   > DISPATCH_RigVMDispatch_Print = RigVMDispatchNode(ScreenColor=(R=1,G=1,B=1,A=1), ScreenDuration=1.5, Enabled=@n1.debugRotation, Value=@RigVMFunction_MathQuaternionToEuler.Result) {template="DISPATCH_RigVMDispatch_Print(in Prefix,in Value,in Enabled,in ScreenDuration,in ScreenColor)"}
   Clamp = &RigVMFunction_MathIntClamp(Value=@n1.primAxis)
   DISPATCH_RigVMDispatch_SelectInt32 = RigVMDispatchNode(
    Values=(0,0,0)
    Index=@Clamp.Result
    Values.0=@RigVMFunction_MathQuaternionToEuler.Result.X
    Values.1=@RigVMFunction_MathQuaternionToEuler.Result.Y
    Values.2=@RigVMFunction_MathQuaternionToEuler.Result.Z
   ) {template="DISPATCH_RigVMDispatch_SelectInt32(in Index,in Values,out Result)"}
   Remap = &RigVMFunction_MathDoubleRemap(
    TargetMaximum=1
    TargetMinimum=0
    bClamp=False
    Value=@DISPATCH_RigVMDispatch_SelectInt32.Result
    SourceMaximum=@n1.max
    SourceMinimum=@n1.min
   )
   > RigVMFunctionReturnNode(outputWeight=&RigVMFunction_MathDoubleClamp(Value=@Remap.Result).Result)
  ]}
 New Function_ContainedGraph {nodes [
   RigVMFunctionEntryNode()
   > n1 = &RigUnit_SphericalPoseReader(
    ActiveRegionScaleFactors=(PositiveWidth=0.5,NegativeWidth=0.5,PositiveHeight=0.5,NegativeHeight=0.5)
    ActiveRegionSize=0.18
    Debug=(bDrawDebug=false,bDraw2D=False,bDrawLocalAxes=False,DebugScale=25,DebugSegments=20,DebugThickness=0.25)
    DriverAxis=(X=1,Y=0,Z=0)
    DriverItem=(Type=Bone,Name="clavicle_l")
    FalloffRegionScaleFactors=(PositiveWidth=0.5,NegativeWidth=0.5,PositiveHeight=0.5,NegativeHeight=0.5)
    FalloffSize=0.4
    FlipHeightScaling=False
    FlipWidthScaling=False
    OptionalParentItem=(Type=Bone,Name="None")
    RotationOffset=(X=0,Y=90,Z=130)
   )
   > alterTransform = RigVMFunctionReferenceNode(
    type=Bone
    transforms="((Bone=\"dyn_L_hair_3\",Transform=(Rotation=(X=0,Y=0,Z=0,W=1),Translation=(X=0,Y=0,Z=-5),Scale3D=(X=1,Y=1,Z=1))),(Bone=\"dyn_L_hair_4\",Transform=(Rotation=(X=0,Y=0,Z=0,W=1),Translation=(X=0,Y=4,Z=-1),Scale3D=(X=1,Y=1,Z=1))),(Bone=\"dyn_L_hair_2\",Transform=(Rot ...(truncated)"
    weight=@n1.OutputParam
   )
   > n2 = &RigUnit_SphericalPoseReader(
    ActiveRegionScaleFactors=(PositiveWidth=0.5,NegativeWidth=0.5,PositiveHeight=0.5,NegativeHeight=0.5)
    ActiveRegionSize=0.18
    Debug=(bDrawDebug=false,bDraw2D=False,bDrawLocalAxes=False,DebugScale=25,DebugSegments=20,DebugThickness=0.25)
    DriverAxis=(X=-1,Y=0,Z=0)
    DriverItem=(Type=Bone,Name="clavicle_r")
    FalloffRegionScaleFactors=(PositiveWidth=0.5,NegativeWidth=0.5,PositiveHeight=0.5,NegativeHeight=0.5)
    FalloffSize=0.4
    FlipHeightScaling=False
    FlipWidthScaling=False
    OptionalParentItem=(Type=Bone,Name="None")
    RotationOffset=(X=0,Y=-90,Z=130)
   )
   > alterTransform_1 = RigVMFunctionReferenceNode(
    type=Bone
    transforms="((Bone=\"dyn_R_hair_3\",Transform=(Rotation=(X=0,Y=0,Z=0,W=1),Translation=(X=0,Y=0,Z=-5),Scale3D=(X=1,Y=1,Z=1))),(Bone=\"dyn_R_hair_4\",Transform=(Rotation=(X=0,Y=0,Z=0,W=1),Translation=(X=0,Y=-4,Z=-1),Scale3D=(X=1,Y=1,Z=1))),(Bone=\"dyn_R_hair_2\",Transform=(Ro ...(truncated)"
    weight=@n2.OutputParam
   )
   > RigVMFunctionReturnNode()
  ]}
 New Function_ContainedGraph {nodes [
   RigVMFunctionEntryNode()
   > &RigUnit_ParentConstraint(
    AdvancedSettings=(RotationOrderForFilter=XZY)
    Child=(Type=Bone,Name="dyn_R_back_hair_03")
    Parents="((Item=(Type=Bone,Name=\"dyn_L_hair_1\"),Weight=0.0017),(Item=(Type=Bone,Name=\"dyn_R_hair_3\"),Weight=0.0076),(Item=(Type=Bone,Name=\"dyn_C_hair_3\"),Weight=0.009),(Item=(Type=Bone,Name=\"dyn_C_hair_2\"),Weight=0.0969),(Item=(Type=Bone,Name=\"dyn_C_hair_1\"),Weight ...(truncated)"
    Weight=1
   )
   > &RigUnit_ParentConstraint(
    AdvancedSettings=(RotationOrderForFilter=XZY)
    Child=(Type=Bone,Name="dyn_R_back_hair_04")
    Parents="((Item=(Type=Bone,Name=\"dyn_C_hair_1\"),Weight=0.0069),(Item=(Type=Bone,Name=\"dyn_R_hair_1\"),Weight=0.0308),(Item=(Type=Bone,Name=\"dyn_C_hair_4\"),Weight=0.0434),(Item=(Type=Bone,Name=\"dyn_R_hair_4\"),Weight=0.0513),(Item=(Type=Bone,Name=\"dyn_C_hair_2\"),Weigh ...(truncated)"
    Weight=1
   )
   > &RigUnit_ParentConstraint(
    AdvancedSettings=(RotationOrderForFilter=XZY)
    Child=(Type=Bone,Name="dyn_R_back_hair_05")
    Parents="((Item=(Type=Bone,Name=\"dyn_R_hair_6\"),Weight=0.0014),(Item=(Type=Bone,Name=\"dyn_R_hair_2\"),Weight=0.0098),(Item=(Type=Bone,Name=\"dyn_C_hair_5\"),Weight=0.0308),(Item=(Type=Bone,Name=\"dyn_C_hair_3\"),Weight=0.038),(Item=(Type=Bone,Name=\"dyn_R_hair_5\"),Weight ...(truncated)"
    Weight=1
   )
   > &RigUnit_ParentConstraint(
    AdvancedSettings=(RotationOrderForFilter=XZY)
    Child=(Type=Bone,Name="dyn_R_back_hair_06")
    Parents="((Item=(Type=Bone,Name=\"dyn_L_hair_5\"),Weight=0.0011),(Item=(Type=Bone,Name=\"dyn_R_hair_7\"),Weight=0.0027),(Item=(Type=Bone,Name=\"dyn_C_hair_6\"),Weight=0.0496),(Item=(Type=Bone,Name=\"dyn_C_hair_4\"),Weight=0.0604),(Item=(Type=Bone,Name=\"dyn_R_hair_6\"),Weigh ...(truncated)"
    Weight=1
   )
   > &RigUnit_ParentConstraint(
    AdvancedSettings=(RotationOrderForFilter=XZY)
    Child=(Type=Bone,Name="dyn_R_back_hair_07")
    Parents="((Item=(Type=Bone,Name=\"dyn_C_hair_5\"),Weight=0.0037),(Item=(Type=Bone,Name=\"dyn_R_hair_5\"),Weight=0.0162),(Item=(Type=Bone,Name=\"dyn_C_hair_7\"),Weight=0.0505),(Item=(Type=Bone,Name=\"dyn_C_hair_6\"),Weight=0.0826),(Item=(Type=Bone,Name=\"dyn_R_hair_6\"),Weigh ...(truncated)"
    Weight=1
   )
   > &RigUnit_ParentConstraint(
    AdvancedSettings=(RotationOrderForFilter=XZY)
    Child=(Type=Bone,Name="dyn_L_back_hair_03")
    Parents="((Item=(Type=Bone,Name=\"dyn_R_hair_1\"),Weight=0.0022),(Item=(Type=Bone,Name=\"dyn_L_hair_3\"),Weight=0.007),(Item=(Type=Bone,Name=\"dyn_C_hair_3\"),Weight=0.0096),(Item=(Type=Bone,Name=\"dyn_C_hair_2\"),Weight=0.1042),(Item=(Type=Bone,Name=\"dyn_C_hair_1\"),Weight ...(truncated)"
    Weight=1
   )
   > &RigUnit_ParentConstraint(
    AdvancedSettings=(RotationOrderForFilter=XZY)
    Child=(Type=Bone,Name="dyn_L_back_hair_04")
    Parents="((Item=(Type=Bone,Name=\"dyn_C_hair_1\"),Weight=0.0071),(Item=(Type=Bone,Name=\"dyn_L_hair_1\"),Weight=0.0365),(Item=(Type=Bone,Name=\"dyn_C_hair_4\"),Weight=0.0427),(Item=(Type=Bone,Name=\"dyn_L_hair_4\"),Weight=0.0559),(Item=(Type=Bone,Name=\"dyn_C_hair_2\"),Weigh ...(truncated)"
    Weight=1
   )
   > &RigUnit_ParentConstraint(
    AdvancedSettings=(RotationOrderForFilter=XZY)
    Child=(Type=Bone,Name="dyn_L_back_hair_05")
    Parents="((Item=(Type=Bone,Name=\"dyn_L_hair_6\"),Weight=0.0023),(Item=(Type=Bone,Name=\"dyn_L_hair_2\"),Weight=0.0143),(Item=(Type=Bone,Name=\"dyn_C_hair_5\"),Weight=0.0269),(Item=(Type=Bone,Name=\"dyn_C_hair_3\"),Weight=0.0311),(Item=(Type=Bone,Name=\"dyn_L_hair_5\"),Weigh ...(truncated)"
    Weight=1
   )
   > &RigUnit_ParentConstraint(
    AdvancedSettings=(RotationOrderForFilter=XZY)
    Child=(Type=Bone,Name="dyn_L_back_hair_06")
    Parents="((Item=(Type=Bone,Name=\"dyn_L_hair_3\"),Weight=0.0008),(Item=(Type=Bone,Name=\"dyn_L_hair_7\"),Weight=0.0035),(Item=(Type=Bone,Name=\"dyn_C_hair_6\"),Weight=0.0304),(Item=(Type=Bone,Name=\"dyn_C_hair_4\"),Weight=0.0383),(Item=(Type=Bone,Name=\"dyn_C_hair_5\"),Weigh ...(truncated)"
    Weight=1
   )
   > &RigUnit_ParentConstraint(
    AdvancedSettings=(RotationOrderForFilter=XZY)
    Child=(Type=Bone,Name="dyn_L_back_hair_07")
    Parents="((Item=(Type=Bone,Name=\"dyn_C_hair_5\"),Weight=0.0034),(Item=(Type=Bone,Name=\"dyn_L_hair_5\"),Weight=0.0202),(Item=(Type=Bone,Name=\"dyn_C_hair_7\"),Weight=0.0341),(Item=(Type=Bone,Name=\"dyn_C_hair_6\"),Weight=0.0676),(Item=(Type=Bone,Name=\"dyn_L_hair_6\"),Weigh ...(truncated)"
    Weight=1
   )
   > RigVMFunctionReturnNode()
  ]}
 New Function_ContainedGraph {nodes [
   n1 = RigVMFunctionEntryNode()
   RigVMFunction_MathBoolToInteger = &RigVMFunction_MathBoolToInteger(Value=@n1.onlyTwist)
   RigVMFunction_MathQuaternionSwingTwist = &RigVMFunction_MathQuaternionSwingTwist(
    twistAxis=@n1.twistAxis
    Input=&RigUnit_GetRelativeTransformForItem(bChildInitial=False, bParentInitial=False, child=@n1.child, parent=@n1.parent).RelativeTransform.Rotation
   )
   n2 = RigVMDispatchNode(
    Values=((),())
    Index=@RigVMFunction_MathBoolToInteger.Result
    Values.0=@RigVMFunction_MathQuaternionSwingTwist.Swing
    Values.1=@RigVMFunction_MathQuaternionSwingTwist.Twist
   ) {template="DISPATCH_RigVMDispatch_SelectInt32(in Index,in Values,out Result)"}
   RigVMFunction_MathQuaternionToEuler = &RigVMFunction_MathQuaternionToEuler(Value=@n2.Result, RotationOrder=@n1.Rotation Order)
   > DISPATCH_RigVMDispatch_Print = RigVMDispatchNode(ScreenColor=(R=1,G=1,B=1,A=1), ScreenDuration=1.5, Enabled=@n1.debugRotation, Value=@RigVMFunction_MathQuaternionToEuler.Result) {template="DISPATCH_RigVMDispatch_Print(in Prefix,in Value,in Enabled,in ScreenDuration,in ScreenColor)"}
   Cast = RigVMDispatchNode(Value=@n1.triggerAxis) {template="DISPATCH_RigVMDispatch_CastEnumToInt(in Value,out Result)"}
   DISPATCH_RigVMDispatch_SelectInt32 = RigVMDispatchNode(
    Values=(0,0,0)
    Index=@Cast.Result
    Values.0=@RigVMFunction_MathQuaternionToEuler.Result.X
    Values.1=@RigVMFunction_MathQuaternionToEuler.Result.Y
    Values.2=@RigVMFunction_MathQuaternionToEuler.Result.Z
   ) {template="DISPATCH_RigVMDispatch_SelectInt32(in Index,in Values,out Result)"}
   Remap = &RigVMFunction_MathDoubleRemap(
    bClamp=False
    Value=@DISPATCH_RigVMDispatch_SelectInt32.Result
    TargetMinimum=@n1.outputMinMax.X
    TargetMaximum=@n1.outputMinMax.Y
    SourceMinimum=@n1.rotationMinMax.X
    SourceMaximum=@n1.rotationMinMax.Y
   )
   Absolute = &RigVMFunction_MathDoubleAbs(Value=@Remap.Result)
   Multiply = &RigVMFunction_MathDoubleMul(
    A=&RigVMFunction_MathDoubleClamp(Value=RigVMDispatchNode(True=@Absolute.Result, Condition=@n1.absoluteValue, False=@Remap.Result) {template="DISPATCH_RigVMDispatch_If(in Condition,in True,in False,out Result)"}.Result).Result
    B=@n1.premult
   )
   > RigVMFunctionReturnNode(outputQuat=@n2.Result, outputWeight=@Multiply.Result)
  ]}
 RigVMFunctionLibrary {nodes [
   Constraints = RigVMCollapseNode(ExecuteContext=())
   DeltaRotation = RigVMCollapseNode(ExecuteContext=(), Rotation Order=ZYX, twistAxis=(X=1,Y=0,Z=0))
   DeltaRotation v4 = RigVMCollapseNode(
    ExecuteContext=()
    Rotation Order=ZYX
    child=(Type=Bone,Name="None")
    outputMinMax=(X=0,Y=1)
    parent=(Type=Bone,Name="None")
    premult=1
    rotationMinMax=(X=0,Y=0)
    twistAxis=(X=1,Y=0,Z=0)
   )
   Hair Adjustments = RigVMCollapseNode(ExecuteContext=())
   alterTransform = RigVMCollapseNode(type=Bone, ExecuteContext=(), transforms=())
   claviclesBackFoward = RigVMCollapseNode(ExecuteContext=())
   controllers = RigVMCollapseNode(ExecuteContext=(), debug=False)
   splineIK = RigVMCollapseNode(ExecuteContext=())
  ]}
 RigVMFunction_Sequence_1_ContainedGraph {nodes [
   n1 = RigVMFunctionEntryNode()
   RigVMFunction_Sequence = &RigVMFunction_Sequence(ExecuteContext=@n1.ExecuteContext)
   n2 = &RigVMFunction_Sequence(ExecuteContext=@RigVMFunction_Sequence.B)
   RigVMFunctionReturnNode(A=@RigVMFunction_Sequence.A, B=@n2.A, C=@n2.B)
  ]}
 RigVMFunction_Sequence_1_ContainedGraph {nodes [
   n1 = RigVMFunctionEntryNode()
   RigVMFunction_Sequence = &RigVMFunction_Sequence(ExecuteContext=@n1.ExecuteContext)
   n2 = &RigVMFunction_Sequence(ExecuteContext=@RigVMFunction_Sequence.B)
   n3 = &RigVMFunction_Sequence(ExecuteContext=@n2.B)
   RigVMFunctionReturnNode(A=@RigVMFunction_Sequence.A, B=@n2.A, C=@n3.A, D=@n3.B)
  ]}
 RigVMFunction_Sequence_1_ContainedGraph {nodes [
   n1 = RigVMFunctionEntryNode()
   RigVMFunction_Sequence = &RigVMFunction_Sequence(ExecuteContext=@n1.ExecuteContext)
   n2 = &RigVMFunction_Sequence(ExecuteContext=@RigVMFunction_Sequence.B)
   n3 = &RigVMFunction_Sequence(ExecuteContext=@n2.B)
   n4 = &RigVMFunction_Sequence(ExecuteContext=@n3.B)
   RigVMFunctionReturnNode(A=@RigVMFunction_Sequence.A, B=@n2.A, C=@n3.A, D=@n4.A, E=@n4.B)
  ]}
 RigVMFunction_Sequence_3_1_ContainedGraph {nodes [
   n1 = RigVMFunctionEntryNode()
   n2 = &RigVMFunction_Sequence(ExecuteContext=@n1.ExecuteContext)
   n3 = &RigVMFunction_Sequence(ExecuteContext=@n2.B)
   RigVMFunctionReturnNode(A=@n2.A, B=@n3.A, C=@n3.B)
  ]}
 RigVMModel {nodes [
   RigUnit_BeginExecution = &RigUnit_BeginExecution()
   controllers = RigVMFunctionReferenceNode(debug=false, ExecuteContext=@RigUnit_BeginExecution.ExecutePin)
   > splineIK = RigVMFunctionReferenceNode()
   > Hair Adjustments = RigVMFunctionReferenceNode()
  ]}
 RigVMModel Rig_4 {nodes [
   n1 = &RigUnit_PrepareForExecution()
   RigVMFunction_Sequence = RigVMAggregateNode(ExecuteContext=@n1.ExecutePin)
   n2 = &RigUnit_ItemArray(Items="((Type=Control,Name=\"dyn_L_hair_1_ctrl\"),(Type=Control,Name=\"dyn_L_hair_2_ctrl\"),(Type=Control,Name=\"dyn_L_hair_3_ctrl\"),(Type=Control,Name=\"dyn_L_hair_4_ctrl\"),(Type=Control,Name=\"dyn_L_hair_5_ctrl\"),(Type=Control,Name=\"dyn_L_hair_6_ctrl\"),(Type=Control,N ...(truncated)")
   RigVMVariableNode(Value=@n2.Items, ExecuteContext=@RigVMFunction_Sequence.A) {variable=l_dyn_hair_controls}
   > RigVMVariableNode(Value=/Script/ControlRig.RigUnit_HierarchyGetPoseItemArray(ElementType=All, Initial=False, ItemsToGet=@n2.Items) {is=struct RigVMUnitNode}.Pose) {variable=l_dyn_hair_pose}
   RigUnit_ItemArray = &RigUnit_ItemArray(Items="((Type=Control,Name=\"dyn_R_hair_1_ctrl\"),(Type=Control,Name=\"dyn_R_hair_2_ctrl\"),(Type=Control,Name=\"dyn_R_hair_3_ctrl\"),(Type=Control,Name=\"dyn_R_hair_4_ctrl\"),(Type=Control,Name=\"dyn_R_hair_5_ctrl\"),(Type=Control,Name=\"dyn_R_hair_6_ctrl\"),(Type=Control,N ...(truncated)")
   RigVMVariableNode(Value=@RigUnit_ItemArray.Items, ExecuteContext=@RigVMFunction_Sequence.B) {variable=r_dyn_hair_controls}
   > RigVMVariableNode(Value=&RigUnit_HierarchyGetPoseItemArray(ElementType=All, Initial=False, ItemsToGet=@RigUnit_ItemArray.Items).Pose) {variable=r_dyn_hair_pose}
   n3 = &RigUnit_ItemArray(Items="((Type=Control,Name=\"dyn_C_hair_1_ctrl\"),(Type=Control,Name=\"dyn_C_hair_2_ctrl\"),(Type=Control,Name=\"dyn_C_hair_3_ctrl\"),(Type=Control,Name=\"dyn_C_hair_4_ctrl\"),(Type=Control,Name=\"dyn_C_hair_5_ctrl\"),(Type=Control,Name=\"dyn_C_hair_6_ctrl\"),(Type=Control,N ...(truncated)")
   RigVMVariableNode(Value=@n3.Items, ExecuteContext=@RigVMFunction_Sequence.C) {variable=c_dyn_hair_controls}
   > RigVMVariableNode(Value=&RigUnit_HierarchyGetPoseItemArray(ElementType=All, Initial=False, ItemsToGet=@n3.Items).Pose) {variable=c_dyn_hair_pose}
  ]}
]
hierarchy [
 root {loc=<0 0 0>, rot=<0 0 0 1>}
 pelvis {parent=root, loc=<0 -0.5301 98.0571>, rot=<0.0127 0.707 -0.0127 -0.707>}
 spine_01 {parent=pelvis, loc=<6.1208 -0.9328 0>, rot=<0 0 0.034 -0.9994>}
 spine_02 {parent=spine_01, loc=<6.5996 -0.2095 0>, rot=<0 0 0.0238 -0.9997>}
 spine_03 {parent=spine_02, loc=<6.4914 -0.1237 0>, rot=<0 0 0.0149 -0.9999>}
 spine_04 {parent=spine_03, loc=<6.9598 0.0794 0>, rot=<0 0 -0.0132 -0.9999>}
 spine_05 {parent=spine_04, loc=<6.9987 0 0>, rot=<0 0 -0.0015 -1>}
 neck_01 {parent=spine_05, loc=<11.3138 0.1529 0>, rot=<0 0 -0.2003 -0.9797>}
 neck_02 {parent=neck_01, loc=<4.5627 0 0>, rot=<0 0 0.0504 -0.9987>}
 head {parent=neck_02, loc=<5.3914 0 0>, rot=<0 0 0.115 -0.9934>}
 base_C_hair_1 {parent=head, loc=<6.9517 -5.4626 0>, rot=<-0.0949 0.7007 0.7007 -0.0949>}
 dyn_C_hair_1 {parent=base_C_hair_1, loc=<0 0 0>, rot=<0 0 0 1>}
 dyn_C_hair_2 {parent=dyn_C_hair_1, loc=<10.6015 0 0>, rot=<0 -0.0172 0 -0.9999>}
 dyn_C_hair_3 {parent=dyn_C_hair_2, loc=<10.6008 0 0>, rot=<0 0.0789 0 -0.9969>}
 dyn_C_hair_4 {parent=dyn_C_hair_3, loc=<10.5829 0 0>, rot=<0 0.0898 0 -0.996>}
 dyn_C_hair_5 {parent=dyn_C_hair_4, loc=<10.5974 0 -0.9162>, rot=<0 -0.0564 0 -0.9984>}
 dyn_C_hair_6 {parent=dyn_C_hair_5, loc=<10.3267 0 0>, rot=<0 0.0471 0 -0.9989>}
 dyn_C_hair_7 {parent=dyn_C_hair_6, loc=<10.7722 0 0>, rot=<0 -0.009 0 -1>}
 dyn_C_hair_8 {parent=dyn_C_hair_7, loc=<11.2752 0 0>, rot=<0 0 0 1>}
 base_L_hair_1 {parent=head, loc=<6.9314 -3.0124 -7.9806>, rot=<0.0956 -0.7006 -0.7001 0.0996>}
 dyn_L_hair_1 {parent=base_L_hair_1, loc=<0 0 0>, rot=<0 0 0 1>}
 dyn_L_hair_2 {parent=dyn_L_hair_1, loc=<10.5141 0.0015 -1.6576>, rot=<0.0014 -0.0775 0.0234 -0.9967>}
 dyn_L_hair_3 {parent=dyn_L_hair_2, loc=<10.7695 0 1.6777>, rot=<-0.0067 0.1556 0.0018 -0.9878>}
 dyn_L_hair_4 {parent=dyn_L_hair_3, loc=<10.67 0 0>, rot=<-0.0019 0.0735 -0.0194 -0.9971>}
 dyn_L_hair_5 {parent=dyn_L_hair_4, loc=<10.7954 0.0001 -0.916>, rot=<0.0008 -0.0439 0.0114 -0.999>}
 dyn_L_hair_6 {parent=dyn_L_hair_5, loc=<10.4728 0 0>, rot=<0.0003 -0.0122 -0.0014 -0.9999>}
 dyn_L_hair_7 {parent=dyn_L_hair_6, loc=<10.8252 0 0>, rot=<-0.0006 0.0233 0 -0.9997>}
 dyn_L_hair_8 {parent=dyn_L_hair_7, loc=<11.2857 0 0>, rot=<0 0 0 1>}
 base_R_hair_1 {parent=head, loc=<6.9319 -3.0124 7.9806>, rot=<-0.0996 0.7001 0.7006 -0.0956>}
 dyn_R_hair_1 {parent=base_R_hair_1, loc=<0 0 0>, rot=<0 0 0 1>}
 dyn_R_hair_2 {parent=dyn_R_hair_1, loc=<10.5143 -0.0015 -1.6575>, rot=<-0.0014 -0.0775 -0.0234 -0.9967>}
 dyn_R_hair_3 {parent=dyn_R_hair_2, loc=<10.7696 0 1.6778>, rot=<0.0067 0.1556 -0.0018 -0.9878>}
 dyn_R_hair_4 {parent=dyn_R_hair_3, loc=<10.6703 0 0.0001>, rot=<0.0019 0.0735 0.0194 -0.9971>}
 dyn_R_hair_5 {parent=dyn_R_hair_4, loc=<10.7954 -0.0001 -0.916>, rot=<-0.0008 -0.0439 -0.0114 -0.999>}
 dyn_R_hair_6 {parent=dyn_R_hair_5, loc=<10.4723 0 0>, rot=<-0.0003 -0.0122 0.0014 -0.9999>}
 dyn_R_hair_7 {parent=dyn_R_hair_6, loc=<10.8255 0 0>, rot=<0.0006 0.0233 0 -0.9997>}
 dyn_R_hair_8 {parent=dyn_R_hair_7, loc=<11.2857 0 0>, rot=<0 0 0 1>}
 base_L_back_hair_01 {parent=head, loc=<13.7449 -6.8492 -4.4155>, rot=<0.5192 -0.48 0.4489 -0.5463>}
 dyn_L_back_hair_01 {parent=base_L_back_hair_01, loc=<0 0 0>, rot=<0 0 0 1>}
 dyn_L_back_hair_02 {parent=dyn_L_back_hair_01, loc=<2.9431 2.1016 1.7682>, rot=<-0.0652 -0.6313 -0.0345 -0.772>}
 dyn_L_back_hair_03 {parent=dyn_L_back_hair_02, loc=<12.8109 0 0>, rot=<-0.006 -0.0517 0.0313 -0.9982>}
 dyn_L_back_hair_04 {parent=dyn_L_back_hair_03, loc=<14.2179 0 0>, rot=<-0.0004 -0.0043 -0.0071 -1>}
 dyn_L_back_hair_05 {parent=dyn_L_back_hair_04, loc=<15.7889 0 0>, rot=<-0.0011 -0.0125 0.0079 -0.9999>}
 dyn_L_back_hair_06 {parent=dyn_L_back_hair_05, loc=<12.8044 0 0>, rot=<-0.0032 -0.0596 0.0295 -0.9978>}
 dyn_L_back_hair_07 {parent=dyn_L_back_hair_06, loc=<13.9786 0 0>, rot=<0 0 0 1>}
 base_R_back_hair_01 {parent=head, loc=<13.7449 -6.8492 4.4155>, rot=<0.5463 -0.4489 0.48 -0.5192>}
 dyn_R_back_hair_01 {parent=base_R_back_hair_01, loc=<0 0 0>, rot=<0 0 0 1>}
 dyn_R_back_hair_02 {parent=dyn_R_back_hair_01, loc=<2.9432 -2.1016 1.7687>, rot=<0.0652 -0.6313 0.0345 -0.772>}
 dyn_R_back_hair_03 {parent=dyn_R_back_hair_02, loc=<12.8112 0 -0.0001>, rot=<0.006 -0.0517 -0.0313 -0.9982>}
 dyn_R_back_hair_04 {parent=dyn_R_back_hair_03, loc=<14.2184 0 -0.0001>, rot=<0.0004 -0.0043 0.0071 -1>}
 dyn_R_back_hair_05 {parent=dyn_R_back_hair_04, loc=<15.7887 0 0.0001>, rot=<0.0011 -0.0125 -0.0079 -0.9999>}
 dyn_R_back_hair_06 {parent=dyn_R_back_hair_05, loc=<12.804 0 0.0001>, rot=<0.0032 -0.0596 -0.0295 -0.9978>}
 dyn_R_back_hair_07 {parent=dyn_R_back_hair_06, loc=<13.9792 0 0>, rot=<0 0 0 -1>}
 base_L_front_hair_01 {parent=head, loc=<7.0918 4.7823 -6.3792>, rot=<-0.0187 0.7069 0.7069 0.0165>}
 dyn_L_front_hair_01 {parent=base_L_front_hair_01, loc=<0 0 0>, rot=<0 0 0 1>}
 dyn_L_front_hair_02 {parent=dyn_L_front_hair_01, loc=<6.6145 0 0>, rot=<0.0026 0.038 -0.1183 -0.9923>}
 dyn_L_front_hair_03 {parent=dyn_L_front_hair_02, loc=<7.4479 0 0>, rot=<0 0 0 1>}
 base_R_front_hair_01 {parent=head, loc=<7.0915 4.7823 6.3792>, rot=<-0.0165 -0.7069 -0.7069 0.0187>}
 dyn_R_front_hair_01 {parent=base_R_front_hair_01, loc=<0 0 0>, rot=<0 0 0 1>}
 dyn_R_front_hair_02 {parent=dyn_R_front_hair_01, loc=<6.6143 0 0>, rot=<-0.0026 0.038 0.1183 -0.9923>}
 dyn_R_front_hair_03 {parent=dyn_R_front_hair_02, loc=<7.4476 -0.0001 0>, rot=<0 0 0 1>}
 head_null_2 {parent=head, type=Null, loc=<0 0 0>, rot=<0 0 0 1>}
 FACIAL_C_FacialRoot {parent=head, loc=<2.1525 -0.7459 -0.0001>, rot=<0.4879 -0.5118 -0.4879 -0.5118>}
 FACIAL_C_UnderChin {parent=FACIAL_C_FacialRoot, loc=<-0.0003 8.3552 3.2913>, rot=<0.4883 -0.0001 -0.0001 -0.8727>}
 FACIAL_L_UnderChin {parent=FACIAL_C_FacialRoot, loc=<2.2387 7.2655 2.5009>, rot=<0.3914 -0.356 0.0775 -0.845>}
 FACIAL_R_UnderChin {parent=FACIAL_C_FacialRoot, loc=<-2.2439 7.2635 2.4994>, rot=<0.3913 0.3559 -0.0777 -0.8451>}
 FACIAL_C_Jaw {parent=FACIAL_C_FacialRoot, loc=<0.0008 2.2948 1.6065>, rot=<0.3295 -0.0001 0 -0.9441>}
 FACIAL_C_Jawline {parent=FACIAL_C_Jaw, loc=<-0.001 2.2998 7.3046>, rot=<0.2586 0 0 -0.966>}
 FACIAL_L_Ear {parent=FACIAL_C_FacialRoot, loc=<7.476 0.1705 -0.9389>, rot=<-0.0396 -0.0004 -0.0262 -0.9989>}
 FACIAL_R_Ear {parent=FACIAL_C_FacialRoot, loc=<-7.4784 0.1629 -0.9525>, rot=<-0.0396 0.0003 0.0261 -0.9989>}
 FACIAL_L_Jawline {parent=FACIAL_C_FacialRoot, loc=<4.1796 5.7582 4.2605>, rot=<0.1965 0.1584 -0.0526 -0.9662>}
 FACIAL_R_Jawline {parent=FACIAL_C_FacialRoot, loc=<-4.1819 5.7592 4.263>, rot=<0.1965 -0.1585 0.0525 -0.9662>}
 head_null {parent=neck_02, type=Null, loc=<5.3914 0 0>, rot=<0 0 0.115 -0.9934>}
 base_Necklace {parent=neck_01, loc=<0.7157 1.3284 0>, rot=<-0.5099 -0.4899 -0.4899 -0.5099>}
 dyn_Necklace {parent=base_Necklace, loc=<0 0 0>, rot=<0 0 0 1>}
 base_frontChain_c1 {parent=dyn_Necklace, loc=<3.3783 0 -1.5314>, rot=<0.9358 0 -0.3524 0>}
 dyn_frontChain_c1_01 {parent=base_frontChain_c1, loc=<0 0 0>, rot=<0 0 0 1>}
 dyn_frontChain_c1_02 {parent=dyn_frontChain_c1_01, loc=<1.2685 0 0>, rot=<0 -0.05 0 -0.9987>}
 base_frontChain_c2 {parent=dyn_Necklace, loc=<3.3124 0 -1.8158>, rot=<0.9409 0 -0.3388 0>}
 dyn_frontChain_c2_01 {parent=base_frontChain_c2, loc=<0 0 0>, rot=<0 0 0 1>}
 dyn_frontChain_c2_02 {parent=dyn_frontChain_c2_01, loc=<1.869 0 0>, rot=<0 -0.0633 0 -0.998>}
 clavicle_l {parent=spine_05, loc=<5.4831 -0.1053 -3.2946>, rot=<0.0465 -0.7289 -0.009 -0.6829>}
 clavicle_r {parent=spine_05, loc=<5.4831 -0.1053 3.2946>, rot=<0.7289 0.0465 -0.6829 0.009>}
 upperarm_r {parent=clavicle_r, loc=<-9.8658 1.2136 0.4105>, rot=<0.0164 -0.3245 -0.0269 -0.9454>}
 clavicle_r_null {parent=spine_05, type=Null, loc=<5.4831 -0.1053 3.2946>, rot=<0.7289 0.0465 -0.6829 0.009>}
 clavicle_l_null {parent=spine_05, type=Null, loc=<5.4831 -0.1053 -3.2946>, rot=<0.0465 -0.7289 -0.009 -0.6829>}
 head_null_sp {parent=spine_05, type=Null, loc=<20.6573 3.5489 0>, rot=<0 0 -0.036 -0.9994>}
 neck_01_null {parent=spine_05, type=Null, loc=<11.3138 0.1529 0>, rot=<0 0 -0.2003 -0.9797>}
 base_diamond {parent=spine_05, loc=<10.5092 5.3627 0>, rot=<0.013 -0.707 -0.707 0.013>}
 dyn_diamond {parent=base_diamond, loc=<0 0 0>, rot=<0 0 0 1>}
 spine_02_null {parent=pelvis, type=Null, loc=<12.6908 -1.5909 0>, rot=<0 0 -0.0578 0.9983>}
 dyn_R_hair_7_ctrl_null {type=Null, loc=<-9.4953 -14.7292 96.1939>, rot=<0.5577 -0.594 -0.4227 -0.3968>}
 dyn_R_hair_7_ctrl {controlType=EULER_TRANSFORM, parent=dyn_R_hair_7_ctrl_null, type=Control, loc=<-10.701 0.3917 -1.0286>, rot=<0.0021 0.0478 0.0183 -0.9987>}
 dyn_R_hair_6_ctrl_null {type=Null, loc=<-9.2115 -12.1921 106.644>, rot=<0.5485 -0.5631 -0.4428 -0.4313>}
 dyn_R_hair_6_ctrl {controlType=EULER_TRANSFORM, parent=dyn_R_hair_6_ctrl_null, type=Control, loc=<-10.2908 -0.0318 -3.1536>, rot=<0.004 0.1481 -0.0009 -0.989>}
 dyn_R_hair_5_ctrl_null {type=Null, loc=<-8.9082 -12.8298 117.384>, rot=<0.4781 -0.4917 -0.5218 -0.5073>}
 dyn_R_hair_5_ctrl {controlType=EULER_TRANSFORM, parent=dyn_R_hair_5_ctrl_null, type=Control, loc=<-10.7688 0.2382 0.2483>, rot=<-0.0002 -0.0115 0.0111 -0.9999>}
 dyn_R_hair_4_ctrl_null {type=Null, loc=<-8.843 -13.2203 128.151>, rot=<0.4894 -0.4923 -0.5105 -0.5074>}
 dyn_R_hair_4_ctrl {
  controlType=EULER_TRANSFORM
  parent=dyn_R_hair_4_ctrl_null
  type=Control
  loc=<-10.547 -0.4151 1.5637>
  rot=<-0.0019 -0.0735 -0.0194 -0.9971>
 }
 dyn_R_hair_3_ctrl_null {type=Null, loc=<-8.3639 -12.0398 138.745>, rot=<-0.515 0.5387 0.482 0.4608>}
 dyn_R_hair_3_ctrl {controlType=EULER_TRANSFORM, parent=dyn_R_hair_3_ctrl_null, type=Control, loc=<-10.2483 0.0154 3.3103>, rot=<-0.0067 -0.1556 0.0018 -0.9878>}
 dyn_R_hair_2_ctrl_null {type=Null, loc=<-7.9192 -7.6159 148.554>, rot=<0.5816 -0.6061 -0.3915 -0.3757>}
 dyn_R_hair_2_ctrl {controlType=EULER_TRANSFORM, parent=dyn_R_hair_2_ctrl_null, type=Control, loc=<-10.6324 0.5008 -1.6641>, rot=<0.0014 0.0775 0.0234 -0.9967>}
 dyn_R_hair_1_ctrl_null {type=Null, loc=<-7.9806 -4.7569 158.941>, rot=<-0.564 0.5608 0.4273 0.4298>}
 dyn_R_hair_1_ctrl {controlType=EULER_TRANSFORM, parent=dyn_R_hair_1_ctrl_null, type=Control, loc=<152.0224 -7.1151 46.7668>, rot=<0.564 -0.5608 -0.4273 0.4298>}
 dyn_L_hair_7_ctrl_null {type=Null, loc=<9.4953 -14.7292 96.1939>, rot=<0.594 -0.5577 -0.3968 -0.4227>}
 dyn_L_hair_7_ctrl {
  controlType=EULER_TRANSFORM
  parent=dyn_L_hair_7_ctrl_null
  type=Control
  loc=<-10.7006 -0.3916 -1.0285>
  rot=<-0.0021 0.0478 -0.0183 -0.9987>
 }
 dyn_L_hair_6_ctrl_null {type=Null, loc=<9.2115 -12.1921 106.6437>, rot=<0.5631 -0.5485 -0.4313 -0.4428>}
 dyn_L_hair_6_ctrl {controlType=EULER_TRANSFORM, parent=dyn_L_hair_6_ctrl_null, type=Control, loc=<-10.2913 0.0318 -3.1537>, rot=<-0.004 0.1481 0.0009 -0.989>}
 dyn_L_hair_5_ctrl_null {type=Null, loc=<8.9082 -12.8298 117.3842>, rot=<0.4917 -0.4781 -0.5073 -0.5218>}
 dyn_L_hair_5_ctrl {
  controlType=EULER_TRANSFORM
  parent=dyn_L_hair_5_ctrl_null
  type=Control
  loc=<-10.7687 -0.2382 0.2483>
  rot=<0.0002 -0.0115 -0.0111 -0.9999>
 }
 dyn_L_hair_4_ctrl_null {type=Null, loc=<8.843 -13.2203 128.1512>, rot=<0.4923 -0.4894 -0.5074 -0.5105>}
 dyn_L_hair_4_ctrl {controlType=EULER_TRANSFORM, parent=dyn_L_hair_4_ctrl_null, type=Control, loc=<-10.5466 0.4151 1.5637>, rot=<0.0019 -0.0735 0.0194 -0.9971>}
 dyn_L_hair_3_ctrl_null {type=Null, loc=<8.364 -12.0398 138.7448>, rot=<-0.5387 0.515 0.4608 0.482>}
 dyn_L_hair_3_ctrl {
  controlType=EULER_TRANSFORM
  parent=dyn_L_hair_3_ctrl_null
  type=Control
  loc=<-10.2482 -0.0154 3.3103>
  rot=<0.0067 -0.1556 -0.0018 -0.9878>
 }
 dyn_L_hair_2_ctrl_null {type=Null, loc=<7.9192 -7.6159 148.5537>, rot=<0.6061 -0.5816 -0.3757 -0.3915>}
 dyn_L_hair_2_ctrl {
  controlType=EULER_TRANSFORM
  parent=dyn_L_hair_2_ctrl_null
  type=Control
  loc=<-10.6322 -0.5008 -1.664>
  rot=<-0.0014 0.0775 -0.0234 -0.9967>
 }
 dyn_L_hair_1_ctrl_null {type=Null, loc=<7.9806 -4.7569 158.9405>, rot=<0.5608 -0.564 -0.4298 -0.4273>}
 dyn_L_hair_1_ctrl {controlType=EULER_TRANSFORM, parent=dyn_L_hair_1_ctrl_null, type=Control, loc=<152.0219 7.1151 46.7667>, rot=<-0.5608 0.564 0.4298 -0.4273>}
 dyn_C_hair_7_ctrl_null {type=Null, loc=<0 -15.0214 96.4145>, rot=<0.5712 -0.5712 -0.4168 -0.4168>}
 dyn_C_hair_7_ctrl {controlType=EULER_TRANSFORM, parent=dyn_C_hair_7_ctrl_null, type=Control, loc=<-10.4268 0 -1.8123>, rot=<0 0.0859 0 -0.9963>}
 dyn_C_hair_6_ctrl_null {type=Null, loc=<0 -13.5663 106.8971>, rot=<0.5333 -0.5333 -0.4644 -0.4644>}
 dyn_C_hair_6_ctrl {controlType=EULER_TRANSFORM, parent=dyn_C_hair_6_ctrl_null, type=Control, loc=<-10.4243 0 -1.9059>, rot=<0 0.0903 0 -0.9959>}
 dyn_C_hair_5_ctrl_null {type=Null, loc=<0 -14.0207 117.4844>, rot=<0.4892 -0.4892 -0.5106 -0.5106>}
 dyn_C_hair_5_ctrl {controlType=EULER_TRANSFORM, parent=dyn_C_hair_5_ctrl_null, type=Control, loc=<-10.5971 0 0.0026>, rot=<0 -0.0001 0 -1>}
 dyn_C_hair_4_ctrl_null {type=Null, loc=<0 -14.4726 128.0719>, rot=<0.4892 -0.4892 -0.5105 -0.5105>}
 dyn_C_hair_4_ctrl {controlType=EULER_TRANSFORM, parent=dyn_C_hair_4_ctrl_null, type=Control, loc=<-10.4123 0 1.8926>, rot=<0 -0.0898 0 -0.996>}
 dyn_C_hair_3_ctrl_null {type=Null, loc=<0 -13.0257 138.5554>, rot=<-0.5331 0.5331 0.4646 0.4646>}
 dyn_C_hair_3_ctrl {controlType=EULER_TRANSFORM, parent=dyn_C_hair_3_ctrl_null, type=Control, loc=<-10.4687 0 1.6685>, rot=<0 -0.0789 0 -0.9969>}
 dyn_C_hair_2_ctrl_null {type=Null, loc=<0 -9.9416 148.6977>, rot=<0.5681 -0.5681 -0.421 -0.421>}
 dyn_C_hair_2_ctrl {controlType=EULER_TRANSFORM, parent=dyn_C_hair_2_ctrl_null, type=Control, loc=<-10.5953 0 -0.3639>, rot=<0 0.0172 0 -0.9999>}
 dyn_C_hair_1_ctrl_null {type=Null, loc=<0 -7.2072 158.9405>, rot=<-0.5608 0.5608 0.4307 0.4307>}
 dyn_C_hair_1_ctrl {controlType=EULER_TRANSFORM, parent=dyn_C_hair_1_ctrl_null, type=Control, loc=<151.7041 0 47.9572>, rot=<0.5608 -0.5608 -0.4307 0.4307>}
 spine_01_ctrl {controlType=EULER_TRANSFORM, type=Control, loc=<0 0 0>, rot=<0 0 0 1>}
 spine_02_ctrl {controlType=EULER_TRANSFORM, parent=spine_01_ctrl, type=Control, loc=<0 0 0>, rot=<0 0 0 1>}
 spine_03_ctrl {controlType=EULER_TRANSFORM, parent=spine_02_ctrl, type=Control, loc=<0 0 0>, rot=<0 0 0 1>}
 spine_04_ctrl {controlType=EULER_TRANSFORM, parent=spine_03_ctrl, type=Control, loc=<0 0 0>, rot=<0 0 0 1>}
 spine_05_ctrl {controlType=EULER_TRANSFORM, parent=spine_04_ctrl, type=Control, loc=<0 0 0>, rot=<0 0 0 1>}
 clavicle_r_ctrl {controlType=EULER_TRANSFORM, parent=spine_05_ctrl, type=Control, loc=<0 0 0>, rot=<0 0 0 1>}
 upperarm_r_ctrl {controlType=EULER_TRANSFORM, parent=clavicle_r_ctrl, type=Control, loc=<0 0 0>, rot=<0 0 0 1>}
 clavicle_l_ctrl {controlType=EULER_TRANSFORM, parent=spine_05_ctrl, type=Control, loc=<0 0 0>, rot=<0 0 0 1>}
 upperarm_l_ctrl {controlType=EULER_TRANSFORM, parent=clavicle_l_ctrl, type=Control, loc=<0 0 0>, rot=<0 0 0 1>}
 neck_01_ctrl {controlType=EULER_TRANSFORM, parent=spine_05_ctrl, type=Control, loc=<0 0 0>, rot=<0 0 0 1>}
 neck_02_ctrl {controlType=EULER_TRANSFORM, parent=neck_01_ctrl, type=Control, loc=<0 0 0>, rot=<0 0 0 1>}
 head_ctrl {controlType=EULER_TRANSFORM, parent=neck_02_ctrl, type=Control, loc=<0 0 0>, rot=<0 0 0 1>}
 pelvis_ctrl {controlType=EULER_TRANSFORM, type=Control, loc=<0 0 0>, rot=<0 0 0 1>}
 thigh_r_ctrl {controlType=EULER_TRANSFORM, parent=pelvis_ctrl, type=Control, loc=<0 0 0>, rot=<0 0 0 1>}
 thigh_l_ctrl {controlType=EULER_TRANSFORM, parent=pelvis_ctrl, type=Control, loc=<0 0 0>, rot=<0 0 0 1>}
 Head Look Right {type=Curve, loc=<0 0 0>, rot=<0 0 0 1>}
 SPR Head Look Right FE {type=Curve, loc=<0 0 0>, rot=<0 0 0 1>}
 SPR Head Look Left FE {type=Curve, loc=<0 0 0>, rot=<0 0 0 1>}
]
variables [
 FE Idle {type=bool, default=False}
 c_dyn_hair_controls {type="TArray<FRigElementKey>"}
 c_dyn_hair_pose {type=FRigPose, default=(Elements=,HierarchyTopologyVersion=-1,PoseHash=-1)}
 is_PistolEquipped lerp {type=double, default=0}
 l_dyn_hair_controls {type="TArray<FRigElementKey>"}
 l_dyn_hair_pose {type=FRigPose, default=(Elements=,HierarchyTopologyVersion=-1,PoseHash=-1)}
 pub disableUpperLimbsCorrection {type=bool, default=False}
 pub diveFloating {type=bool, default=False}
 pub isAssaultRifleEquipped {type=double, default=0}
 pub isPistolEquipped {type=bool, default=False}
 pub isTargeting {type=bool, default=False}
 r_dyn_hair_controls {type="TArray<FRigElementKey>"}
 r_dyn_hair_pose {type=FRigPose, default=(Elements=,HierarchyTopologyVersion=-1,PoseHash=-1)}
]
